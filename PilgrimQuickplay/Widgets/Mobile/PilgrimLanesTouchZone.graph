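# Mobile touch input zone for Pilgrim music battles, with one touch button per lane that fires that lane's press input action.
# On battle instance init it reads the local game's lane count, reveals the panel, and shows the fifth button and some spacers for five lanes.
# Setup for Track then adjusts spacers and zone width for lane count and expanded-track mode.
# It reruns when the expanded-track setting event arrives through the event router.
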WidgetBlueprint /PilgrimQuickplay/Widgets/Mobile/PilgrimLanesTouchZone
graphs [
 EventGraph {nodes [
   n1 = Event PreConstruct()
   > Parent: PreConstruct(IsDesignTime=@n1.IsDesignTime)
   Event Construct()
   > n2 = PilgrimWaitForEventRouter!(execute:Exec, ContextActor:Actor Object Reference=GetOwningPlayer()) -> (then:Exec, OnEventRouterAvailable:Exec, EventRouter:Gameplay Event Router Component Object Reference) {is=title K2Node_AsyncAction}
   n3 = StartListeningToEvent!(
    execute:Exec=@n2.OnEventRouterAvailable
    Target:Gameplay Event Router Component Object Reference=@n2
    EventType:Script Struct Object Reference=/Script/PilgrimSettingsRuntime.PilgrimEvent_PilgrimSettings_EnableMobileExpandedTrackChanged
    Context:Object Reference
    EventBubblingRule:EEventBubblingRule Enum
   ) -> (
    then:Exec
    AsyncTaskProxy:Async Action Start Listening To Event Object Reference
    OnEventReceived:Exec
    Payload:Pilgrim Event Pilgrim Settings Enable Mobile Expanded Track Changed Structure
   )
   > Add(
    TargetArray:Array of Async Action Start Listening To Event Object References=Get GameplayEventHandles() -> (GameplayEventHandles:Array of Async Action Start Listening To Event Object References)
    NewItem:"Async Action Start Listening To Event Object Reference (by ref)"=@n3
   ) {is=title K2Node_CallArrayFunction}
   Setup for Track(execute=@n3.OnEventReceived)
   n4 = Event Tick()
   > Parent: Tick!(execute, MyGeometry:Geometry Structure=@n4.MyGeometry, InDeltaTime:float=@n4.InDeltaTime) -> (then) {is=title K2Node_CallParentFunction}
   Event OnMusicBattleInstanceInitializeEvent() {is=title K2Node_Event}
   n5 = GetLocalPilgrimGameForPlayer!(
    self:Pilgrim Music Battle Instance Object Reference=Get MusicBattleInstance() -> (MusicBattleInstance:Pilgrim Music Battle Instance Object Reference)
    InPC:Player Controller Object Reference=GetOwningPlayer()
   ) -> (ReturnValue:Pilgrim Game Object Reference)
   > Set IsUsing5Lanes(IsUsing5Lanes:bool="Equal (Integer)"(A=GetNumLanes!(self:Pilgrim Game Object Reference=@n5) -> (ReturnValue:int=0), B=5)) -> (Output_Get:bool)
   > SetVisibility(self=Get MobileTouchPanel() -> (MobileTouchPanel:Overlay Object Reference), InVisibility=SelfHitTestInvisible)
   > SetVisibility(
    self=Get Lane5ButtonSizeBox() -> (Lane5ButtonSizeBox:Size Box Object Reference)
    InVisibility=Select(Option 0=Collapsed, Option 1=SelfHitTestInvisible, Index=Get IsUsing5Lanes() -> (IsUsing5Lanes:bool))
   )
   > SetVisibility(
    self=[Get 5LanesSpacer0() -> (5LanesSpacer0:Spacer Object Reference), Get 5LanesSpacer2() -> (5LanesSpacer2:Spacer Object Reference)]
    InVisibility=Select(Option 0=Collapsed, Option 1=HitTestInvisible, Index=Get IsUsing5Lanes())
   )
   > Setup for Track()
  ]}
 Setup for Track {nodes [
   Setup for Track() {is=title K2Node_FunctionEntry}
   > n1 = Set Is Using Expanded Track(Is Using Expanded Track:bool=IsMobileAndShouldUseExpandedTrack!() -> (ReturnValue:bool)) -> (Output_Get:bool)
   > SetSize!(
    execute
    self:Spacer Object Reference=Get 5LanesSpacer0()
    InSize:Vector 2D Structure=Select(Option 0:Vector 2D Structure=(X=15,Y=0), Option 1:Vector 2D Structure=(X=50,Y=0), Index=@n1) -> (ReturnValue:Vector 2D Structure)
   ) -> (then)
   n2 = Get IsUsing5Lanes()
   > SetVisibility(
    self=[Get 5LanesSpacer3() -> (5LanesSpacer3:Spacer Object Reference), Get 5LanesSpacer4() -> (5LanesSpacer4:Spacer Object Reference)]
    InVisibility=Select(Option 0:ESlateVisibility Enum=Collapsed, Option 1:ESlateVisibility Enum=HitTestInvisible, Index=AND Boolean(A=@n1, B=@n2)) -> (ReturnValue:ESlateVisibility Enum)
   )
   n3 = Get 4-Lanes Expanded Width Override() -> (4-Lanes Expanded Width Override:double)
   n4 = Get 4-Lanes Width Override() -> (4-Lanes Width Override:double)
   n5 = Get Is Using Expanded Track() -> (Is Using Expanded Track:bool)
   n6 = Get 5-Lanes Expanded Width Override() -> (5-Lanes Expanded Width Override:double)
   n7 = Get 5-Lanes Width Override() -> (5-Lanes Width Override:double)
   > SetWidthOverride(
    self=Get InputZoneSizeBox() -> (InputZoneSizeBox:Size Box Object Reference)
    InWidthOverride=SelectFloat(
     A=Select(Option 0:double=@n7, Option 1:double=@n6, Index=@n5) -> (ReturnValue:double)
     B=Select(Option 0=@n4, Option 1=@n3, Index=@n5)
     bPickA=Get IsUsing5Lanes()
    )
   )
  ]}
]
variables [
 4-Lanes Expanded Width Override {type=number, default=1320}
 4-Lanes Width Override {type=number, default=580}
 5-Lanes Expanded Width Override {type=number, default=1500}
 5-Lanes Width Override {type=number, default=720}
 Is Using Expanded Track {type=boolean}
 IsUsing5Lanes {type=boolean}
]
widgets {
 info {inheritedWidgetCount=0, parentClass=/Script/PilgrimCoreUI.PilgrimGameWidgetBase, rootWidgetClass=&Overlay, widgetCount=18}
 widgets [
  MobileTouchPanel {widgetClassPath=&Overlay, bIsVariable=true, widget=MobileTouchPanel, properties {visibility=Collapsed}}
  ScaleBox_95 {
   widgetClassPath=&ScaleBox
   parent=MobileTouchPanel
   widget=ScaleBox_95
   slot {class=OverlaySlot, properties {horizontalAlignment=HAlign_Fill, verticalAlignment=VAlign_Fill}}
  }
  Overlay_347 {widgetClassPath=&Overlay, parent=ScaleBox_95, widget=Overlay_347, slot {class=ScaleBoxSlot}}
  TrackRenderTarget {
   widgetClassPath=&Image
   bIsVariable=true
   parent=Overlay_347
   widget=TrackRenderTarget
   properties {
    visibility=HitTestInvisible
    brush {
     bIsDynamicallyLoaded=false
     drawAs=NoDrawType
     imageType=NoImage
     mirroring=NoMirror
     resourceName=None
     resourceObject=null
     tiling=NoTile
     imageSize=<1000 1000>
     margin=<ltrb 0 0 0 0>
     outlineSettings {
      bUseBrushTransparency=false
      roundingType=HalfHeightRadius
      width=0
      color {colorUseRule=UseColor_Specified, specifiedColor=<rgba 0 0 0 0>}
      cornerRadii=<0 0 0 0>
     }
     tintColor {colorUseRule=UseColor_Specified, specifiedColor=<rgba 1 1 1 1>}
     uVRegion {bIsValid=false, max=<0 0>, min=<0 0>}
    }
   }
   slot {class=OverlaySlot, properties {horizontalAlignment=HAlign_Center, verticalAlignment=VAlign_Center}}
  }
  SZ_InputZone {
   widgetClassPath=&SafeZone
   bIsVariable=true
   parent=Overlay_347
   widget=SZ_InputZone
   properties {padLeft=false, padRight=false, padTop=false}
   slot {class=OverlaySlot, properties {horizontalAlignment=HAlign_Center, verticalAlignment=VAlign_Bottom}}
  }
  InputZoneSizeBox {
   widgetClassPath=&SizeBox
   bIsVariable=true
   parent=SZ_InputZone
   widget=InputZoneSizeBox
   properties {bOverride_HeightOverride=true, bOverride_WidthOverride=true, heightOverride=280, widthOverride=720}
   slot {class=SafeZoneSlot}
  }
  HorizontalBox_0 {widgetClassPath=&HorizontalBox, parent=InputZoneSizeBox, widget=HorizontalBox_0, slot {class=SizeBoxSlot}}
  5LanesSpacer0 {
   widgetClassPath=&Spacer
   bIsVariable=true
   parent=HorizontalBox_0
   widget=5LanesSpacer0
   properties {visibility=Collapsed, size=<15 1>}
   slot {class=HorizontalBoxSlot}
  }
  PilgrimTouchButtonBase_lane1_1 {
   widgetClassPath=/Script/PilgrimCoreRuntime.PilgrimTouchButtonBase
   parent=HorizontalBox_0
   widget=PilgrimTouchButtonBase_lane1_1
   properties {
    bConsumePointerInput=true
    deactivateInputActionIfDragOutsideHitbox=true
    pressAction=/PilgrimSettings/Input/InputActions/IA_PilgrimLane1Press
    visibility=Visible
   }
   slot {class=HorizontalBoxSlot, properties {size {sizeRule=Fill, value=1}}}
  }
  5LanesSpacer1 {
   widgetClassPath=&Spacer
   bIsVariable=true
   parent=HorizontalBox_0
   widget=5LanesSpacer1
   properties {visibility=Collapsed, size=<12 1>}
   slot {class=HorizontalBoxSlot}
  }
  PilgrimTouchButtonBase_lane2_1 {
   widgetClassPath=&PilgrimTouchButtonBase
   parent=HorizontalBox_0
   widget=PilgrimTouchButtonBase_lane2_1
   properties {
    bConsumePointerInput=true
    deactivateInputActionIfDragOutsideHitbox=true
    pressAction=/PilgrimSettings/Input/InputActions/IA_PilgrimLane2Press
    visibility=Visible
   }
   slot {class=HorizontalBoxSlot, properties {size {sizeRule=Fill, value=1}}}
  }
  5LanesSpacer2 {
   widgetClassPath=&Spacer
   bIsVariable=true
   parent=HorizontalBox_0
   widget=5LanesSpacer2
   properties {visibility=Collapsed, size=<12 1>}
   slot {class=HorizontalBoxSlot}
  }
  PilgrimTouchButtonBase_lane3_1 {
   widgetClassPath=&PilgrimTouchButtonBase
   parent=HorizontalBox_0
   widget=PilgrimTouchButtonBase_lane3_1
   properties {
    bConsumePointerInput=true
    deactivateInputActionIfDragOutsideHitbox=true
    pressAction=/PilgrimSettings/Input/InputActions/IA_PilgrimLane3Press
    visibility=Visible
   }
   slot {class=HorizontalBoxSlot, properties {size {sizeRule=Fill, value=1}}}
  }
  5LanesSpacer3 {
   widgetClassPath=&Spacer
   bIsVariable=true
   parent=HorizontalBox_0
   widget=5LanesSpacer3
   properties {visibility=Collapsed, size=<12 1>}
   slot {class=HorizontalBoxSlot}
  }
  PilgrimTouchButtonBase_lane4_1 {
   widgetClassPath=&PilgrimTouchButtonBase
   parent=HorizontalBox_0
   widget=PilgrimTouchButtonBase_lane4_1
   properties {
    bConsumePointerInput=true
    deactivateInputActionIfDragOutsideHitbox=true
    pressAction=/PilgrimSettings/Input/InputActions/IA_PilgrimLane4Press
    visibility=Visible
   }
   slot {class=HorizontalBoxSlot, properties {size {sizeRule=Fill, value=1}}}
  }
  Lane5ButtonSizeBox {
   widgetClassPath=&SizeBox
   bIsVariable=true
   parent=HorizontalBox_0
   widget=Lane5ButtonSizeBox
   properties {visibility=Collapsed, widthOverride=156}
   slot {class=HorizontalBoxSlot, properties {size {sizeRule=Fill, value=1}}}
  }
  PilgrimTouchButtonBase_lane5 {
   widgetClassPath=&PilgrimTouchButtonBase
   parent=Lane5ButtonSizeBox
   widget=PilgrimTouchButtonBase_lane5
   properties {
    bConsumePointerInput=true
    deactivateInputActionIfDragOutsideHitbox=true
    pressAction=/PilgrimSettings/Input/InputActions/IA_PilgrimLane5Press_5Lane
    visibility=Visible
   }
   slot {class=SizeBoxSlot}
  }
  5LanesSpacer4 {
   widgetClassPath=&Spacer
   bIsVariable=true
   parent=HorizontalBox_0
   widget=5LanesSpacer4
   properties {visibility=Collapsed, size=<50 1>}
   slot {class=HorizontalBoxSlot}
  }
 ]
}
