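# Extra-large mobile touch button widget for Pilgrim, built on PilgrimTouchButtonBase, with a hitbox larger than its centered, scaled visual background and icon.
# On PreConstruct and Construct, UpdateBrush sets the icon to the Replace Slate Brush override when that brush has a valid resource.
# Otherwise it asks the owning player's Pilgrim controller component for the touch brush matching PressAction, applying it only if that lookup succeeds.
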WidgetBlueprint /PilgrimQuickplay/Widgets/Mobile/PilgrimTouchButton_XL
graphs [
 EventGraph {nodes [
   n1 = Event PreConstruct()
   UpdateBrush(execute=[@n1.then, @n2.then])
   n2 = Event Construct()
  ]}
 UpdateBrush {nodes [
   UpdateBrush() {is=title K2Node_FunctionEntry}
   > n1 = Is Valid(InputObject=GetBrushResource!(Brush:"Slate Brush Structure (by ref)"=Get Replace Slate Brush() -> (Replace Slate Brush:Slate Brush Structure)) -> (ReturnValue:Object Reference))
   n2 = GetTouchBrushForInputAction!(
    self:Pilgrim Player Controller Component Object Reference=GetComponentByClass!(
     self:Actor Object Reference=GetOwningPlayer()
     ComponentClass:Actor Component Class Reference=/PilgrimCore/Player/Components/BP_PilgrimPlayerControllerComponent.BP_PilgrimPlayerControllerComponent_C
    ) -> (ReturnValue:BP Pilgrim Player Controller Component Object Reference)
    InputAction:Input Action Object Reference=Get PressAction() -> (PressAction:Input Action Object Reference)
   ) -> (OutBrush:Slate Brush Structure, ReturnValue:bool)
   Branch(execute=@n1.Is Not Valid, Condition=@n2.ReturnValue)
   > SetBrush(self=Get Image_Icon() -> (Image_Icon:Image Object Reference), InBrush=@n2)
   SetBrush(execute=@n1.Is Valid, self=Get Image_Icon(), InBrush=Get Replace Slate Brush())
  ]}
]
variables [Replace Slate Brush {
  type=SlateBrush
  default {
   bIsDynamicallyLoaded=false
   drawAs=Image
   imageType=NoImage
   mirroring=NoMirror
   resourceName=None
   resourceObject=null
   tiling=NoTile
   imageSize=<32 32>
   margin=<ltrb 0 0 0 0>
   outlineSettings {
    bUseBrushTransparency=false
    roundingType=HalfHeightRadius
    width=0
    color {colorUseRule=UseColor_Specified, specifiedColor=<rgba 0 0 0 0>}
    cornerRadii=<0 0 0 1>
   }
   tintColor {colorUseRule=UseColor_Specified, specifiedColor=<rgba 1 1 1 1>}
   uVRegion {bIsValid=false, max=<0 0>, min=<0 0>}
  }
 }]
widgets {
 info {inheritedWidgetCount=1, parentClass=/Script/PilgrimCoreRuntime.PilgrimTouchButtonBase, rootWidgetClass=&SizeBox, widgetCount=8}
 widgets [
  SizeBox_HitboxSize {
   widgetClassPath=&SizeBox
   widget=SizeBox_HitboxSize
   properties {bOverride_HeightOverride=true, bOverride_WidthOverride=true, heightOverride=325, visibility=Visible, widthOverride=325}
  }
  SizeBox_VisualSize {
   widgetClassPath=&SizeBox
   parent=SizeBox_HitboxSize
   widget=SizeBox_VisualSize
   properties {bOverride_HeightOverride=true, bOverride_WidthOverride=true, heightOverride=220, widthOverride=220}
   slot {class=SizeBoxSlot, properties {horizontalAlignment=HAlign_Center, verticalAlignment=VAlign_Center}}
  }
  ScaleBox_1 {widgetClassPath=&ScaleBox, parent=SizeBox_VisualSize, widget=ScaleBox_1, slot {class=SizeBoxSlot}}
  SizeBox_1 {
   widgetClassPath=&SizeBox
   parent=ScaleBox_1
   widget=SizeBox_1
   properties {bOverride_HeightOverride=true, bOverride_WidthOverride=true, heightOverride=128, widthOverride=128}
   slot {class=ScaleBoxSlot}
  }
  Overlay_ButtonContents {widgetClassPath=&Overlay, parent=SizeBox_1, widget=Overlay_ButtonContents, slot {class=SizeBoxSlot}}
  Image_Background {
   widgetClassPath=&Image
   bIsVariable=true
   parent=Overlay_ButtonContents
   widget=Image_Background
   properties {brush {
     bIsDynamicallyLoaded=false
     drawAs=Image
     imageType=NoImage
     mirroring=NoMirror
     resourceName=None
     resourceObject=/Game/Athena/HUD/Atlas/Bacchus/T_ButtonsBG_01_Sprite
     tiling=NoTile
     imageSize=<128 128>
     margin=<ltrb 0 0 0 0>
     outlineSettings {
      bUseBrushTransparency=false
      roundingType=HalfHeightRadius
      width=0
      color {colorUseRule=UseColor_Specified, specifiedColor=<rgba 0 0 0 0>}
      cornerRadii=<0 0 0 0>
     }
     tintColor {colorUseRule=UseColor_Specified, specifiedColor=<rgba 1 1 1 1>}
     uVRegion {bIsValid=false, max=<0 0>, min=<0 0>}
    }}
   slot {class=OverlaySlot, properties {horizontalAlignment=HAlign_Fill, verticalAlignment=VAlign_Fill}}
  }
  SizeBox_0 {
   widgetClassPath=&SizeBox
   parent=Overlay_ButtonContents
   widget=SizeBox_0
   properties {
    bOverride_MaxDesiredHeight=true
    bOverride_MaxDesiredWidth=true
    bOverride_MinDesiredHeight=true
    bOverride_MinDesiredWidth=true
    heightOverride=128
    maxDesiredHeight=128
    maxDesiredWidth=128
    minDesiredHeight=60
    minDesiredWidth=60
    widthOverride=128
   }
   slot {class=OverlaySlot, properties {horizontalAlignment=HAlign_Center, verticalAlignment=VAlign_Center}}
  }
  Image_Icon {
   widgetClassPath=&Image
   bInherited=true
   bIsVariable=true
   parent=SizeBox_0
   widget=Image_Icon
   properties {
    visibility=HitTestInvisible
    brush {
     bIsDynamicallyLoaded=false
     drawAs=Image
     imageType=NoImage
     mirroring=NoMirror
     resourceName=None
     resourceObject=/PilgrimCore/Widgets/Mobile/Textures/T_Mobile_Lane1_Sprite
     tiling=NoTile
     imageSize=<64 64>
     margin=<ltrb 0 0 0 0>
     outlineSettings {
      bUseBrushTransparency=false
      roundingType=HalfHeightRadius
      width=0
      color {colorUseRule=UseColor_Specified, specifiedColor=<rgba 0 0 0 0>}
      cornerRadii=<0 0 0 0>
     }
     tintColor {colorUseRule=UseColor_Specified, specifiedColor=<rgba 1 1 1 1>}
     uVRegion {bIsValid=false, max=<0 0>, min=<0 0>}
    }
    colorAndOpacity=<rgba 1 1 1 0.6>
   }
   slot {class=SizeBoxSlot}
  }
 ]
}
