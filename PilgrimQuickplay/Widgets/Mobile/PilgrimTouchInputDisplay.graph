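# Mobile touch input overlay for Pilgrim music battles, with on-screen lane and Overdrive buttons in bottom-corner panels.
# Left and right visibility switchers pick 4- or 5-lane and lefty-flip variants from lane count and player settings.
# Each variant pairs touch buttons with polygonal input zone controllers.
# Event Router listeners refresh the layout on settings changes; the expanded-track handler toggles visibility and repositions 5-lane Overdrive.
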WidgetBlueprint /PilgrimQuickplay/Widgets/Mobile/PilgrimTouchInputDisplay
graphs [
 EventGraph {nodes [
   Event Construct()
   > n1 = PilgrimWaitForEventRouter!(execute:Exec, ContextActor:Actor Object Reference=GetOwningPlayer()) -> (then:Exec, OnEventRouterAvailable:Exec, EventRouter:Gameplay Event Router Component Object Reference) {is=title K2Node_AsyncAction}
   > SubscribeForChangingVisibilityByTag!(
    execute
    self:Pilgrim Game Widget Base Object Reference
    GameplayTag:"Gameplay Tag Structure (by ref)"=MakeLiteralGameplayTag!(Value:Gameplay Tag Structure=(TagName="Pilgrim.Widget.TouchInputDisplay")) -> (ReturnValue:Gameplay Tag Structure)
   ) -> (then)
   n2 = Set Event Router(execute=@n1.OnEventRouterAvailable, Event Router:Gameplay Event Router Component Object Reference=@n1) -> (Output_Get:Gameplay Event Router Component Object Reference)
   > n3 = StartListeningToEvent!(
    execute:Exec
    Target:Gameplay Event Router Component Object Reference=@n2
    EventType:Script Struct Object Reference=/Script/PilgrimSettingsRuntime.PilgrimEvent_PilgrimSettings_UseAlternative5LaneTouchLayoutChanged
    Context:Object Reference
    EventBubblingRule:EEventBubblingRule Enum
   ) -> (
    then:Exec
    AsyncTaskProxy:Async Action Start Listening To Event Object Reference
    OnEventReceived:Exec
    Payload:Pilgrim Event Pilgrim Settings Use Alternative 5Lane Touch Layout Changed Structure
   )
   > Add(
    TargetArray:Array of Async Action Start Listening To Event Object References=Get GameplayEventHandles() -> (GameplayEventHandles:Array of Async Action Start Listening To Event Object References)
    NewItem:"Async Action Start Listening To Event Object Reference (by ref)"=@n3
   ) {is=title K2Node_CallArrayFunction}
   > n4 = StartListeningToEvent(
    Target=Get Event Router() -> (Event Router:Gameplay Event Router Component Object Reference)
    EventType=/Script/PilgrimSettingsRuntime.PilgrimEvent_PilgrimSettings_UseGuitarLeftyFlipChanged
   ) -> (Payload:Pilgrim Event Pilgrim Settings Use Guitar Lefty Flip Changed Structure)
   > Add(TargetArray=Get GameplayEventHandles(), NewItem=@n4)
   UpdateOnScreenButtonsVisibility(execute=[@n4.OnEventReceived, @n3.OnEventReceived])
   Event OnMusicBattleInstanceInitializeEvent() {is=title K2Node_Event}
   n5 = GetLocalPilgrimGameForPlayer!(
    self:Pilgrim Music Battle Instance Object Reference=Get MusicBattleInstance() -> (MusicBattleInstance:Pilgrim Music Battle Instance Object Reference)
    InPC:Player Controller Object Reference=GetOwningPlayer()
   ) -> (ReturnValue:Pilgrim Game Object Reference)
   > Set IsUsing5Lanes(IsUsing5Lanes:bool="Equal (Integer)"(A=GetNumLanes!(self:Pilgrim Game Object Reference=@n5) -> (ReturnValue:int=0), B=5)) -> (Output_Get:bool)
   > SetVisibility(self=Get MobileTouchPanel() -> (MobileTouchPanel:Canvas Panel Object Reference), InVisibility=SelfHitTestInvisible)
   > UpdateOnScreenButtonsVisibility()
   n6 = StartListeningToEvent(Target=Get Event Router(), EventType=/Script/PilgrimSettingsRuntime.PilgrimEvent_PilgrimSettings_EnableMobileExpandedTrackChanged) -> (Payload:Pilgrim Event Pilgrim Settings Enable Mobile Expanded Track Changed Structure)
   > Add(TargetArray=Get GameplayEventHandles(), NewItem=@n6)
   Setup for Expanded Track(execute=@n6.OnEventReceived)
  ]}
 GetIsUsingLeftyFlip {nodes [
   GetIsUsingLeftyFlip() {is=title K2Node_FunctionEntry}
   n1 = GetComponentByClass!(
    self:Actor Object Reference=GetOwningPlayer()
    ComponentClass:Actor Component Class Reference=/PilgrimCore/Player/Components/BP_PilgrimPlayerControllerComponent.BP_PilgrimPlayerControllerComponent_C
   ) -> (ReturnValue:BP Pilgrim Player Controller Component Object Reference)
   > Is Valid(InputObject=@n1)
   > Return Node!(
    execute
    IsUsingLeftyFlip:bool=GetIsGuitarLeftyFlip!(self:Pilgrim Client Setting Record Partition Object Reference=GetPilgrimClientSettingRecordPartition!(self:Pilgrim Player Controller Component Object Reference=@n1) -> (ReturnValue:Pilgrim Client Setting Record Partition Object Reference)) -> (ReturnValue:bool)
   )
  ]}
 GetUseAlternate5LaneTouchLayout {nodes [
   GetUseAlternate5LaneTouchLayout() {is=title K2Node_FunctionEntry}
   n1 = GetComponentByClass(
    self=GetOwningPlayer()
    ComponentClass=&BP_PilgrimPlayerControllerComponent_C
   )
   > Return Node!(
    execute
    Use Alternate 5 Lane Layout:bool=GetUseAlternative5LaneTouchLayout!(self:Pilgrim Client Setting Record Partition Object Reference=GetPilgrimClientSettingRecordPartition(self=@n1)) -> (ReturnValue:bool)
   )
  ]}
 Setup for Expanded Track {nodes [
   SetVisibility(InVisibility=Select(Option 0=SelfHitTestInvisible, Option 1=Collapsed, Index=IsMobileAndShouldUseExpandedTrack!() -> (ReturnValue:bool)))
   > SetPosition!(
    execute
    self:Canvas Panel Slot Object Reference=SlotAsCanvasSlot!(Widget:Widget Object Reference=Get PilgrimTouchButton_Medium_Overdrive_5lane() -> (PilgrimTouchButton_Medium_Overdrive_5lane:Pilgrim Touch Button Medium Object Reference)) -> (ReturnValue:Canvas Panel Slot Object Reference)
    InPosition:Vector 2D Structure=Select(Option 0:Vector 2D Structure=(X=-223,Y=-223), Option 1:Vector 2D Structure=(X=-220,Y=-290), Index=IsMobileAndShouldUseExpandedTrack()) -> (ReturnValue:Vector 2D Structure)
   ) -> (then)
   Setup for Expanded Track() {is=title K2Node_FunctionEntry}
  ]}
 UpdateOnScreenButtonsVisibility {nodes [
   UpdateOnScreenButtonsVisibility() {is=title K2Node_FunctionEntry}
   > n1 = Sequence()
   > Set MustApplyLeftyFlip(MustApplyLeftyFlip:bool=GetIsUsingLeftyFlip!() -> (IsUsingLeftyFlip:bool) {is=title K2Node_CallFunction}) -> (Output_Get:bool)
   > Set MustUseAlternate5LaneTouchLayoutOnLeftHand(MustUseAlternate5LaneTouchLayoutOnLeftHand:bool=AND Boolean(
     A=Get IsUsing5Lanes() -> (IsUsing5Lanes:bool)
     B=GetUseAlternate5LaneTouchLayout!(self) -> (Use Alternate 5 Lane Layout:bool) {is=title K2Node_CallFunction}
    )) -> (Output_Get:bool)
   n2 = Sequence(execute=@n1.then_1)
   n3 = Get MustApplyLeftyFlip() -> (MustApplyLeftyFlip:bool)
   n4 = Get LeftPanel5LaneTouch() -> (LeftPanel5LaneTouch:Overlay Object Reference)
   n5 = Get LeftPanel5LaneTouch_LeftyFlip() -> (LeftPanel5LaneTouch_LeftyFlip:Overlay Object Reference)
   n6 = Get LeftPanel4LaneTouch() -> (LeftPanel4LaneTouch:Overlay Object Reference)
   n7 = Get MustApplyLeftyFlip()
   n8 = Get LeftPanel4LaneTouch_LeftyFlip() -> (LeftPanel4LaneTouch_LeftyFlip:Overlay Object Reference)
   > SetActiveWidget(
    self=Get LeftPanelCommonVisibilitySwitcher() -> (LeftPanelCommonVisibilitySwitcher:Common Visibility Switcher Object Reference)
    Widget=Select(
     Option 0:Widget Object Reference=Select(Option 0=@n6, Option 1=@n8, Index=@n7)
     Option 1:Widget Object Reference=Select(Option 0=@n4, Option 1=@n5, Index=@n3)
     Index=Get MustUseAlternate5LaneTouchLayoutOnLeftHand() -> (MustUseAlternate5LaneTouchLayoutOnLeftHand:bool)
    ) -> (ReturnValue:Widget Object Reference)
   )
   n9 = Get RightPanel4LaneTouch() -> (RightPanel4LaneTouch:Overlay Object Reference)
   n10 = Get RightPanel4LaneTouch_LeftyFlip() -> (RightPanel4LaneTouch_LeftyFlip:Overlay Object Reference)
   n11 = Get MustApplyLeftyFlip()
   n12 = Get RightPanel5LaneTouch() -> (RightPanel5LaneTouch:Overlay Object Reference)
   n13 = Get RightPanel5LaneTouch_LeftyFlip() -> (RightPanel5LaneTouch_LeftyFlip:Overlay Object Reference)
   n14 = Get MustApplyLeftyFlip()
   SetActiveWidget(
    execute=@n2.then_1
    self=Get RightPanelCommonVisibilitySwitcher() -> (RightPanelCommonVisibilitySwitcher:Common Visibility Switcher Object Reference)
    Widget=Select(
     Option 0=Select(Option 0=@n9, Option 1=@n10, Index=@n11)
     Option 1=Select(Option 0=@n12, Option 1=@n13, Index=@n14)
     Index=Get IsUsing5Lanes()
    )
   )
  ]}
]
variables [
 Event Router {type=GameplayEventRouterComponent, default=null}
 IsUsing5Lanes {type=boolean}
 IsUsing5LanesAlternativeLayout {type=boolean}
 VisDebug_Is5Lanes {type=boolean}
 VisDebug_Is5LanesAlternativeLayout {type=boolean}
]
widgets {
 info {inheritedWidgetCount=0, parentClass=/Script/PilgrimCoreUI.PilgrimGameWidgetBase, rootWidgetClass=&CanvasPanel, widgetCount=43}
 widgets [
  MobileTouchPanel {widgetClassPath=&CanvasPanel, bIsVariable=true, widget=MobileTouchPanel, properties {visibility=Hidden}}
  LeftPanelCommonVisibilitySwitcher {
   widgetClassPath=&CommonVisibilitySwitcher
   bIsVariable=true
   parent=MobileTouchPanel
   widget=LeftPanelCommonVisibilitySwitcher
   slot {
    class=CanvasPanelSlot
    properties {bAutoSize=true, layoutData {alignment=<0 1>, anchors {maximum=<0 1>, minimum=<0 1>}, offsets=<ltrb 0 0 0 0>}}
   }
  }
  LeftPanel4LaneTouch {
   widgetClassPath=&Overlay
   bIsVariable=true
   parent=LeftPanelCommonVisibilitySwitcher
   widget=LeftPanel4LaneTouch
   properties {bOverride_Cursor=true}
   slot {class=CommonVisibilitySwitcherSlot}
  }
  PilgrimTouchButton_XL_lane1 {
   widgetClassPath=&/PilgrimTouchButton_XL
   bIsVariable=true
   parent=LeftPanel4LaneTouch
   widget=PilgrimTouchButton_XL_lane1
   properties {
    deactivateInputActionIfDragOutsideHitbox=true
    pressAction=/PilgrimSettings/Input/InputActions/IA_PilgrimLane1Press
    visibility=HitTestInvisible
    replace Slate Brush {
     bIsDynamicallyLoaded=false
     drawAs=Image
     imageType=NoImage
     mirroring=NoMirror
     resourceName=None
     resourceObject=/PilgrimCore/Widgets/Mobile/Textures/T_Mobile_Lane1_Sprite
     tiling=NoTile
     imageSize=<64 64>
     margin=<ltrb 0 0 0 0>
     outlineSettings {
      bUseBrushTransparency=false
      roundingType=HalfHeightRadius
      width=0
      color {colorUseRule=UseColor_Specified, specifiedColor=<rgba 0 0 0 0>}
      cornerRadii=<0 0 0 1>
     }
     tintColor {colorUseRule=UseColor_Specified, specifiedColor=<rgba 1 1 1 1>}
     uVRegion {bIsValid=false, max=<0 0>, min=<0 0>}
    }
    screenSpaceButtonBoundaries [<0.001 1>, <0.001 0.25>, <0.28 0.495>]
   }
   slot {class=OverlaySlot, properties {verticalAlignment=VAlign_Bottom, padding=<ltrb 16 0 0 239>}}
  }
  PilgrimTouchButton_XL_lane2 {
   widgetClassPath=&PilgrimTouchButton_XL
   bIsVariable=true
   parent=LeftPanel4LaneTouch
   widget=PilgrimTouchButton_XL_lane2
   properties {
    deactivateInputActionIfDragOutsideHitbox=true
    pressAction=/PilgrimSettings/Input/InputActions/IA_PilgrimLane2Press
    visibility=HitTestInvisible
    replace Slate Brush {
     bIsDynamicallyLoaded=false
     drawAs=Image
     imageType=NoImage
     mirroring=NoMirror
     resourceName=None
     resourceObject=/PilgrimCore/Widgets/Mobile/Textures/T_Mobile_Lane2_Sprite
     tiling=NoTile
     imageSize=<64 64>
     margin=<ltrb 0 0 0 0>
     outlineSettings {
      bUseBrushTransparency=false
      roundingType=HalfHeightRadius
      width=0
      color {colorUseRule=UseColor_Specified, specifiedColor=<rgba 0 0 0 0>}
      cornerRadii=<0 0 0 1>
     }
     tintColor {colorUseRule=UseColor_Specified, specifiedColor=<rgba 1 1 1 1>}
     uVRegion {bIsValid=false, max=<0 0>, min=<0 0>}
    }
    screenSpaceButtonBoundaries [<0 1>, <0.28 0.495>, <0.35 1>]
   }
   slot {class=OverlaySlot, properties {verticalAlignment=VAlign_Bottom, padding=<ltrb 239 0 0 16>}}
  }
  WBP_PolygonalInputZoneController_Left4Lane {
   widgetClassPath=/PolygonalInputZones/Assets/WBP_PolygonalInputZoneController
   bIsVariable=true
   parent=LeftPanel4LaneTouch
   widget=WBP_PolygonalInputZoneController_Left4Lane
   properties {inputZones [
     {
      activated=true
      deactivateOnRelease=false
      priority=0
      zoneAction=null
      zoneId=""
      zoneWidget=&self:WidgetTree.PilgrimTouchButton_XL_lane1
      zoneVertices [<0 1>, <0 0>, <0.85 0.26>]
     }
     {
      activated=true
      deactivateOnRelease=false
      priority=0
      zoneAction=null
      zoneId=""
      zoneWidget=&self:WidgetTree.PilgrimTouchButton_XL_lane2
      zoneVertices [<0 1>, <0.85 0.26>, <0.85 1>]
     }
    ]}
   slot {class=OverlaySlot, properties {horizontalAlignment=HAlign_Fill, verticalAlignment=VAlign_Fill, padding=<ltrb 0 -200 -100 0>}}
  }
  LeftPanel5LaneTouch {
   widgetClassPath=&Overlay
   bIsVariable=true
   parent=LeftPanelCommonVisibilitySwitcher
   widget=LeftPanel5LaneTouch
   properties {bOverride_Cursor=true}
   slot {class=CommonVisibilitySwitcherSlot}
  }
  PilgrimTouchButton_L_lane1 {
   widgetClassPath=&/PilgrimTouchButton_L
   bIsVariable=true
   parent=LeftPanel5LaneTouch
   widget=PilgrimTouchButton_L_lane1
   properties {
    deactivateInputActionIfDragOutsideHitbox=true
    pressAction=/PilgrimSettings/Input/InputActions/IA_PilgrimLane1Press_5Lane
    visibility=HitTestInvisible
    replace Slate Brush {
     bIsDynamicallyLoaded=false
     drawAs=Image
     imageType=NoImage
     mirroring=NoMirror
     resourceName=None
     resourceObject=&T_Mobile_Lane1_Sprite
     tiling=NoTile
     imageSize=<64 64>
     margin=<ltrb 0 0 0 0>
     outlineSettings {
      bUseBrushTransparency=false
      roundingType=HalfHeightRadius
      width=0
      color {colorUseRule=UseColor_Specified, specifiedColor=<rgba 0 0 0 0>}
      cornerRadii=<0 0 0 1>
     }
     tintColor {colorUseRule=UseColor_Specified, specifiedColor=<rgba 1 1 1 1>}
     uVRegion {bIsValid=false, max=<0 0>, min=<0 0>}
    }
    screenSpaceButtonBoundaries [<0 1>, <0 0.425>, <0.17 0.525>]
   }
   slot {class=OverlaySlot, properties {verticalAlignment=VAlign_Bottom, padding=<ltrb 21 0 0 310>}}
  }
  PilgrimTouchButton_L_lane2 {
   widgetClassPath=&PilgrimTouchButton_L
   bIsVariable=true
   parent=LeftPanel5LaneTouch
   widget=PilgrimTouchButton_L_lane2
   properties {
    deactivateInputActionIfDragOutsideHitbox=true
    pressAction=/PilgrimSettings/Input/InputActions/IA_PilgrimLane2Press_5Lane
    visibility=HitTestInvisible
    replace Slate Brush {
     bIsDynamicallyLoaded=false
     drawAs=Image
     imageType=NoImage
     mirroring=NoMirror
     resourceName=None
     resourceObject=&T_Mobile_Lane2_Sprite
     tiling=NoTile
     imageSize=<64 64>
     margin=<ltrb 0 0 0 0>
     outlineSettings {
      bUseBrushTransparency=false
      roundingType=HalfHeightRadius
      width=0
      color {colorUseRule=UseColor_Specified, specifiedColor=<rgba 0 0 0 0>}
      cornerRadii=<0 0 0 1>
     }
     tintColor {colorUseRule=UseColor_Specified, specifiedColor=<rgba 1 1 1 1>}
     uVRegion {bIsValid=false, max=<0 0>, min=<0 0>}
    }
    screenSpaceButtonBoundaries [<0 1>, <0.17 0.525>, <0.26 0.7>]
   }
   slot {class=OverlaySlot, properties {verticalAlignment=VAlign_Bottom, padding=<ltrb 219 0 0 219>}}
  }
  PilgrimTouchButton_L_lane3_left {
   widgetClassPath=&PilgrimTouchButton_L
   bIsVariable=true
   parent=LeftPanel5LaneTouch
   widget=PilgrimTouchButton_L_lane3_left
   properties {
    deactivateInputActionIfDragOutsideHitbox=true
    pressAction=/PilgrimSettings/Input/InputActions/IA_PilgrimLane3Press_5Lane
    visibility=HitTestInvisible
    replace Slate Brush {
     bIsDynamicallyLoaded=false
     drawAs=Image
     imageType=NoImage
     mirroring=NoMirror
     resourceName=None
     resourceObject=/PilgrimCore/Widgets/Mobile/Textures/T_Mobile_Lane3_Sprite
     tiling=NoTile
     imageSize=<73 64>
     margin=<ltrb 0 0 0 0>
     outlineSettings {
      bUseBrushTransparency=false
      roundingType=HalfHeightRadius
      width=0
      color {colorUseRule=UseColor_Specified, specifiedColor=<rgba 0 0 0 0>}
      cornerRadii=<0 0 0 1>
     }
     tintColor {colorUseRule=UseColor_Specified, specifiedColor=<rgba 1 1 1 1>}
     uVRegion {bIsValid=false, max=<0 0>, min=<0 0>}
    }
    screenSpaceButtonBoundaries [<0 1>, <0.26 0.7>, <0.29 1>]
   }
   slot {class=OverlaySlot, properties {verticalAlignment=VAlign_Bottom, padding=<ltrb 254 0 0 21>}}
  }
  WBP_PolygonalInputZoneController_Left5Lane {
   widgetClassPath=&WBP_PolygonalInputZoneController
   bIsVariable=true
   parent=LeftPanel5LaneTouch
   widget=WBP_PolygonalInputZoneController_Left5Lane
   properties {inputZones [
     {
      activated=true
      deactivateOnRelease=false
      priority=0
      zoneAction=null
      zoneId=""
      zoneWidget=&self:WidgetTree.PilgrimTouchButton_L_lane1
      zoneVertices [<0 1>, <0 0>, <0.5 0.2>]
     }
     {
      activated=true
      deactivateOnRelease=false
      priority=0
      zoneAction=null
      zoneId=""
      zoneWidget=&self:WidgetTree.PilgrimTouchButton_L_lane2
      zoneVertices [<0 1>, <0.5 0.2>, <0.85 0.52>]
     }
     {
      activated=true
      deactivateOnRelease=false
      priority=0
      zoneAction=null
      zoneId=""
      zoneWidget=&self:WidgetTree.PilgrimTouchButton_L_lane3_left
      zoneVertices [<0 1>, <0.85 0.52>, <0.85 1>]
     }
    ]}
   slot {class=OverlaySlot, properties {horizontalAlignment=HAlign_Fill, verticalAlignment=VAlign_Fill, padding=<ltrb 0 -220 -175 0>}}
  }
  LeftPanel4LaneTouch_LeftyFlip {
   widgetClassPath=&Overlay
   bIsVariable=true
   parent=LeftPanelCommonVisibilitySwitcher
   widget=LeftPanel4LaneTouch_LeftyFlip
   properties {bOverride_Cursor=true}
   slot {class=CommonVisibilitySwitcherSlot}
  }
  PilgrimTouchButton_XL_lane4_leftyflip {
   widgetClassPath=&PilgrimTouchButton_XL
   bIsVariable=true
   parent=LeftPanel4LaneTouch_LeftyFlip
   widget=PilgrimTouchButton_XL_lane4_leftyflip
   properties {
    deactivateInputActionIfDragOutsideHitbox=true
    pressAction=/PilgrimSettings/Input/InputActions/IA_PilgrimLane4Press
    visibility=HitTestInvisible
    replace Slate Brush {
     bIsDynamicallyLoaded=false
     drawAs=Image
     imageType=NoImage
     mirroring=NoMirror
     resourceName=None
     resourceObject=&T_Mobile_Lane1_Sprite
     tiling=NoTile
     imageSize=<64 64>
     margin=<ltrb 0 0 0 0>
     outlineSettings {
      bUseBrushTransparency=false
      roundingType=HalfHeightRadius
      width=0
      color {colorUseRule=UseColor_Specified, specifiedColor=<rgba 0 0 0 0>}
      cornerRadii=<0 0 0 1>
     }
     tintColor {colorUseRule=UseColor_Specified, specifiedColor=<rgba 1 1 1 1>}
     uVRegion {bIsValid=false, max=<0 0>, min=<0 0>}
    }
    screenSpaceButtonBoundaries [<0.001 1>, <0.001 0.25>, <0.28 0.495>]
   }
   slot {class=OverlaySlot, properties {verticalAlignment=VAlign_Bottom, padding=<ltrb 16 0 0 239>}}
  }
  PilgrimTouchButton_XL_lane3_leftyflip {
   widgetClassPath=&PilgrimTouchButton_XL
   bIsVariable=true
   parent=LeftPanel4LaneTouch_LeftyFlip
   widget=PilgrimTouchButton_XL_lane3_leftyflip
   properties {
    deactivateInputActionIfDragOutsideHitbox=true
    pressAction=/PilgrimSettings/Input/InputActions/IA_PilgrimLane3Press
    visibility=HitTestInvisible
    replace Slate Brush {
     bIsDynamicallyLoaded=false
     drawAs=Image
     imageType=NoImage
     mirroring=NoMirror
     resourceName=None
     resourceObject=&T_Mobile_Lane2_Sprite
     tiling=NoTile
     imageSize=<64 64>
     margin=<ltrb 0 0 0 0>
     outlineSettings {
      bUseBrushTransparency=false
      roundingType=HalfHeightRadius
      width=0
      color {colorUseRule=UseColor_Specified, specifiedColor=<rgba 0 0 0 0>}
      cornerRadii=<0 0 0 1>
     }
     tintColor {colorUseRule=UseColor_Specified, specifiedColor=<rgba 1 1 1 1>}
     uVRegion {bIsValid=false, max=<0 0>, min=<0 0>}
    }
    screenSpaceButtonBoundaries [<0 1>, <0.28 0.495>, <0.35 1>]
   }
   slot {class=OverlaySlot, properties {verticalAlignment=VAlign_Bottom, padding=<ltrb 239 0 0 16>}}
  }
  WBP_PolygonalInputZoneController_Left4Lane_1 {
   widgetClassPath=&WBP_PolygonalInputZoneController
   bIsVariable=true
   parent=LeftPanel4LaneTouch_LeftyFlip
   widget=WBP_PolygonalInputZoneController_Left4Lane_1
   properties {inputZones [
     {
      activated=true
      deactivateOnRelease=false
      priority=0
      zoneAction=null
      zoneId=""
      zoneWidget=&self:WidgetTree.PilgrimTouchButton_XL_lane4_leftyflip
      zoneVertices [<0 1>, <0 0>, <0.85 0.26>]
     }
     {
      activated=true
      deactivateOnRelease=false
      priority=0
      zoneAction=null
      zoneId=""
      zoneWidget=&self:WidgetTree.PilgrimTouchButton_XL_lane3_leftyflip
      zoneVertices [<0 1>, <0.85 0.26>, <0.85 1>]
     }
    ]}
   slot {class=OverlaySlot, properties {horizontalAlignment=HAlign_Fill, verticalAlignment=VAlign_Fill, padding=<ltrb 0 -200 -100 0>}}
  }
  LeftPanel5LaneTouch_LeftyFlip {
   widgetClassPath=&Overlay
   bIsVariable=true
   parent=LeftPanelCommonVisibilitySwitcher
   widget=LeftPanel5LaneTouch_LeftyFlip
   properties {bOverride_Cursor=true}
   slot {class=CommonVisibilitySwitcherSlot}
  }
  PilgrimTouchButton_L_lane5_leftyflip {
   widgetClassPath=&PilgrimTouchButton_L
   bIsVariable=true
   parent=LeftPanel5LaneTouch_LeftyFlip
   widget=PilgrimTouchButton_L_lane5_leftyflip
   properties {
    deactivateInputActionIfDragOutsideHitbox=true
    pressAction=/PilgrimSettings/Input/InputActions/IA_PilgrimLane5Press_5Lane
    visibility=HitTestInvisible
    replace Slate Brush {
     bIsDynamicallyLoaded=false
     drawAs=Image
     imageType=NoImage
     mirroring=NoMirror
     resourceName=None
     resourceObject=&T_Mobile_Lane1_Sprite
     tiling=NoTile
     imageSize=<64 64>
     margin=<ltrb 0 0 0 0>
     outlineSettings {
      bUseBrushTransparency=false
      roundingType=HalfHeightRadius
      width=0
      color {colorUseRule=UseColor_Specified, specifiedColor=<rgba 0 0 0 0>}
      cornerRadii=<0 0 0 1>
     }
     tintColor {colorUseRule=UseColor_Specified, specifiedColor=<rgba 1 1 1 1>}
     uVRegion {bIsValid=false, max=<0 0>, min=<0 0>}
    }
    screenSpaceButtonBoundaries [<0 1>, <0 0.425>, <0.17 0.525>]
   }
   slot {class=OverlaySlot, properties {verticalAlignment=VAlign_Bottom, padding=<ltrb 21 0 0 310>}}
  }
  PilgrimTouchButton_L_lane4_leftyflip {
   widgetClassPath=&PilgrimTouchButton_L
   bIsVariable=true
   parent=LeftPanel5LaneTouch_LeftyFlip
   widget=PilgrimTouchButton_L_lane4_leftyflip
   properties {
    deactivateInputActionIfDragOutsideHitbox=true
    pressAction=/PilgrimSettings/Input/InputActions/IA_PilgrimLane4Press_5Lane
    visibility=HitTestInvisible
    replace Slate Brush {
     bIsDynamicallyLoaded=false
     drawAs=Image
     imageType=NoImage
     mirroring=NoMirror
     resourceName=None
     resourceObject=&T_Mobile_Lane2_Sprite
     tiling=NoTile
     imageSize=<64 64>
     margin=<ltrb 0 0 0 0>
     outlineSettings {
      bUseBrushTransparency=false
      roundingType=HalfHeightRadius
      width=0
      color {colorUseRule=UseColor_Specified, specifiedColor=<rgba 0 0 0 0>}
      cornerRadii=<0 0 0 1>
     }
     tintColor {colorUseRule=UseColor_Specified, specifiedColor=<rgba 1 1 1 1>}
     uVRegion {bIsValid=false, max=<0 0>, min=<0 0>}
    }
    screenSpaceButtonBoundaries [<0 1>, <0.17 0.525>, <0.26 0.7>]
   }
   slot {class=OverlaySlot, properties {verticalAlignment=VAlign_Bottom, padding=<ltrb 219 0 0 219>}}
  }
  PilgrimTouchButton_L_lane3_leftyflip {
   widgetClassPath=&PilgrimTouchButton_L
   bIsVariable=true
   parent=LeftPanel5LaneTouch_LeftyFlip
   widget=PilgrimTouchButton_L_lane3_leftyflip
   properties {
    deactivateInputActionIfDragOutsideHitbox=true
    pressAction=&IA_PilgrimLane3Press_5Lane
    visibility=HitTestInvisible
    replace Slate Brush {
     bIsDynamicallyLoaded=false
     drawAs=Image
     imageType=NoImage
     mirroring=NoMirror
     resourceName=None
     resourceObject=&T_Mobile_Lane3_Sprite
     tiling=NoTile
     imageSize=<73 64>
     margin=<ltrb 0 0 0 0>
     outlineSettings {
      bUseBrushTransparency=false
      roundingType=HalfHeightRadius
      width=0
      color {colorUseRule=UseColor_Specified, specifiedColor=<rgba 0 0 0 0>}
      cornerRadii=<0 0 0 1>
     }
     tintColor {colorUseRule=UseColor_Specified, specifiedColor=<rgba 1 1 1 1>}
     uVRegion {bIsValid=false, max=<0 0>, min=<0 0>}
    }
    screenSpaceButtonBoundaries [<0 1>, <0.26 0.7>, <0.29 1>]
   }
   slot {class=OverlaySlot, properties {verticalAlignment=VAlign_Bottom, padding=<ltrb 254 0 0 21>}}
  }
  WBP_PolygonalInputZoneController_Left5Lane_1 {
   widgetClassPath=&WBP_PolygonalInputZoneController
   bIsVariable=true
   parent=LeftPanel5LaneTouch_LeftyFlip
   widget=WBP_PolygonalInputZoneController_Left5Lane_1
   properties {inputZones [
     {
      activated=true
      deactivateOnRelease=false
      priority=0
      zoneAction=null
      zoneId=""
      zoneWidget=&self:WidgetTree.PilgrimTouchButton_L_lane5_leftyflip
      zoneVertices [<0 1>, <0 0>, <0.5 0.2>]
     }
     {
      activated=true
      deactivateOnRelease=false
      priority=0
      zoneAction=null
      zoneId=""
      zoneWidget=&self:WidgetTree.PilgrimTouchButton_L_lane4_leftyflip
      zoneVertices [<0 1>, <0.5 0.2>, <0.85 0.52>]
     }
     {
      activated=true
      deactivateOnRelease=false
      priority=0
      zoneAction=null
      zoneId=""
      zoneWidget=&self:WidgetTree.PilgrimTouchButton_L_lane3_leftyflip
      zoneVertices [<0 1>, <0.85 0.52>, <0.85 1>]
     }
    ]}
   slot {class=OverlaySlot, properties {horizontalAlignment=HAlign_Fill, verticalAlignment=VAlign_Fill, padding=<ltrb 0 -220 -175 0>}}
  }
  RightPanelCommonVisibilitySwitcher {
   widgetClassPath=&CommonVisibilitySwitcher
   bIsVariable=true
   parent=MobileTouchPanel
   widget=RightPanelCommonVisibilitySwitcher
   slot {
    class=CanvasPanelSlot
    properties {bAutoSize=true, layoutData {alignment=<1 1>, anchors {maximum=<1 1>, minimum=<1 1>}, offsets=<ltrb 0 0 0 0>}}
   }
  }
  RightPanel4LaneTouch {
   widgetClassPath=&Overlay
   bIsVariable=true
   parent=RightPanelCommonVisibilitySwitcher
   widget=RightPanel4LaneTouch
   properties {bOverride_Cursor=true}
   slot {class=CommonVisibilitySwitcherSlot}
  }
  PilgrimTouchButton_Medium_Overdrive_4lane {
   widgetClassPath=&/PilgrimTouchButton_Medium
   bIsVariable=true
   parent=RightPanel4LaneTouch
   widget=PilgrimTouchButton_Medium_Overdrive_4lane
   properties {
    deactivateInputActionIfDragOutsideHitbox=true
    pressAction=/PilgrimSettings/Input/InputActions/IA_PilgrimDeployOverdrive
    visibility=HitTestInvisible
    replace Slate Brush {
     bIsDynamicallyLoaded=false
     drawAs=Image
     imageType=NoImage
     mirroring=NoMirror
     resourceName=None
     resourceObject=/PilgrimCore/Widgets/Mobile/Textures/T_Icon_Overdrive_Sprite
     tiling=NoTile
     imageSize=<64 64>
     margin=<ltrb 0 0 0 0>
     outlineSettings {
      bUseBrushTransparency=false
      roundingType=HalfHeightRadius
      width=0
      color {colorUseRule=UseColor_Specified, specifiedColor=<rgba 0 0 0 0>}
      cornerRadii=<0 0 0 1>
     }
     tintColor {colorUseRule=UseColor_Specified, specifiedColor=<rgba 1 1 1 1>}
     uVRegion {bIsValid=false, max=<0 0>, min=<0 0>}
    }
    screenSpaceButtonBoundaries [<0.8575 1>, <0.883 0.79>, <1 0.75>, <1 1>]
   }
   slot {class=OverlaySlot, properties {horizontalAlignment=HAlign_Right, verticalAlignment=VAlign_Bottom, padding=<ltrb 0 0 32 32>}}
  }
  PilgrimTouchButton_XL_lane4_1 {
   widgetClassPath=&PilgrimTouchButton_XL
   bIsVariable=true
   parent=RightPanel4LaneTouch
   widget=PilgrimTouchButton_XL_lane4_1
   properties {
    deactivateInputActionIfDragOutsideHitbox=true
    pressAction=&IA_PilgrimLane4Press
    visibility=HitTestInvisible
    replace Slate Brush {
     bIsDynamicallyLoaded=false
     drawAs=Image
     imageType=NoImage
     mirroring=NoMirror
     resourceName=None
     resourceObject=/PilgrimCore/Widgets/Mobile/Textures/T_Mobile_Lane4
     tiling=NoTile
     imageSize=<64 64>
     margin=<ltrb 0 0 0 0>
     outlineSettings {
      bUseBrushTransparency=false
      roundingType=HalfHeightRadius
      width=0
      color {colorUseRule=UseColor_Specified, specifiedColor=<rgba 0 0 0 0>}
      cornerRadii=<0 0 0 1>
     }
     tintColor {colorUseRule=UseColor_Specified, specifiedColor=<rgba 1 1 1 1>}
     uVRegion {bIsValid=false, max=<0 0>, min=<0 0>}
    }
    screenSpaceButtonBoundaries [<0.7 0.465>, <1 0.25>, <1 1>]
   }
   slot {class=OverlaySlot, properties {horizontalAlignment=HAlign_Right, verticalAlignment=VAlign_Bottom, padding=<ltrb 0 0 16 239>}}
  }
  PilgrimTouchButton_XL_lane3_1 {
   widgetClassPath=&PilgrimTouchButton_XL
   bIsVariable=true
   parent=RightPanel4LaneTouch
   widget=PilgrimTouchButton_XL_lane3_1
   properties {
    deactivateInputActionIfDragOutsideHitbox=true
    pressAction=&IA_PilgrimLane3Press
    visibility=HitTestInvisible
    replace Slate Brush {
     bIsDynamicallyLoaded=false
     drawAs=Image
     imageType=NoImage
     mirroring=NoMirror
     resourceName=None
     resourceObject=&T_Mobile_Lane3_Sprite
     tiling=NoTile
     imageSize=<73 64>
     margin=<ltrb 0 0 0 0>
     outlineSettings {
      bUseBrushTransparency=false
      roundingType=HalfHeightRadius
      width=0
      color {colorUseRule=UseColor_Specified, specifiedColor=<rgba 0 0 0 0>}
      cornerRadii=<0 0 0 1>
     }
     tintColor {colorUseRule=UseColor_Specified, specifiedColor=<rgba 1 1 1 1>}
     uVRegion {bIsValid=false, max=<0 0>, min=<0 0>}
    }
    screenSpaceButtonBoundaries [<0.625 1>, <0.7 0.465>, <1 1>]
   }
   slot {class=OverlaySlot, properties {horizontalAlignment=HAlign_Right, verticalAlignment=VAlign_Bottom, padding=<ltrb 0 0 239 16>}}
  }
  WBP_PolygonalInputZoneController_Right4Lane {
   widgetClassPath=&WBP_PolygonalInputZoneController
   bIsVariable=true
   parent=RightPanel4LaneTouch
   widget=WBP_PolygonalInputZoneController_Right4Lane
   properties {inputZones [
     {
      activated=true
      deactivateOnRelease=false
      priority=0
      zoneAction=null
      zoneId=""
      zoneWidget=&self:WidgetTree.PilgrimTouchButton_Medium_Overdrive_4lane
      zoneVertices [<0.6 1>, <0.65 0.7>, <1 0.67>, <1 1>]
     }
     {
      activated=true
      deactivateOnRelease=false
      priority=0
      zoneAction=null
      zoneId=""
      zoneWidget=&self:WidgetTree.PilgrimTouchButton_XL_lane4_1
      zoneVertices [<0.15 0.26>, <1 0>, <1 1>]
     }
     {
      activated=true
      deactivateOnRelease=false
      priority=0
      zoneAction=null
      zoneId=""
      zoneWidget=&self:WidgetTree.PilgrimTouchButton_XL_lane3_1
      zoneVertices [<0.15 1>, <0.15 0.26>, <1 1>]
     }
    ]}
   slot {class=OverlaySlot, properties {horizontalAlignment=HAlign_Fill, verticalAlignment=VAlign_Fill, padding=<ltrb -100 -200 0 0>}}
  }
  RightPanel5LaneTouch {
   widgetClassPath=&Overlay
   bIsVariable=true
   parent=RightPanelCommonVisibilitySwitcher
   widget=RightPanel5LaneTouch
   properties {bOverride_Cursor=true}
   slot {class=CommonVisibilitySwitcherSlot}
  }
  PilgrimTouchButton_Medium_Overdrive_5lane {
   widgetClassPath=&PilgrimTouchButton_Medium
   bIsVariable=true
   parent=RightPanel5LaneTouch
   widget=PilgrimTouchButton_Medium_Overdrive_5lane
   properties {
    deactivateInputActionIfDragOutsideHitbox=true
    pressAction=/PilgrimSettings/Input/InputActions/IA_PilgrimDeployOverdrive_5Lanes
    visibility=HitTestInvisible
    replace Slate Brush {
     bIsDynamicallyLoaded=false
     drawAs=Image
     imageType=NoImage
     mirroring=NoMirror
     resourceName=None
     resourceObject=&T_Icon_Overdrive_Sprite
     tiling=NoTile
     imageSize=<64 64>
     margin=<ltrb 0 0 0 0>
     outlineSettings {
      bUseBrushTransparency=false
      roundingType=HalfHeightRadius
      width=0
      color {colorUseRule=UseColor_Specified, specifiedColor=<rgba 0 0 0 0>}
      cornerRadii=<0 0 0 1>
     }
     tintColor {colorUseRule=UseColor_Specified, specifiedColor=<rgba 1 1 1 1>}
     uVRegion {bIsValid=false, max=<0 0>, min=<0 0>}
    }
    screenSpaceButtonBoundaries [<0.8575 1>, <0.883 0.79>, <1 0.75>, <1 1>]
   }
   slot {class=OverlaySlot, properties {horizontalAlignment=HAlign_Right, verticalAlignment=VAlign_Bottom, padding=<ltrb 0 0 32 32>}}
  }
  PilgrimTouchButton_L_lane3_right_1 {
   widgetClassPath=&PilgrimTouchButton_L
   bIsVariable=true
   parent=RightPanel5LaneTouch
   widget=PilgrimTouchButton_L_lane3_right_1
   properties {
    deactivateInputActionIfDragOutsideHitbox=true
    pressAction=&IA_PilgrimLane3Press_5Lane
    visibility=HitTestInvisible
    replace Slate Brush {
     bIsDynamicallyLoaded=false
     drawAs=Image
     imageType=NoImage
     mirroring=NoMirror
     resourceName=None
     resourceObject=&T_Mobile_Lane3_Sprite
     tiling=NoTile
     imageSize=<73 64>
     margin=<ltrb 0 0 0 0>
     outlineSettings {
      bUseBrushTransparency=false
      roundingType=HalfHeightRadius
      width=0
      color {colorUseRule=UseColor_Specified, specifiedColor=<rgba 0 0 0 0>}
      cornerRadii=<0 0 0 1>
     }
     tintColor {colorUseRule=UseColor_Specified, specifiedColor=<rgba 1 1 1 1>}
     uVRegion {bIsValid=false, max=<0 0>, min=<0 0>}
    }
    screenSpaceButtonBoundaries [<1 1>, <0.7119 1>, <0.735 0.713>]
   }
   slot {class=OverlaySlot, properties {horizontalAlignment=HAlign_Right, verticalAlignment=VAlign_Bottom, padding=<ltrb 0 0 254 21>}}
  }
  PilgrimTouchButton_L_lane4_1 {
   widgetClassPath=&PilgrimTouchButton_L
   bIsVariable=true
   parent=RightPanel5LaneTouch
   widget=PilgrimTouchButton_L_lane4_1
   properties {
    deactivateInputActionIfDragOutsideHitbox=true
    pressAction=&IA_PilgrimLane4Press_5Lane
    visibility=HitTestInvisible
    replace Slate Brush {
     bIsDynamicallyLoaded=false
     drawAs=Image
     imageType=NoImage
     mirroring=NoMirror
     resourceName=None
     resourceObject=/PilgrimCore/Widgets/Mobile/Textures/T_Mobile_Lane4_Sprite
     tiling=NoTile
     imageSize=<64 64>
     margin=<ltrb 0 0 0 0>
     outlineSettings {
      bUseBrushTransparency=false
      roundingType=HalfHeightRadius
      width=0
      color {colorUseRule=UseColor_Specified, specifiedColor=<rgba 0 0 0 0>}
      cornerRadii=<0 0 0 1>
     }
     tintColor {colorUseRule=UseColor_Specified, specifiedColor=<rgba 1 1 1 1>}
     uVRegion {bIsValid=false, max=<0 0>, min=<0 0>}
    }
    screenSpaceButtonBoundaries [<1 1>, <0.735 0.713>, <0.834 0.525>]
   }
   slot {class=OverlaySlot, properties {horizontalAlignment=HAlign_Right, verticalAlignment=VAlign_Bottom, padding=<ltrb 0 0 219 219>}}
  }
  PilgrimTouchButton_L_lane5_1 {
   widgetClassPath=&PilgrimTouchButton_L
   bIsVariable=true
   parent=RightPanel5LaneTouch
   widget=PilgrimTouchButton_L_lane5_1
   properties {
    deactivateInputActionIfDragOutsideHitbox=true
    pressAction=&IA_PilgrimLane5Press_5Lane
    visibility=HitTestInvisible
    replace Slate Brush {
     bIsDynamicallyLoaded=false
     drawAs=Image
     imageType=NoImage
     mirroring=NoMirror
     resourceName=None
     resourceObject=/PilgrimCore/Widgets/Mobile/Textures/T_Mobile_Lane5_Sprite
     tiling=NoTile
     imageSize=<68 64>
     margin=<ltrb 0 0 0 0>
     outlineSettings {
      bUseBrushTransparency=false
      roundingType=HalfHeightRadius
      width=0
      color {colorUseRule=UseColor_Specified, specifiedColor=<rgba 0 0 0 0>}
      cornerRadii=<0 0 0 1>
     }
     tintColor {colorUseRule=UseColor_Specified, specifiedColor=<rgba 1 1 1 1>}
     uVRegion {bIsValid=false, max=<0 0>, min=<0 0>}
    }
    screenSpaceButtonBoundaries [<1 1>, <0.834 0.525>, <1 0.4325>]
   }
   slot {class=OverlaySlot, properties {horizontalAlignment=HAlign_Right, verticalAlignment=VAlign_Bottom, padding=<ltrb 0 0 21 310>}}
  }
  WBP_PolygonalInputZoneController_Right5Lane {
   widgetClassPath=&WBP_PolygonalInputZoneController
   bIsVariable=true
   parent=RightPanel5LaneTouch
   widget=WBP_PolygonalInputZoneController_Right5Lane
   properties {inputZones [
     {
      activated=true
      deactivateOnRelease=false
      priority=0
      zoneAction=null
      zoneId=""
      zoneWidget=&self:WidgetTree.PilgrimTouchButton_Medium_Overdrive_5lane
      zoneVertices [<0.6 1>, <0.65 0.7>, <1 0.67>, <1 1>]
     }
     {
      activated=true
      deactivateOnRelease=false
      priority=0
      zoneAction=null
      zoneId=""
      zoneWidget=&self:WidgetTree.PilgrimTouchButton_L_lane3_right_1
      zoneVertices [<1 1>, <0.15 1>, <0.15 0.52>]
     }
     {
      activated=true
      deactivateOnRelease=false
      priority=0
      zoneAction=null
      zoneId=""
      zoneWidget=&self:WidgetTree.PilgrimTouchButton_L_lane4_1
      zoneVertices [<1 1>, <0.15 0.52>, <0.5 0.2>]
     }
     {
      activated=true
      deactivateOnRelease=false
      priority=0
      zoneAction=null
      zoneId=""
      zoneWidget=&self:WidgetTree.PilgrimTouchButton_L_lane5_1
      zoneVertices [<1 1>, <0.5 0.2>, <1 0>]
     }
    ]}
   slot {class=OverlaySlot, properties {horizontalAlignment=HAlign_Fill, verticalAlignment=VAlign_Fill, padding=<ltrb -175 -220 0 0>}}
  }
  RightPanel4LaneTouch_LeftyFlip {
   widgetClassPath=&Overlay
   bIsVariable=true
   parent=RightPanelCommonVisibilitySwitcher
   widget=RightPanel4LaneTouch_LeftyFlip
   properties {bOverride_Cursor=true}
   slot {class=CommonVisibilitySwitcherSlot}
  }
  PilgrimTouchButton_Medium_Overdrive_4lane_1 {
   widgetClassPath=&PilgrimTouchButton_Medium
   bIsVariable=true
   parent=RightPanel4LaneTouch_LeftyFlip
   widget=PilgrimTouchButton_Medium_Overdrive_4lane_1
   properties {
    deactivateInputActionIfDragOutsideHitbox=true
    pressAction=&IA_PilgrimDeployOverdrive
    visibility=HitTestInvisible
    replace Slate Brush {
     bIsDynamicallyLoaded=false
     drawAs=Image
     imageType=NoImage
     mirroring=NoMirror
     resourceName=None
     resourceObject=&T_Icon_Overdrive_Sprite
     tiling=NoTile
     imageSize=<64 64>
     margin=<ltrb 0 0 0 0>
     outlineSettings {
      bUseBrushTransparency=false
      roundingType=HalfHeightRadius
      width=0
      color {colorUseRule=UseColor_Specified, specifiedColor=<rgba 0 0 0 0>}
      cornerRadii=<0 0 0 1>
     }
     tintColor {colorUseRule=UseColor_Specified, specifiedColor=<rgba 1 1 1 1>}
     uVRegion {bIsValid=false, max=<0 0>, min=<0 0>}
    }
    screenSpaceButtonBoundaries [<0.8575 1>, <0.883 0.79>, <1 0.75>, <1 1>]
   }
   slot {class=OverlaySlot, properties {horizontalAlignment=HAlign_Right, verticalAlignment=VAlign_Bottom, padding=<ltrb 0 0 32 32>}}
  }
  PilgrimTouchButton_XL_lane1_leftyflip {
   widgetClassPath=&PilgrimTouchButton_XL
   bIsVariable=true
   parent=RightPanel4LaneTouch_LeftyFlip
   widget=PilgrimTouchButton_XL_lane1_leftyflip
   properties {
    deactivateInputActionIfDragOutsideHitbox=true
    pressAction=&IA_PilgrimLane1Press
    visibility=HitTestInvisible
    replace Slate Brush {
     bIsDynamicallyLoaded=false
     drawAs=Image
     imageType=NoImage
     mirroring=NoMirror
     resourceName=None
     resourceObject=&T_Mobile_Lane4
     tiling=NoTile
     imageSize=<64 64>
     margin=<ltrb 0 0 0 0>
     outlineSettings {
      bUseBrushTransparency=false
      roundingType=HalfHeightRadius
      width=0
      color {colorUseRule=UseColor_Specified, specifiedColor=<rgba 0 0 0 0>}
      cornerRadii=<0 0 0 1>
     }
     tintColor {colorUseRule=UseColor_Specified, specifiedColor=<rgba 1 1 1 1>}
     uVRegion {bIsValid=false, max=<0 0>, min=<0 0>}
    }
    screenSpaceButtonBoundaries [<0.7 0.465>, <1 0.25>, <1 1>]
   }
   slot {class=OverlaySlot, properties {horizontalAlignment=HAlign_Right, verticalAlignment=VAlign_Bottom, padding=<ltrb 0 0 16 239>}}
  }
  PilgrimTouchButton_XL_lane2_leftyflip {
   widgetClassPath=&PilgrimTouchButton_XL
   bIsVariable=true
   parent=RightPanel4LaneTouch_LeftyFlip
   widget=PilgrimTouchButton_XL_lane2_leftyflip
   properties {
    deactivateInputActionIfDragOutsideHitbox=true
    pressAction=&IA_PilgrimLane2Press
    visibility=HitTestInvisible
    replace Slate Brush {
     bIsDynamicallyLoaded=false
     drawAs=Image
     imageType=NoImage
     mirroring=NoMirror
     resourceName=None
     resourceObject=&T_Mobile_Lane3_Sprite
     tiling=NoTile
     imageSize=<73 64>
     margin=<ltrb 0 0 0 0>
     outlineSettings {
      bUseBrushTransparency=false
      roundingType=HalfHeightRadius
      width=0
      color {colorUseRule=UseColor_Specified, specifiedColor=<rgba 0 0 0 0>}
      cornerRadii=<0 0 0 1>
     }
     tintColor {colorUseRule=UseColor_Specified, specifiedColor=<rgba 1 1 1 1>}
     uVRegion {bIsValid=false, max=<0 0>, min=<0 0>}
    }
    screenSpaceButtonBoundaries [<0.625 1>, <0.7 0.465>, <1 1>]
   }
   slot {class=OverlaySlot, properties {horizontalAlignment=HAlign_Right, verticalAlignment=VAlign_Bottom, padding=<ltrb 0 0 239 16>}}
  }
  WBP_PolygonalInputZoneController_Right4Lane_1 {
   widgetClassPath=&WBP_PolygonalInputZoneController
   bIsVariable=true
   parent=RightPanel4LaneTouch_LeftyFlip
   widget=WBP_PolygonalInputZoneController_Right4Lane_1
   properties {inputZones [
     {
      activated=true
      deactivateOnRelease=false
      priority=0
      zoneAction=null
      zoneId=""
      zoneWidget=&self:WidgetTree.PilgrimTouchButton_Medium_Overdrive_4lane_1
      zoneVertices [<0.6 1>, <0.65 0.7>, <1 0.67>, <1 1>]
     }
     {
      activated=true
      deactivateOnRelease=false
      priority=0
      zoneAction=null
      zoneId=""
      zoneWidget=&self:WidgetTree.PilgrimTouchButton_XL_lane1_leftyflip
      zoneVertices [<0.15 0.26>, <1 0>, <1 1>]
     }
     {
      activated=true
      deactivateOnRelease=false
      priority=0
      zoneAction=null
      zoneId=""
      zoneWidget=&self:WidgetTree.PilgrimTouchButton_XL_lane2_leftyflip
      zoneVertices [<0.15 1>, <0.15 0.26>, <1 1>]
     }
    ]}
   slot {class=OverlaySlot, properties {horizontalAlignment=HAlign_Fill, verticalAlignment=VAlign_Fill, padding=<ltrb -100 -200 0 0>}}
  }
  RightPanel5LaneTouch_LeftyFlip {
   widgetClassPath=&Overlay
   bIsVariable=true
   parent=RightPanelCommonVisibilitySwitcher
   widget=RightPanel5LaneTouch_LeftyFlip
   properties {bOverride_Cursor=true}
   slot {class=CommonVisibilitySwitcherSlot}
  }
  PilgrimTouchButton_Medium_Overdrive_5lane_1 {
   widgetClassPath=&PilgrimTouchButton_Medium
   bIsVariable=true
   parent=RightPanel5LaneTouch_LeftyFlip
   widget=PilgrimTouchButton_Medium_Overdrive_5lane_1
   properties {
    deactivateInputActionIfDragOutsideHitbox=true
    pressAction=&IA_PilgrimDeployOverdrive_5Lanes
    visibility=HitTestInvisible
    replace Slate Brush {
     bIsDynamicallyLoaded=false
     drawAs=Image
     imageType=NoImage
     mirroring=NoMirror
     resourceName=None
     resourceObject=&T_Icon_Overdrive_Sprite
     tiling=NoTile
     imageSize=<64 64>
     margin=<ltrb 0 0 0 0>
     outlineSettings {
      bUseBrushTransparency=false
      roundingType=HalfHeightRadius
      width=0
      color {colorUseRule=UseColor_Specified, specifiedColor=<rgba 0 0 0 0>}
      cornerRadii=<0 0 0 1>
     }
     tintColor {colorUseRule=UseColor_Specified, specifiedColor=<rgba 1 1 1 1>}
     uVRegion {bIsValid=false, max=<0 0>, min=<0 0>}
    }
    screenSpaceButtonBoundaries [<0.8575 1>, <0.883 0.79>, <1 0.75>, <1 1>]
   }
   slot {class=OverlaySlot, properties {horizontalAlignment=HAlign_Right, verticalAlignment=VAlign_Bottom, padding=<ltrb 0 0 32 32>}}
  }
  PilgrimTouchButton_L_lane3_right {
   widgetClassPath=&PilgrimTouchButton_L
   bIsVariable=true
   parent=RightPanel5LaneTouch_LeftyFlip
   widget=PilgrimTouchButton_L_lane3_right
   properties {
    deactivateInputActionIfDragOutsideHitbox=true
    pressAction=&IA_PilgrimLane3Press_5Lane
    visibility=HitTestInvisible
    replace Slate Brush {
     bIsDynamicallyLoaded=false
     drawAs=Image
     imageType=NoImage
     mirroring=NoMirror
     resourceName=None
     resourceObject=&T_Mobile_Lane3_Sprite
     tiling=NoTile
     imageSize=<73 64>
     margin=<ltrb 0 0 0 0>
     outlineSettings {
      bUseBrushTransparency=false
      roundingType=HalfHeightRadius
      width=0
      color {colorUseRule=UseColor_Specified, specifiedColor=<rgba 0 0 0 0>}
      cornerRadii=<0 0 0 1>
     }
     tintColor {colorUseRule=UseColor_Specified, specifiedColor=<rgba 1 1 1 1>}
     uVRegion {bIsValid=false, max=<0 0>, min=<0 0>}
    }
    screenSpaceButtonBoundaries [<1 1>, <0.7119 1>, <0.735 0.713>]
   }
   slot {class=OverlaySlot, properties {horizontalAlignment=HAlign_Right, verticalAlignment=VAlign_Bottom, padding=<ltrb 0 0 254 21>}}
  }
  PilgrimTouchButton_L_lane4 {
   widgetClassPath=&PilgrimTouchButton_L
   bIsVariable=true
   parent=RightPanel5LaneTouch_LeftyFlip
   widget=PilgrimTouchButton_L_lane4
   properties {
    deactivateInputActionIfDragOutsideHitbox=true
    pressAction=&IA_PilgrimLane2Press_5Lane
    visibility=HitTestInvisible
    replace Slate Brush {
     bIsDynamicallyLoaded=false
     drawAs=Image
     imageType=NoImage
     mirroring=NoMirror
     resourceName=None
     resourceObject=&T_Mobile_Lane4_Sprite
     tiling=NoTile
     imageSize=<64 64>
     margin=<ltrb 0 0 0 0>
     outlineSettings {
      bUseBrushTransparency=false
      roundingType=HalfHeightRadius
      width=0
      color {colorUseRule=UseColor_Specified, specifiedColor=<rgba 0 0 0 0>}
      cornerRadii=<0 0 0 1>
     }
     tintColor {colorUseRule=UseColor_Specified, specifiedColor=<rgba 1 1 1 1>}
     uVRegion {bIsValid=false, max=<0 0>, min=<0 0>}
    }
    screenSpaceButtonBoundaries [<1 1>, <0.735 0.713>, <0.834 0.525>]
   }
   slot {class=OverlaySlot, properties {horizontalAlignment=HAlign_Right, verticalAlignment=VAlign_Bottom, padding=<ltrb 0 0 219 219>}}
  }
  PilgrimTouchButton_L_lane5 {
   widgetClassPath=&PilgrimTouchButton_L
   bIsVariable=true
   parent=RightPanel5LaneTouch_LeftyFlip
   widget=PilgrimTouchButton_L_lane5
   properties {
    deactivateInputActionIfDragOutsideHitbox=true
    pressAction=&IA_PilgrimLane1Press_5Lane
    visibility=HitTestInvisible
    replace Slate Brush {
     bIsDynamicallyLoaded=false
     drawAs=Image
     imageType=NoImage
     mirroring=NoMirror
     resourceName=None
     resourceObject=&T_Mobile_Lane5_Sprite
     tiling=NoTile
     imageSize=<68 64>
     margin=<ltrb 0 0 0 0>
     outlineSettings {
      bUseBrushTransparency=false
      roundingType=HalfHeightRadius
      width=0
      color {colorUseRule=UseColor_Specified, specifiedColor=<rgba 0 0 0 0>}
      cornerRadii=<0 0 0 1>
     }
     tintColor {colorUseRule=UseColor_Specified, specifiedColor=<rgba 1 1 1 1>}
     uVRegion {bIsValid=false, max=<0 0>, min=<0 0>}
    }
    screenSpaceButtonBoundaries [<1 1>, <0.834 0.525>, <1 0.4325>]
   }
   slot {class=OverlaySlot, properties {horizontalAlignment=HAlign_Right, verticalAlignment=VAlign_Bottom, padding=<ltrb 0 0 21 310>}}
  }
  WBP_PolygonalInputZoneController_Right5Lane_1 {
   widgetClassPath=&WBP_PolygonalInputZoneController
   bIsVariable=true
   parent=RightPanel5LaneTouch_LeftyFlip
   widget=WBP_PolygonalInputZoneController_Right5Lane_1
   properties {inputZones [
     {
      activated=true
      deactivateOnRelease=false
      priority=0
      zoneAction=null
      zoneId=""
      zoneWidget=&self:WidgetTree.PilgrimTouchButton_Medium_Overdrive_5lane_1
      zoneVertices [<0.6 1>, <0.65 0.7>, <1 0.67>, <1 1>]
     }
     {
      activated=true
      deactivateOnRelease=false
      priority=0
      zoneAction=null
      zoneId=""
      zoneWidget=&self:WidgetTree.PilgrimTouchButton_L_lane3_right
      zoneVertices [<1 1>, <0.15 1>, <0.15 0.52>]
     }
     {
      activated=true
      deactivateOnRelease=false
      priority=0
      zoneAction=null
      zoneId=""
      zoneWidget=&self:WidgetTree.PilgrimTouchButton_L_lane4
      zoneVertices [<1 1>, <0.15 0.52>, <0.5 0.2>]
     }
     {
      activated=true
      deactivateOnRelease=false
      priority=0
      zoneAction=null
      zoneId=""
      zoneWidget=&self:WidgetTree.PilgrimTouchButton_L_lane5
      zoneVertices [<1 1>, <0.5 0.2>, <1 0>]
     }
    ]}
   slot {class=OverlaySlot, properties {horizontalAlignment=HAlign_Fill, verticalAlignment=VAlign_Fill, padding=<ltrb -175 -220 0 0>}}
  }
 ]
}
