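# Mobile band score widget for Pilgrim Quickplay, stacking a "Band Score" label, a band star display and a borderless score widget.
# On Construct it subscribes to tag-driven visibility changes via Pilgrim.Widget.BandScore and creates a PilgrimBandScoreVM view model.
# When the music battle instance initializes, it counts player states and shows itself only if more than one exists, otherwise collapsing.
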WidgetBlueprint /PilgrimQuickplay/Widgets/Mobile/WBP_PilgrimQuickplay_BandScore_Mobile
graphs [EventGraph {nodes [
   Event Construct()
   > SubscribeForChangingVisibilityByTag!(
    execute
    self:Pilgrim Game Widget Base Object Reference
    GameplayTag:"Gameplay Tag Structure (by ref)"=MakeLiteralGameplayTag!(Value:Gameplay Tag Structure=(TagName="Pilgrim.Widget.BandScore")) -> (ReturnValue:Gameplay Tag Structure)
   ) -> (then)
   > n1 = Construct Object from Class(Class=/Script/PilgrimCoreUI.PilgrimBandScoreVM) -> (ReturnValue:Pilgrim Band Score VM Object Reference)
   > Set PilgrimBandScoreVM(PilgrimBandScoreVM:Pilgrim Band Score VM Object Reference=@n1) -> (Output_Get:Pilgrim Band Score VM Object Reference)
   Event OnMusicBattleInstanceInitializeEvent() {is=title K2Node_Event}
   n2 = Get MusicBattleInstance() -> (MusicBattleInstance:Pilgrim Music Battle Instance Object Reference)
   n3 = Length(TargetArray:Array of Pilgrim Music Battle Player State Structures=Break Pilgrim Music Battle Data!(PilgrimMusicBattleData:"Pilgrim Music Battle Data Structure (by ref)"=GetMusicBattleData!(self:Pilgrim Music Battle Instance Object Reference=@n2) -> (ReturnValue:"Pilgrim Music Battle Data Structure (by ref)")) -> (MusicBattleStates:Array of Pilgrim Music Battle Player State Structures, BandState:Pilgrim Band State Structure) {is=title K2Node_BreakStruct})
   > SetVisibility(InVisibility=Select(Option 0=Collapsed, Option 1=SelfHitTestInvisible, Index="integer > integer"(A=@n3, B=1)))
  ]}]
variables []
widgets {
 info {inheritedWidgetCount=0, parentClass=/Script/PilgrimCoreUI.PilgrimGameWidgetBase, rootWidgetClass=&Overlay, widgetCount=8}
 widgets [
  Overlay_44 {widgetClassPath=&Overlay, widget=Overlay_44}
  SizeBox_1 {
   widgetClassPath=&SizeBox
   parent=Overlay_44
   widget=SizeBox_1
   properties {bOverride_HeightOverride=true, bOverride_WidthOverride=true, heightOverride=204, widthOverride=472.3283}
   slot {class=OverlaySlot, properties {horizontalAlignment=HAlign_Center, verticalAlignment=VAlign_Center}}
  }
  VerticalBox {widgetClassPath=&VerticalBox, parent=SizeBox_1, widget=VerticalBox, slot {class=SizeBoxSlot}}
  Overlay_0 {widgetClassPath=&Overlay, parent=VerticalBox, widget=Overlay_0, slot {class=VerticalBoxSlot}}
  SizeBox_0 {
   widgetClassPath=&SizeBox
   parent=Overlay_0
   widget=SizeBox_0
   properties {bOverride_HeightOverride=true, bOverride_WidthOverride=true, heightOverride=50, widthOverride=374}
   slot {class=OverlaySlot, properties {horizontalAlignment=HAlign_Fill, verticalAlignment=VAlign_Fill}}
  }
  CommonTextBlock_81 {
   widgetClassPath=&CommonTextBlock
   parent=Overlay_0
   widget=CommonTextBlock_81
   properties {
    style=/FortUILibrary/Styles/Typography/_TextStyle_NormalTag._TextStyle_NormalTag_C
    text=Band Score
    font {
     bForceMonospaced=false
     bMaterialIsStencil=false
     fontMaterial=null
     fontObject=&BurbankSmall
     letterSpacing=20
     monospacedWidth=1
     size=15
     skewAmount=0
     typefaceFontName=Black
     outlineSettings {
      bApplyOutlineToDropShadows=false
      bMiteredCorners=false
      bSeparateFillAlpha=false
      outlineMaterial=null
      outlineSize=0
      outlineColor=<rgba 0 0 0 1>
     }
    }
    shadowOffset=<0 0>
    strikeBrush {
     bIsDynamicallyLoaded=false
     drawAs=NoDrawType
     imageType=NoImage
     mirroring=NoMirror
     resourceName=None
     resourceObject=null
     tiling=NoTile
     imageSize=<32 32>
     margin=<ltrb 0 0 0 0>
     outlineSettings {
      bUseBrushTransparency=false
      roundingType=HalfHeightRadius
      width=0
      color {colorUseRule=UseColor_Specified, specifiedColor=<rgba 0 0 0 0>}
      cornerRadii=<0 0 0 0>
     }
     tintColor {colorUseRule=UseColor_Specified, specifiedColor=<rgba 1 1 1 1>}
     uVRegion {bIsValid=false, max=<0 0>, min=<0 0>}
    }
   }
   slot {class=OverlaySlot, properties {horizontalAlignment=HAlign_Right, verticalAlignment=VAlign_Center, padding=<ltrb 0 0 12 0>}}
  }
  WBP_PilgrimQuickplay_BandStarDisplay {
   widgetClassPath=/PilgrimQuickplay/Widgets/WBP_PilgrimQuickplay_BandStarDisplay
   bIsVariable=true
   parent=VerticalBox
   widget=WBP_PilgrimQuickplay_BandStarDisplay
   slot {
    class=VerticalBoxSlot
    properties {horizontalAlignment=HAlign_Center, verticalAlignment=VAlign_Center, padding=<ltrb 0 8 0 8>, size {sizeRule=Fill, value=1}}
   }
  }
  WBP_PilgrimQuickplay_Score_Band {
   widgetClassPath=/PilgrimQuickplay/Widgets/WBP_PilgrimQuickplay_Score
   bIsVariable=true
   parent=VerticalBox
   widget=WBP_PilgrimQuickplay_Score_Band
   properties {shouldShowBorder=false}
   slot {class=VerticalBoxSlot}
  }
 ]
}
