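# Mobile Pilgrim Quickplay widget showing the local player's score, star bar and label, with optional instrument and difficulty icons.
# On Construct it subscribes to tag-driven visibility, updates the label and creates a player score view model.
# The label update resolves the playspace's local player override, falling back to the owning player state, for instrument and difficulty.
# The Arabic flip function ignores its input, so scale stays unflipped.
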WidgetBlueprint /PilgrimQuickplay/Widgets/Mobile/WBP_PilgrimQuickplay_LocalScore_Mobile
graphs [
 EventGraph {nodes [
   Event PreConstruct()
   > FlipReadDirecitonArabic!(execute, self, IsAr:bool="Equal, Case Insensitive (String)"(A=GetCurrentLanguage!() -> (ReturnValue:String), B=ar)) -> (then)
   Event Construct()
   > SetVisibility(
    self=Get InstrumentBox() -> (InstrumentBox:Horizontal Box Object Reference)
    InVisibility=Select(Option 0=Collapsed, Option 1=SelfHitTestInvisible, Index=Get UseRoleAndDifficultyIcons() -> (UseRoleAndDifficultyIcons:bool))
   )
   > FlipReadDirecitonArabic(IsAr="Equal, Case Insensitive (String)"(A=GetCurrentLanguage(), B=ar))
   > SubscribeForChangingVisibilityByTag!(
    execute
    self:Pilgrim Game Widget Base Object Reference
    GameplayTag:"Gameplay Tag Structure (by ref)"=MakeLiteralGameplayTag!(Value:Gameplay Tag Structure=(TagName="Pilgrim.Widget.LocalScore")) -> (ReturnValue:Gameplay Tag Structure)
   ) -> (then)
   > UpdateInstrumentText()
   > n1 = Construct Object from Class(Class=/Script/PilgrimCoreUI.PilgrimPlayerScoreVM) -> (ReturnValue:Pilgrim Player Score VM Object Reference)
   > Set PilgrimPlayerScoreVM(PilgrimPlayerScoreVM:Pilgrim Player Score VM Object Reference=@n1) -> (Output_Get:Pilgrim Player Score VM Object Reference)
  ]}
 FlipReadDirecitonArabic {nodes [
   FlipReadDirecitonArabic!() -> (then, IsAr:bool=~) {is=title K2Node_FunctionEntry}
   > SetRenderScale!(
    execute
    self:Widget Object Reference=[Get Overlay_0() -> (Overlay_0:Overlay Object Reference), Get SizeBox_0() -> (SizeBox_0:Size Box Object Reference)]
    Scale:Vector 2D Structure=MakeVector2D(X=Select(Option 0:double=1, Option 1:double=-1, Index=false) -> (ReturnValue:double), Y=1)
   ) -> (then)
  ]}
 Set Instrument Type {nodes [
   n1 = Set Instrument Type!() -> (then, TrackType:EPilgrimTrackType Enum=TrackGuitar) {is=title K2Node_FunctionEntry}
   > SetBrushFromSoftMaterial!(
    execute
    self:Image Object Reference=Get InstrumentImage() -> (InstrumentImage:Fort Lazy Image Object Reference)
    SoftMaterial:Material Interface Soft Object Reference=GetIconMaterialByInstrument!(
     Instrument:EPilgrimTrackType Enum=@n1
     SongMetaData:Pilgrim Song Metadata Object Reference
     SparksInstrument:ESparksInstrumentType Enum=None
     bUseKeytarIcon:bool=true
     UseProIcons:bool=true
    ) -> (IconMaterial:Material Instance Soft Object Reference)
   ) -> (then)
  ]}
 UpdateInstrumentText {nodes [
   UpdateInstrumentText() {is=title K2Node_FunctionEntry}
   > n1 = Sequence()
   n2 = GetSparksMusicPlayspaceOfClass!(
    PlayspaceType:Sparks Music Playspace Class Reference=/Script/PilgrimQuickplayRuntime.PilgrimQuickplayPlayspace
    Actor:Actor Object Reference=GetOwningPlayer()
    bMatchChildTypes:bool=true
   ) -> (ReturnValue:Pilgrim Quickplay Playspace Object Reference)
   > n3 = Is Valid(InputObject=@n2)
   n4 = GetComponentByClass!(
    self:Actor Object Reference=@n2
    ComponentClass:Actor Component Class Reference=/Script/PilgrimQuickplayRuntime.PilgrimQuickplayPlayspaceComponent_PlayerManager
   ) -> (ReturnValue:Pilgrim Quickplay Playspace Component Player Manager Object Reference)
   > n5 = Is Valid(InputObject=@n4)
   n6 = GetLocalPlayerOverride!(self:Pilgrim Quickplay Playspace Component Player Manager Object Reference=@n4) -> (ReturnValue:Player State Object Reference)
   > n7 = Is Valid(InputObject=@n6)
   > n8 = Set PlayerToUse(PlayerToUse:Player State Object Reference=@n6) -> (Output_Get:Player State Object Reference)
   n9 = GetPilgrimQuickplayPlayerStateComponent!(execute=[@n10.then, @n8.then], PlayerState:Player State Object Reference=Get PlayerToUse() -> (PlayerToUse:Player State Object Reference)) -> (then, ReturnValue:Pilgrim Quickplay Player State Component Object Reference)
   > Set TrackType(TrackType:EPilgrimTrackType Enum=GetInstrumentChoice!(self:Pilgrim Core Player State Component Object Reference=@n9) -> (ReturnValue:EPilgrimTrackType Enum=TrackGuitar)) -> (Output_Get:EPilgrimTrackType Enum=TrackGuitar)
   n10 = Set PlayerToUse(
    execute=[@n7.Is Not Valid, @n5.Is Not Valid, @n3.Is Not Valid]
    PlayerToUse=Get PlayerState!(self:Controller Object Reference=GetOwningPlayer()) -> (PlayerState:Player State Object Reference)
   )
   n11 = Branch(execute=@n1.then_1, Condition=Get UseRoleAndDifficultyIcons())
   > Set Instrument Type!(execute, self, TrackType:EPilgrimTrackType Enum=Get TrackType() -> (TrackType:EPilgrimTrackType Enum=TrackGuitar)) -> (then) {is=title K2Node_CallFunction}
   > n12 = GetPilgrimQuickplayPlayerStateComponent(PlayerState=Get PlayerToUse())
   > SetDifficulty!(
    execute
    self:WBP Pilgrim Results Difficulty Object Reference=Get WBP_Pilgrim_Results_Difficulty() -> (WBP_Pilgrim_Results_Difficulty:WBP Pilgrim Results Difficulty Object Reference)
    Difficulty:EPilgrimSongDifficulty Enum=GetSongDifficulty!(self:Pilgrim Core Player State Component Object Reference=@n12) -> (ReturnValue:EPilgrimSongDifficulty Enum=DifficultyEasy)
   ) -> (then)
   > Set Text(Text="LOCTEXT(\"Score\")", self=Get PlayerScore_TextBlock() -> (PlayerScore_TextBlock:Common Text Object Reference))
   n13 = Get TrackType() -> (TrackType=TrackGuitar)
   Set Text(
    execute=@n11.else
    Text=Format Text!(
     Format="LOCTEXT(\"{Instrument} Score\")"
     Instrument:Text=GetInstrumentNameFromPart!(TrackType:EPilgrimTrackType Enum=@n13) -> (InstrumentText:Text)
    ) -> (Result)
    self=Get PlayerScore_TextBlock()
   )
  ]}
]
variables [
 Index {type=EPilgrimTrackType, default=TrackNone}
 PlayerToUse {type=PlayerState, default=null}
 TrackType {type=EPilgrimTrackType, default=TrackGuitar}
 UseRoleAndDifficultyIcons {type=boolean}
]
widgets {
 info {inheritedWidgetCount=0, parentClass=/Script/PilgrimCoreUI.PilgrimGameWidgetBase, rootWidgetClass=&Overlay, widgetCount=14}
 widgets [
  Overlay_44 {widgetClassPath=&Overlay, widget=Overlay_44}
  SizeBox_2 {
   widgetClassPath=&SizeBox
   parent=Overlay_44
   widget=SizeBox_2
   properties {bOverride_HeightOverride=true, bOverride_WidthOverride=true, heightOverride=204, widthOverride=472.3283}
   slot {class=OverlaySlot, properties {horizontalAlignment=HAlign_Center, verticalAlignment=VAlign_Center}}
  }
  VerticalBox_0 {widgetClassPath=&VerticalBox, parent=SizeBox_2, widget=VerticalBox_0, slot {class=SizeBoxSlot}}
  Overlay_0 {widgetClassPath=&Overlay, bIsVariable=true, parent=VerticalBox_0, widget=Overlay_0, slot {class=VerticalBoxSlot}}
  WBP_PilgrimQuickplay_Score_Player {
   widgetClassPath=/PilgrimQuickplay/Widgets/WBP_PilgrimQuickplay_Score
   bIsVariable=true
   parent=Overlay_0
   widget=WBP_PilgrimQuickplay_Score_Player
   properties {flipBacking=true}
   slot {class=OverlaySlot, properties {horizontalAlignment=HAlign_Fill, verticalAlignment=VAlign_Fill}}
  }
  SizeBox_0 {
   widgetClassPath=&SizeBox
   bIsVariable=true
   parent=Overlay_0
   widget=SizeBox_0
   properties {bOverride_HeightOverride=true, bOverride_WidthOverride=true, heightOverride=50, widthOverride=374}
   slot {class=OverlaySlot, properties {horizontalAlignment=HAlign_Fill, verticalAlignment=VAlign_Fill}}
  }
  HorizontalBox_0 {
   widgetClassPath=&HorizontalBox
   parent=Overlay_0
   widget=HorizontalBox_0
   slot {class=OverlaySlot, properties {verticalAlignment=VAlign_Fill}}
  }
  InstrumentBox {widgetClassPath=&HorizontalBox, bIsVariable=true, parent=HorizontalBox_0, widget=InstrumentBox, slot {class=HorizontalBoxSlot}}
  InstrumentImage {
   widgetClassPath=&FortLazyImage
   bIsVariable=true
   parent=InstrumentBox
   widget=InstrumentImage
   properties {brush {
     bIsDynamicallyLoaded=false
     drawAs=Image
     imageType=NoImage
     mirroring=NoMirror
     resourceName=None
     resourceObject=/PilgrimQuickplay/Widgets/Materials/MI_UI_Icon_Instrument_Type_Guitar
     tiling=NoTile
     imageSize=<52 52>
     margin=<ltrb 0 0 0 0>
     outlineSettings {
      bUseBrushTransparency=false
      roundingType=HalfHeightRadius
      width=0
      color {colorUseRule=UseColor_Specified, specifiedColor=<rgba 0 0 0 0>}
      cornerRadii=<0 0 0 0>
     }
     tintColor {colorUseRule=UseColor_Specified, specifiedColor=<rgba 1 1 1 1>}
     uVRegion {bIsValid=false, max=<0 0>, min=<0 0>}
    }}
   slot {class=HorizontalBoxSlot, properties {horizontalAlignment=HAlign_Right, verticalAlignment=VAlign_Center, padding=<ltrb 0 0 6 0>}}
  }
  SC_Difficulty {
   widgetClassPath=&ScaleBox
   parent=InstrumentBox
   widget=SC_Difficulty
   properties {stretch=UserSpecified, userSpecifiedScale=0.5}
   slot {class=HorizontalBoxSlot, properties {horizontalAlignment=HAlign_Right, verticalAlignment=VAlign_Center, padding=<ltrb 0 0 6 0>}}
  }
  WBP_Pilgrim_Results_Difficulty {
   widgetClassPath=/PilgrimQuickplay/Widgets/Results/WBP_Pilgrim_Results_Difficulty
   bIsVariable=true
   parent=SC_Difficulty
   widget=WBP_Pilgrim_Results_Difficulty
   properties {renderTransform {angle=0, scale=<1.5 1.5>, shear=<0 0>, translation=<0 0>}}
   slot {class=ScaleBoxSlot}
  }
  LabelSizeBox {
   widgetClassPath=&SizeBox
   bIsVariable=true
   parent=HorizontalBox_0
   widget=LabelSizeBox
   properties {bOverride_MaxDesiredWidth=true, maxDesiredWidth=160, minDesiredWidth=144}
   slot {class=HorizontalBoxSlot, properties {padding=<ltrb 12 0 0 0>}}
  }
  PlayerScore_TextBlock {
   widgetClassPath=&CommonTextBlock
   bIsVariable=true
   parent=LabelSizeBox
   widget=PlayerScore_TextBlock
   properties {
    scrollStyle=&TextScrollStyle-Base_C
    text=Score
    textTransformPolicy=ToUpper
    font {
     bForceMonospaced=false
     bMaterialIsStencil=false
     fontMaterial=/PilgrimCore/Widgets/Materials/MI_UI_Piligrim_HUDTextColor
     fontObject=&HeadingNow
     letterSpacing=20
     monospacedWidth=1
     size=12
     skewAmount=0
     typefaceFontName=76 Bold
     outlineSettings {
      bApplyOutlineToDropShadows=false
      bMiteredCorners=false
      bSeparateFillAlpha=true
      outlineMaterial=&M-FontShadow
      outlineSize=2
      outlineColor=<rgba 0 0 0 1>
     }
    }
    shadowOffset=<0 0>
    strikeBrush {
     bIsDynamicallyLoaded=false
     drawAs=NoDrawType
     imageType=NoImage
     mirroring=NoMirror
     resourceName=None
     resourceObject=null
     tiling=NoTile
     imageSize=<32 32>
     margin=<ltrb 0 0 0 0>
     outlineSettings {
      bUseBrushTransparency=false
      roundingType=HalfHeightRadius
      width=0
      color {colorUseRule=UseColor_Specified, specifiedColor=<rgba 0 0 0 0>}
      cornerRadii=<0 0 0 0>
     }
     tintColor {colorUseRule=UseColor_Specified, specifiedColor=<rgba 1 1 1 1>}
     uVRegion {bIsValid=false, max=<0 0>, min=<0 0>}
    }
   }
   slot {class=SizeBoxSlot, properties {horizontalAlignment=HAlign_Left, verticalAlignment=VAlign_Center}}
  }
  WBP_PilgrimQuickplay_LocalStarBar {
   widgetClassPath=/PilgrimQuickplay/Widgets/WBP_PilgrimQuickplay_LocalStarBar
   bIsVariable=true
   parent=VerticalBox_0
   widget=WBP_PilgrimQuickplay_LocalStarBar
   properties {shouldStarsAnimateComplete=true}
   slot {class=VerticalBoxSlot, properties {padding=<ltrb 5 8 5 8>, size {sizeRule=Fill, value=1}}}
  }
 ]
}
