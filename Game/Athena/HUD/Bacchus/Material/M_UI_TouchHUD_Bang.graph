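# M_UI_TouchHUD_Bang is a translucent UI material for the touch HUD that draws an anti-aliased dot with a repeating ping ring.
# Scaled time, wrapped by Fmod and eased by Power, drives the ring's growing radius, thinning width, changing softness and fade-out.
# An inverted circle SDF masks the ring's center, then the static dot is added and the result saturated.
# Parameters expose ping speed, size, amount and the mask size.
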Material /Game/Athena/HUD/Bacchus/Material/M_UI_TouchHUD_Bang
graph {
 properties {bAutomaticallySetUsageInEditor=false, blendMode=BLEND_Translucent, materialDomain=MD_UI, usedShadingModels=Default Lit}
 nodes [
  Constant() {r=1}
  n1 = ScalarParameter() {defaultValue=0.52, parameterName=InnerCircleMaskSize}
  n2 = Constant() {r=0.025}
  n3 = OneMinus(None=MaterialFunctionCall(DistanceField=MaterialFunctionCall(Size=@n1, GlowMax=@n2) {materialFunction=&MF_UI_Circle}.SDF) {materialFunction=&MF_UI_AntiAliasing}.Inner)
  n4 = Time() {period=1}
  n5 = ScalarParameter() {defaultValue=0.8, group=Ping, parameterName=PingSpeed}
  n6 = Power(Base=Fmod(A=Multiply(A=@n4, B=@n5) {constB=0.5}, B=Constant() {r=1})) {constExponent=0.4}
  n7 = MaterialFunctionCall!(
   RingSize=LinearInterpolate(A=ScalarParameter() {defaultValue=0.2, group=Ping, parameterName=PingRingMaxSize}, Alpha=@n6) {constA=0.1, constB=0.01}
   Edge
   EdgeSoftness=LinearInterpolate(
    A=ScalarParameter() {defaultValue=0.002, parameterName=OuterCircleSoftness}
    B=ScalarParameter() {defaultValue=0.05, parameterName=InnerCircleSoftness}
    Alpha=@n6
   ) {constA=0.002, constB=0.05}
   OuterRadius=LinearInterpolate(B=ScalarParameter() {defaultValue=0.5, group=Ping, parameterName=PingOuterRadius}, Alpha=@n6) {constB=0.5}
  ) {materialFunction=/Game/UI/InGame/HUD/Textures/MF_UI_Ring}
  n8 = LinearInterpolate(Alpha=@n6) {constA=2, constB=0}
  n9 = Multiply(A=Clamp(None=Multiply(A=@n7.Result, B=@n8)), B=ScalarParameter() {defaultValue=1, group=Ping, parameterName=PingAmount, sliderMax=1})
  n10 = MaterialFunctionCall(Size=Constant() {r=0.4}) {materialFunction=&MF_UI_Circle}
  Saturate(None=Add(
    A=Multiply(A=@n3, B=@n9)
    B=MaterialFunctionCall(DistanceField=@n10.SDF) {materialFunction=&MF_UI_AntiAliasing}.Inner
   ))
 ]
}
