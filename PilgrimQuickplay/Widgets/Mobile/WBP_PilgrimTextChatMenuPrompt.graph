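# Mobile text chat prompt widget: a chat icon with an invisible button and an unread notification bang.
# Pressing it injects IA_OpenTextChat through player 0's Enhanced Input subsystem, so it acts like the real input.
# It listens on the global event router for bottom-bar notification counts, showing the bang for unread counts on touch or mobile only.
# Optional flags hide the whole prompt outside touch/mobile and override the icon's padding.
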WidgetBlueprint /PilgrimQuickplay/Widgets/Mobile/WBP_PilgrimTextChatMenuPrompt
graphs [
 DetermineMobileTouchCase {nodes [
   DetermineMobileTouchCase() {is=title K2Node_FunctionEntry}
   n1 = GetOwningPlayer()
   > Return Node!(execute, isTouchOrMobile:bool=OR Boolean(A=IsUsingTouch(PlayerController=@n1), B=IsMobileGame()))
  ]}
 EventGraph {nodes [
   "On Pressed (TextChatMenuButton)"() {is=title K2Node_ComponentBoundEvent}
   n1 = GetPlayerController!(PlayerIndex:int=0) -> (ReturnValue:Player Controller Object Reference)
   > Is Valid(InputObject=@n1)
   n2 = Get EnhancedInputLocalPlayerSubsystem!(PlayerController:Player Controller Object Reference=@n1) -> (ReturnValue:Enhanced Input Local Player Subsystem Object Reference) {is=title K2Node_GetSubsystemFromPC}
   > Is Valid(InputObject=@n2)
   > InjectInputForAction!(
    execute
    self:Enhanced Input Subsystem Interface Interface=@n2
    Action:Input Action Object Reference=/Game/Input/EnhancedInput/FortPlayerController/IA_OpenTextChat
    RawValue_X:double=1
    RawValue_Y:double=0
    RawValue_Z:double=0
    RawValue_ValueType:EInputActionValueType Enum=Boolean
    Modifiers:Array of Input Modifier Object References
    Triggers:Array of Input Trigger Object References
   ) -> (then)
   Event Construct()
   > Parent: Construct()
   > n3 = Sequence!(execute) -> (then_0, then_1, then_2:Exec)
   > Branch(Condition=Get NeedCheckMobile() -> (NeedCheckMobile:bool))
   n4 = GetOwningPlayer()
   > n5 = Is Valid(InputObject=@n4)
   > n6 = Branch(Condition=OR Boolean(A=IsUsingTouch(PlayerController=@n4), B=IsMobileGame()))
   > SetVisibility(InVisibility=Visible)
   SetVisibility(execute=[@n6.else, @n5.Is Not Valid], InVisibility=Collapsed)
   n7 = StartListeningToEvent!(
    execute:Exec=@n3.then_1
    Target:Gameplay Event Router Component Object Reference=GetGlobalEventRouter()
    EventType:Script Struct Object Reference=/Game/UI/Frontend/BottomBar/BPS_BottomBarNotificationCount
    Context:Object Reference=GetOwningPlayer()
    EventBubblingRule:EEventBubblingRule Enum
   ) -> (
    then:Exec
    AsyncTaskProxy:Async Action Start Listening To Event Object Reference
    OnEventReceived:Exec
    Payload:BPS Bottom Bar Notification Count Structure
   )
   > Set UnreadChannelsListener(UnreadChannelsListener:Async Action Start Listening To Event Object Reference=@n7) -> (Output_Get:Async Action Start Listening To Event Object Reference)
   SetUnreadNotificationChannel!(
    execute=@n7.OnEventReceived
    self
    UnreadChannelsCount:int=Break BPS Bottom Bar Notification Count!(BPS_BottomBarNotificationCount:"BPS Bottom Bar Notification Count Structure (by ref)"=@n7.Payload) -> (NotificationCount_8_#1:int) {is=title K2Node_BreakStruct}
   ) -> (then)
   n8 = Event PreConstruct()
   > Parent: PreConstruct(IsDesignTime=@n8.IsDesignTime)
   > n9 = Sequence!(execute) -> (then_0, then_1)
   > Branch(Condition=Get RemovePadding() -> (RemovePadding:bool))
   > SetPadding!(
    execute
    self:Overlay Slot Object Reference=SlotAsOverlaySlot(Widget=Get HorizontalBox_0() -> (HorizontalBox_0:Horizontal Box Object Reference))
    InPadding:Margin Structure=Get New Padding Hit Zone() -> (New Padding Hit Zone:Margin Structure)
   ) -> (then)
   SetVisibility(execute=@n9.then_1, self=Get Image_Bang_Notification() -> (Image_Bang_Notification:Image Object Reference), InVisibility=Collapsed)
   Event Destruct()
   > StopListeningToEvent(self:Async Action Start Listening To Event Object Reference=Get UnreadChannelsListener() -> (UnreadChannelsListener:Async Action Start Listening To Event Object Reference))
  ]}
 SetUnreadNotificationChannel {nodes [
   n1 = SetUnreadNotificationChannel!() -> (then, UnreadChannelsCount:int) {is=title K2Node_FunctionEntry}
   > n2 = Branch(Condition="Equal (Integer)"(A=@n1))
   n3 = Get Image_Bang_Notification()
   > SetVisibility(self=@n3, InVisibility=Collapsed)
   Branch(execute=@n2.else, Condition=DetermineMobileTouchCase!() -> (isTouchOrMobile:bool) {is=title K2Node_CallFunction})
   > SetVisibility(self=@n3, InVisibility=HitTestInvisible)
  ]}
]
variables [
 NeedCheckMobile {type=boolean}
 New Padding Hit Zone {type=Margin, default=<ltrb 0 0 0 0>}
 RemovePadding {type=boolean}
 UnreadChannelsListener {type=AsyncAction_StartListeningToEvent, default=null}
]
widgets {
 info {inheritedWidgetCount=0, parentClass=&CommonUserWidget, rootWidgetClass=&ScaleBox, widgetCount=6}
 widgets [
  ScaleBox_0 {widgetClassPath=&ScaleBox, widget=ScaleBox_0, properties {stretch=UserSpecified}}
  TextChatMenuPrompt {
   widgetClassPath=&Overlay
   bIsVariable=true
   parent=ScaleBox_0
   widget=TextChatMenuPrompt
   slot {class=ScaleBoxSlot, properties {horizontalAlignment=HAlign_Fill, verticalAlignment=VAlign_Fill}}
  }
  HorizontalBox_0 {
   widgetClassPath=&HorizontalBox
   bIsVariable=true
   parent=TextChatMenuPrompt
   widget=HorizontalBox_0
   slot {class=OverlaySlot, properties {horizontalAlignment=HAlign_Fill, verticalAlignment=VAlign_Fill, padding=<ltrb 20 20 20 20>}}
  }
  Image_Icon {
   widgetClassPath=&Image
   bIsVariable=true
   parent=HorizontalBox_0
   widget=Image_Icon
   properties {
    visibility=SelfHitTestInvisible
    brush {
     bIsDynamicallyLoaded=false
     drawAs=Image
     imageType=FullColor
     mirroring=NoMirror
     resourceName=None
     resourceObject=/FMCore/Textures/T_UI_Mobile_OpenChatSparks
     tiling=NoTile
     imageSize=<84 84>
     margin=<ltrb 0 0 0 0>
     outlineSettings {
      bUseBrushTransparency=false
      roundingType=HalfHeightRadius
      width=0
      color {colorUseRule=UseColor_Specified, specifiedColor=<rgba 0 0 0 0>}
      cornerRadii=<0 0 0 0>
     }
     tintColor {colorUseRule=UseColor_Specified, specifiedColor=<rgba 1 1 1 1>}
     uVRegion {bIsValid=false, max=<0 0>, min=<0 0>}
    }
   }
   slot {class=HorizontalBoxSlot}
  }
  Image_Bang_Notification {
   widgetClassPath=&Image
   bIsVariable=true
   parent=TextChatMenuPrompt
   widget=Image_Bang_Notification
   properties {brush {
     bIsDynamicallyLoaded=false
     drawAs=Image
     imageType=FullColor
     mirroring=NoMirror
     resourceName=None
     resourceObject=/Game/Athena/HUD/Bacchus/Material/M_UI_TouchHUD_Bang
     tiling=NoTile
     imageSize=<40 40>
     margin=<ltrb 0 0 0 0>
     outlineSettings {
      bUseBrushTransparency=false
      roundingType=HalfHeightRadius
      width=0
      color {colorUseRule=UseColor_Specified, specifiedColor=<rgba 0 0 0 0>}
      cornerRadii=<0 0 0 0>
     }
     tintColor {colorUseRule=UseColor_Specified, specifiedColor=<rgba 0.9301 1 0.0103 1>}
     uVRegion {bIsValid=false, max=<0 0>, min=<0 0>}
    }}
   slot {class=OverlaySlot, properties {horizontalAlignment=HAlign_Right}}
  }
  TextChatMenuButton {
   widgetClassPath=&Button
   parent=TextChatMenuPrompt
   widget=TextChatMenuButton
   properties {widgetStyle {
     clickedSlateSound {resourceObject=null}
     disabled {
      bIsDynamicallyLoaded=false
      drawAs=NoDrawType
      imageType=NoImage
      mirroring=NoMirror
      resourceName=None
      resourceObject=null
      tiling=NoTile
      imageSize=<0 0>
      margin=<ltrb 0 0 0 0>
      outlineSettings {
       bUseBrushTransparency=false
       roundingType=HalfHeightRadius
       width=0
       color {colorUseRule=UseColor_Specified, specifiedColor=<rgba 0 0 0 0>}
       cornerRadii=<0 0 0 0>
      }
      tintColor {colorUseRule=UseColor_Specified, specifiedColor=<rgba 1 1 1 1>}
      uVRegion {bIsValid=false, max=<0 0>, min=<0 0>}
     }
     disabledForeground {colorUseRule=UseColor_Specified, specifiedColor=<rgba 0.5271 0.5271 0.5271 1>}
     hovered {
      bIsDynamicallyLoaded=false
      drawAs=NoDrawType
      imageType=NoImage
      mirroring=NoMirror
      resourceName=None
      resourceObject=null
      tiling=NoTile
      imageSize=<32 32>
      margin=<ltrb 0 0 0 0>
      outlineSettings {
       bUseBrushTransparency=true
       roundingType=FixedRadius
       width=1
       color {colorUseRule=UseColor_Specified, specifiedColor=<rgba 0.7243 0.7243 0.7243 1>}
       cornerRadii=<4 4 4 4>
      }
      tintColor {colorUseRule=UseColor_Specified, specifiedColor=<rgba 0.7243 0.7243 0.7243 1>}
      uVRegion {bIsValid=false, max=<0 0>, min=<0 0>}
     }
     hoveredForeground {colorUseRule=UseColor_Specified, specifiedColor=<rgba 0 1 1 1>}
     hoveredSlateSound {resourceObject=null}
     normal {
      bIsDynamicallyLoaded=false
      drawAs=NoDrawType
      imageType=NoImage
      mirroring=NoMirror
      resourceName=None
      resourceObject=null
      tiling=NoTile
      imageSize=<32 32>
      margin=<ltrb 0 0 0 0>
      outlineSettings {
       bUseBrushTransparency=true
       roundingType=FixedRadius
       width=1
       color {colorUseRule=UseColor_Specified, specifiedColor=<rgba 0.6951 0.6951 0.6951 1>}
       cornerRadii=<4 4 4 4>
      }
      tintColor {colorUseRule=UseColor_Specified, specifiedColor=<rgba 0.4955 0.4955 0.4955 1>}
      uVRegion {bIsValid=false, max=<0 0>, min=<0 0>}
     }
     normalForeground {colorUseRule=UseColor_Specified, specifiedColor=<rgba 0 1 1 1>}
     normalPadding=<ltrb 12 1.5 12 1.5>
     pressed {
      bIsDynamicallyLoaded=false
      drawAs=NoDrawType
      imageType=NoImage
      mirroring=NoMirror
      resourceName=None
      resourceObject=null
      tiling=NoTile
      imageSize=<32 32>
      margin=<ltrb 0 0 0 0>
      outlineSettings {
       bUseBrushTransparency=true
       roundingType=FixedRadius
       width=1
       color {colorUseRule=UseColor_Specified, specifiedColor=<rgba 0.7243 0.7243 0.7243 1>}
       cornerRadii=<4 4 4 4>
      }
      tintColor {colorUseRule=UseColor_Specified, specifiedColor=<rgba 0.3843 0.3843 0.3843 1>}
      uVRegion {bIsValid=false, max=<0 0>, min=<0 0>}
     }
     pressedForeground {colorUseRule=UseColor_Specified, specifiedColor=<rgba 0 1 1 1>}
     pressedPadding=<ltrb 12 1.5 12 1.5>
     pressedSlateSound {resourceObject=null}
    }}
   slot {class=OverlaySlot, properties {horizontalAlignment=HAlign_Fill, verticalAlignment=VAlign_Fill}}
  }
 ]
}
