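# Touch HUD widget for Pilgrim quickplay that groups the text chat prompt, song menu prompt and hamburger menu button in a horizontal row.
# On construct, it starts async listeners for the Pilgrim.Widget.HamburgerButton gameplay tag on the owning player's PlayerState.
# The whole widget collapses when the tag is added and returns to SelfHitTestInvisible when it is removed.
# On destruct, it cancels and clears the stored async handles.
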WidgetBlueprint /PilgrimQuickplay/Widgets/Mobile/WBP_PilgrimTouch_HamburgerMenus
graphs [EventGraph {nodes [
   Event Destruct()
   > Parent: Destruct()
   n1 = Get AsyncEventHandles() -> (AsyncEventHandles:Array of Cancellable Async Action Object References)
   > n2 = For Each Loop(Array:Array of Cancellable Async Action Object References=@n1) -> (Array Element:Cancellable Async Action Object Reference)
   > Is Valid(InputObject=@n2)
   > Cancel(self:Cancellable Async Action Object Reference=@n2)
   Clear(execute=@n2.Completed, TargetArray:Array of Cancellable Async Action Object References=@n1)
   Event Construct()
   > Parent: Construct()
   > n3 = Sequence()
   n4 = Get PlayerState!(self:Controller Object Reference=GetOwningPlayer()) -> (PlayerState:Player State Object Reference)
   > n5 = WaitGameplayTagAddToActor!(
    execute:Exec
    TargetActor:Actor Object Reference=@n4
    Tag:Gameplay Tag Structure=(TagName="Pilgrim.Widget.HamburgerButton")
    OnlyTriggerOnce:bool=false
   ) -> (then:Exec, AsyncTaskProxy:Ability Async Wait Gameplay Tag Added Object Reference, Added:Exec) {is=title K2Node_AsyncAction}
   > Add(
    TargetArray:Array of Cancellable Async Action Object References=Get AsyncEventHandles()
    NewItem:"Cancellable Async Action Object Reference (by ref)"=@n5
   ) {is=title K2Node_CallArrayFunction}
   SetVisibility(execute=@n5.Added, InVisibility=Collapsed)
   n6 = WaitGameplayTagRemoveFromActor!(
    execute:Exec=@n3.then_1
    TargetActor:Actor Object Reference=@n4
    Tag:Gameplay Tag Structure=(TagName="Pilgrim.Widget.HamburgerButton")
    OnlyTriggerOnce:bool=false
   ) -> (then:Exec, AsyncTaskProxy:Ability Async Wait Gameplay Tag Removed Object Reference, Removed:Exec) {is=title K2Node_AsyncAction}
   > Add(TargetArray=Get AsyncEventHandles())
   SetVisibility(execute=@n6.Removed, InVisibility=SelfHitTestInvisible)
  ]}]
variables [AsyncEventHandles {type="array<CancellableAsyncAction>", default []}]
widgets {
 info {inheritedWidgetCount=0, parentClass=&CommonUserWidget, rootWidgetClass=&Overlay, widgetCount=5}
 widgets [
  Overlay_15 {widgetClassPath=&Overlay, widget=Overlay_15}
  HorizontalBox_1 {widgetClassPath=&HorizontalBox, parent=Overlay_15, widget=HorizontalBox_1, slot {class=OverlaySlot}}
  WBP_PilgrimTextChatMenuPrompt {
   widgetClassPath=&/WBP_PilgrimTextChatMenuPrompt
   bIsVariable=true
   parent=HorizontalBox_1
   widget=WBP_PilgrimTextChatMenuPrompt
   slot {class=HorizontalBoxSlot}
  }
  WBP_SongMenuPrompt_Touch {
   widgetClassPath=&/WBP_SongMenuPrompt_Touch
   bIsVariable=true
   parent=HorizontalBox_1
   widget=WBP_SongMenuPrompt_Touch
   slot {class=HorizontalBoxSlot, properties {horizontalAlignment=HAlign_Left, verticalAlignment=VAlign_Top}}
  }
  HamburgerButton {
   widgetClassPath=/Game/Athena/HUD/Bacchus/CustomizableHUD/HamburgerButton
   bIsVariable=true
   parent=HorizontalBox_1
   widget=HamburgerButton
   properties {elementTag {tagName=MobileHUD.HamburgerMenu}}
   slot {class=HorizontalBoxSlot, properties {horizontalAlignment=HAlign_Left, verticalAlignment=VAlign_Top, padding=<ltrb 0 15 0 0>}}
  }
 ]
}
