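# Mobile touch button widget for Pilgrim Quickplay that shows an emote prompt during breaks and plays a quick emote on touch.
# PlayFadeAnimation reacts to break-state changes, fading the button in or out through a FortSimpleWidgetAnimationsPanel and collapsing the widget after FadeOut.
# It starts collapsed and transparent, and clears the button's emote press action in PortraitSensor orientation to block the emote wheel.
# Touches call PlayQuickEmote on the local player's Pilgrim controller component.
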WidgetBlueprint /PilgrimQuickplay/Widgets/Mobile/WBP_Pilgrim_TouchButton_Emote
graphs [
 Block EmoteWheel If Portrait {nodes [
   Block EmoteWheel If Portrait() {is=title K2Node_FunctionEntry}
   > Branch(Condition="Equal (Enum)"(
     A:EScreenOrientation Enum=GetAllowedDeviceOrientation!() -> (ReturnValue:EScreenOrientation Enum=Unknown)
     B:EScreenOrientation Enum=PortraitSensor
    ))
   > Set PressAction!(
    execute
    PressAction:Input Action Object Reference
    self:Pilgrim Touch Button Base Object Reference=Get PilgrimTouchButton_L_EmotePrompt() -> (PilgrimTouchButton_L_EmotePrompt:Pilgrim Touch Button L Object Reference)
   ) -> (then, Output_Get:Input Action Object Reference)
  ]}
 EventGraph {nodes [
   "On Pilgrim Button Touch Start (PilgrimTouchButton_L_EmotePrompt)"() {is=title K2Node_ComponentBoundEvent}
   > PlayQuickEmote(self:Pilgrim Player Controller Component Object Reference=GetComponentByClass!(
     self:Actor Object Reference=GetFirstLocalFortPlayerController!(bIncludeNonSpectators:bool=true, bIncludeSpectators:bool=false) -> (ReturnValue:Fort Player Controller Object Reference)
     ComponentClass:Actor Component Class Reference=/PilgrimCore/Player/Components/BP_PilgrimPlayerControllerComponent.BP_PilgrimPlayerControllerComponent_C
    ) -> (ReturnValue:BP Pilgrim Player Controller Component Object Reference))
   n1 = "On Animations Finished (AnimationPanel)"!() -> (OutputDelegate, then, Name:Name=None) {is=title K2Node_ComponentBoundEvent}
   > Branch(Condition="Equal (Name)"(A=@n1.Name, B=FadeOut))
   > SetVisibility(InVisibility=Collapsed)
   Event Construct()
   > Parent: Construct()
   > Block EmoteWheel If Portrait() {is=title K2Node_CallFunction}
   > SetVisibility(InVisibility=Collapsed)
   > SetRenderOpacity(self=Get PilgrimTouchButton_L_EmotePrompt(), InOpacity=0)
  ]}
 PlayFadeAnimation {nodes [
   n1 = PlayFadeAnimation!() -> (then, InBreak:bool=~) {is=title K2Node_FunctionEntry}
   > Branch(Condition="Not Equal (Boolean)"(A=@n1, B=Get IsInBreak() -> (IsInBreak:bool)) -> (ReturnValue=~))
   > n2 = Set IsInBreak(IsInBreak:bool=Get InBreak() -> (InBreak:bool)) -> (Output_Get:bool)
   > PlayAnimationByName!(
    execute
    self:Fort Simple Widget Animations Panel Object Reference=Get AnimationPanel() -> (AnimationPanel:Fort Simple Widget Animations Panel Object Reference)
    Name:"Name (by ref)"=SelectName!(A:Name=FadeIn, B:Name=FadeOut, bPickA:bool=@n2) -> (ReturnValue:Name=None)
    Speed:float=1
    StartAtTime:float=0
   ) -> (then, ReturnValue:bool)
   > Branch(Condition=Get IsInBreak())
   > SetVisibility(InVisibility=SelfHitTestInvisible)
  ]}
]
variables [IsInBreak {type=boolean}]
widgets {
 info {
  inheritedWidgetCount=0
  parentClass=&CommonUserWidget
  rootWidgetClass=/Script/FortniteUI.FortSimpleWidgetAnimationsPanel
  widgetCount=2
 }
 widgets [
  AnimationPanel {
   widgetClassPath=&FortSimpleWidgetAnimationsPanel
   bIsVariable=true
   widget=AnimationPanel
   properties {animations [
     FadeIn {
      bIsEnabled=true
      onAnimationsFinished=()
      targets [{
        alpha=/PilgrimCore/Widgets/Curves/CV_1sFadeIn
        bResetOnFinish=false
        bShouldAnimateAlpha=true
        bShouldAnimateScale=false
        bShouldAnimateTranslation=false
        scale=null
        targetName=PilgrimTouchButton_L_EmotePrompt
        translation=null
        materialScalarAnimations []
       }]
     }
     FadeOut {
      bIsEnabled=true
      onAnimationsFinished=()
      targets [{
        alpha=/PilgrimCore/Widgets/Curves/CV_1sFadeOut
        bResetOnFinish=false
        bShouldAnimateAlpha=true
        bShouldAnimateScale=false
        bShouldAnimateTranslation=false
        scale=null
        targetName=PilgrimTouchButton_L_EmotePrompt
        translation=null
        materialScalarAnimations []
       }]
     }
    ]}
  }
  PilgrimTouchButton_L_EmotePrompt {
   widgetClassPath=&/PilgrimTouchButton_L
   bIsVariable=true
   parent=AnimationPanel
   widget=PilgrimTouchButton_L_EmotePrompt
   properties {
    bConsumePointerInput=true
    pressAction=/PilgrimCore/Input/InputActions/IA_PilgrimEmote
    widgetTransitionType=DoNothing
    replace Slate Brush {
     bIsDynamicallyLoaded=false
     drawAs=Image
     imageType=FullColor
     mirroring=NoMirror
     resourceName=None
     resourceObject=/Game/Athena/HUD/Art/BacchusHudIcons/T_UI_Emote
     tiling=NoTile
     imageSize=<96 96>
     margin=<ltrb 0 0 0 0>
     outlineSettings {
      bUseBrushTransparency=false
      roundingType=HalfHeightRadius
      width=0
      color {colorUseRule=UseColor_Specified, specifiedColor=<rgba 0 0 0 0>}
      cornerRadii=<0 0 0 1>
     }
     tintColor {colorUseRule=UseColor_Specified, specifiedColor=<rgba 1 1 1 1>}
     uVRegion {bIsValid=false, max=<0 0>, min=<0 0>}
    }
   }
   slot {class=FortSimpleWidgetAnimationsSlot}
  }
 ]
}
