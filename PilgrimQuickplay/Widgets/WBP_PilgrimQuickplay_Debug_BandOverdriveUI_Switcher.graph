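# Debug widget named as a band Overdrive UI switcher for Pilgrim Quickplay, though it contains no switching logic.
# An Overlay root holds a single Overdrive UI display widget, labeled VersionB, stretched to fill it.
# The event graph is inert: PreConstruct, Construct and Tick are empty, and a SetVisibility(Collapsed) call on itself is left unconnected.
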WidgetBlueprint /PilgrimQuickplay/Widgets/WBP_PilgrimQuickplay_Debug_BandOverdriveUI_Switcher
graphs [EventGraph {nodes [
   SetVisibility(self=Self-Reference(), InVisibility=Collapsed)
   Event PreConstruct()
   Event Construct()
   Event Tick()
  ]}]
variables []
widgets {
 info {inheritedWidgetCount=0, parentClass=&UserWidget, rootWidgetClass=&Overlay, widgetCount=2}
 widgets [
  Overlay {widgetClassPath=&Overlay, widget=Overlay}
  WBP_PilgrimQuickplay_OverdriveUI_VersionB {
   widgetClassPath=&/WBP_PilgrimQuickplay_OverdriveUI_Display
   parent=Overlay
   widget=WBP_PilgrimQuickplay_OverdriveUI_VersionB
   slot {class=OverlaySlot, properties {horizontalAlignment=HAlign_Fill, verticalAlignment=VAlign_Fill}}
  }
 ]
}
