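# Song ident widget for Pilgrim Quickplay that briefly shows an artist name and song title over a gradient backing.
# On Construct it plays A_ShowIdentInfo, which slides and fades in backing, artist and title in a stagger, holds, then fades out and collapses the overlay.
# PreConstruct fills short or long sample text at design time only, chosen by preview booleans, to test wrapping.
# Its graph sets no runtime text.
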WidgetBlueprint /PilgrimQuickplay/Widgets/WBP_SongIdent
animations [A_ShowIdentInfo {
  binding_count=4
  length_seconds=3.65
  object=&self:A_ShowIdentInfo
  sequence {bindings [
    SongTitle_TextBlock {
     id=#1
     possessed_class=UEFN_TextBlock_C
     tracks [
      {
       type=MovieScene2DTransformTrack
       sections [{
         type=MovieScene2DTransformSection
         end_frame=73
         start_frame=0
         channels [
          Translation.X {type=MovieSceneScriptingFloatChannel, keys [[0,-36],[4,-36],[14,0],[64,0],[73,-36]]}
          Translation.Y {type=MovieSceneScriptingFloatChannel}
         ]
        }]
      }
      {
       type=MovieSceneFloatTrack
       display_name=Render Opacity
       sections [{
         type=MovieSceneFloatSection
         end_frame=73
         start_frame=0
         channels [None {type=MovieSceneScriptingFloatChannel, keys [[0,0],[4,0],[14,1],[64,1],[72,0],[73,0]]}]
        }]
      }
     ]
    }
    ArtistName_TextBlock {
     id=#2
     possessed_class=UEFN_TextBlock_C
     tracks [
      {
       type=MovieScene2DTransformTrack
       sections [{
         type=MovieScene2DTransformSection
         end_frame=73
         start_frame=0
         channels [
          Translation.X {type=MovieSceneScriptingFloatChannel, keys [[0,-36],[2,-36],[12,0],[62,0],[70,-36],[73,-36]]}
          Translation.Y {type=MovieSceneScriptingFloatChannel}
         ]
        }]
      }
      {
       type=MovieSceneFloatTrack
       display_name=Render Opacity
       sections [{
         type=MovieSceneFloatSection
         end_frame=73
         start_frame=0
         channels [None {type=MovieSceneScriptingFloatChannel, keys [[0,0],[2,0],[12,1],[62,1],[70,0],[73,0]]}]
        }]
      }
     ]
    }
    Border_89 {
     id=#3
     possessed_class=Border
     tracks [
      {
       type=MovieScene2DTransformTrack
       sections [{
         type=MovieScene2DTransformSection
         end_frame=73
         start_frame=0
         channels [
          Translation.X {type=MovieSceneScriptingFloatChannel, keys [[0,-36],[10,0],[60,0],[67,-36],[73,-36]]}
          Translation.Y {type=MovieSceneScriptingFloatChannel}
         ]
        }]
      }
      {
       type=MovieSceneFloatTrack
       display_name=Render Opacity
       sections [{
         type=MovieSceneFloatSection
         end_frame=73
         start_frame=0
         channels [None {type=MovieSceneScriptingFloatChannel, keys [[0,0],[10,1],[60,1],[67,0],[73,0]]}]
        }]
      }
     ]
    }
    Overlay_20 {
     id=#4
     possessed_class=Overlay
     tracks [
      {
       type=MovieSceneByteTrack
       display_name=Visibility
       sections [{type=MovieSceneByteSection, end_frame=73, start_frame=0, channels [None {type=MovieSceneScriptingByteChannel, keys [[0,4],[72,1],[73,1]]}]}]
      }
      {
       type=MovieScene2DTransformTrack
       sections [{
         type=MovieScene2DTransformSection
         end_frame=73
         start_frame=0
         channels [
          Translation.X {type=MovieSceneScriptingFloatChannel, keys [[0,0],[60,10],[73,0]]}
          Translation.Y {type=MovieSceneScriptingFloatChannel}
          Angle {type=MovieSceneScriptingFloatChannel}
          Scale.X {type=MovieSceneScriptingFloatChannel}
          Scale.Y {type=MovieSceneScriptingFloatChannel}
          Shear.X {type=MovieSceneScriptingFloatChannel}
          Shear.Y {type=MovieSceneScriptingFloatChannel}
         ]
        }]
      }
     ]
    }
   ]}
 }]
graphs [EventGraph {nodes [
   n1 = Event PreConstruct()
   > Branch(Condition=@n1.IsDesignTime)
   > "SetText (Text)"(
    self=Get ArtistName_TextBlock() -> (ArtistName_TextBlock:Text Block Object Reference)
    InText=SelectText(
     A="LOCTEXT(\"Blur\")"
     B="LOCTEXT(\"John Williams && The London Symphony Orchestra\")"
     bPickA=Get Preview ShortOrLong ArtistName() -> (Preview ShortOrLong ArtistName:bool)
    )
   )
   > "SetText (Text)"(
    self=Get SongTitle_TextBlock() -> (SongTitle_TextBlock:Text Block Object Reference)
    InText=SelectText(
     A="LOCTEXT(\"Song 2\")"
     B="LOCTEXT(\"Star Wars Main Title Theme/March of the Resistance (Fortnite Rearrangement)\")"
     bPickA=Get Preview ShortOrLong SongTitle() -> (Preview ShortOrLong SongTitle:bool)
    )
   )
   Event Construct()
   > PlayAnimation(InAnimation=Get A_ShowIdentInfo() -> (A_ShowIdentInfo:Widget Animation Object Reference))
  ]}]
variables [
 Event Router {type=GameplayEventRouterComponent, default=null}
 Preview ShortOrLong ArtistName {type=boolean}
 Preview ShortOrLong SongTitle {type=boolean}
]
widgets {
 info {inheritedWidgetCount=0, parentClass=&UserWidget, rootWidgetClass=&Overlay, widgetCount=7}
 widgets [
  Overlay_20 {widgetClassPath=&Overlay, widget=Overlay_20, properties {pixelSnapping=Disabled, visibility=HitTestInvisible}}
  SafeZone_0 {
   widgetClassPath=&SafeZone
   bIsVariable=true
   parent=Overlay_20
   widget=SafeZone_0
   properties {padBottom=false, padLeft=false, padRight=false, padTop=false}
   slot {class=OverlaySlot, properties {horizontalAlignment=HAlign_Fill, verticalAlignment=VAlign_Fill, padding=<ltrb -160 0 0 0>}}
  }
  Border_89 {
   widgetClassPath=&Border
   parent=SafeZone_0
   widget=Border_89
   properties {
    background {
     bIsDynamicallyLoaded=false
     drawAs=Image
     imageType=NoImage
     mirroring=NoMirror
     resourceName=None
     resourceObject=/Game/UI/Inventory/Art/BasicLinearUIGradient_Cr_ContentBrowserTileEdge_Horz
     tiling=NoTile
     imageSize=<32 32>
     margin=<ltrb 0 0 0 0>
     outlineSettings {
      bUseBrushTransparency=false
      roundingType=HalfHeightRadius
      width=0
      color {colorUseRule=UseColor_Specified, specifiedColor=<rgba 0 0 0 0>}
      cornerRadii=<0 0 0 0>
     }
     tintColor {colorUseRule=UseColor_Specified, specifiedColor=<rgba 1 1 1 1>}
     uVRegion {bIsValid=false, max=<0 0>, min=<0 0>}
    }
    brushColor=<rgba 0.0017 0 0.005 0.6>
    padding=<ltrb 0 0 0 0>
   }
   slot {class=SafeZoneSlot, properties {bIsTitleSafe=false}}
  }
  SizeBox_203 {
   widgetClassPath=&SizeBox
   parent=Overlay_20
   widget=SizeBox_203
   properties {bOverride_MaxDesiredWidth=true, bOverride_MinDesiredWidth=true, maxDesiredWidth=570, minDesiredWidth=384}
   slot {class=OverlaySlot, properties {horizontalAlignment=HAlign_Fill}}
  }
  VerticalBox_0 {widgetClassPath=&VerticalBox, parent=SizeBox_203, widget=VerticalBox_0, slot {class=SizeBoxSlot}}
  ArtistName_TextBlock {
   widgetClassPath=&UEFN_TextBlock
   bIsVariable=true
   parent=VerticalBox_0
   widget=ArtistName_TextBlock
   properties {
    autoWrapText=true
    text=ArtistName
    visibility=HitTestInvisible
    font {
     bForceMonospaced=false
     bMaterialIsStencil=false
     fontMaterial=null
     fontObject=&HeadingNow
     letterSpacing=0
     monospacedWidth=1
     size=16.5
     skewAmount=0
     typefaceFontName=75 Medium Italic
     outlineSettings {
      bApplyOutlineToDropShadows=false
      bMiteredCorners=false
      bSeparateFillAlpha=false
      outlineMaterial=null
      outlineSize=0
      outlineColor=<rgba 0 0 0 1>
     }
    }
   }
   slot {class=VerticalBoxSlot, properties {verticalAlignment=VAlign_Center, padding=<ltrb 0 8 24 0>}}
  }
  SongTitle_TextBlock {
   widgetClassPath=&UEFN_TextBlock
   bIsVariable=true
   parent=VerticalBox_0
   widget=SongTitle_TextBlock
   properties {
    autoWrapText=true
    lineHeightPercentage=0.8
    text=SongTitle
    visibility=HitTestInvisible
    font {
     bForceMonospaced=false
     bMaterialIsStencil=false
     fontMaterial=null
     fontObject=&HeadingNow
     letterSpacing=0
     monospacedWidth=1
     size=24
     skewAmount=0
     typefaceFontName=86 Bold Italic
     outlineSettings {
      bApplyOutlineToDropShadows=false
      bMiteredCorners=false
      bSeparateFillAlpha=false
      outlineMaterial=null
      outlineSize=0
      outlineColor=<rgba 0 0 0 1>
     }
    }
   }
   slot {class=VerticalBoxSlot, properties {verticalAlignment=VAlign_Center, padding=<ltrb 0 2 24 14>}}
  }
 ]
}
