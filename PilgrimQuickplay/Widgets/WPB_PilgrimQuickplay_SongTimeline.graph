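# Pilgrim Quickplay song timeline widget showing artist, song title, a duration readout and a progress bar with playhead.
# SongUpdated and ResetForNewPlayer reset the child timeline bar; PreConstruct sets its width and fills short or long preview text at design time.
# Construct collapses the widget, plays A_Show and subscribes to visibility changes by gameplay tag.
# ToggleTimeline plays A_Show forward or in reverse, and animation completion collapses the widget if no song is playing.
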WidgetBlueprint /PilgrimQuickplay/Widgets/WPB_PilgrimQuickplay_SongTimeline
animations [A_Show {
  binding_count=2
  length_seconds=4.75
  object=&self:A_Show
  sequence {bindings [
    VerticalBox {
     id=#1
     possessed_class=VerticalBox
     tracks [{
       type=MovieSceneByteTrack
       display_name=Visibility
       sections [{type=MovieSceneByteSection, channels [None {type=MovieSceneScriptingByteChannel, keys [[0,1],[80,4]]}]}]
      }]
    }
    WPB_PilgrimQuickplay_SongTimeline_C_20 {
     id=#2
     possessed_class=WPB_PilgrimQuickplay_SongTimeline_C
     tracks [{
       type=MovieSceneFloatTrack
       display_name=Render Opacity
       sections [{type=MovieSceneFloatSection, channels [None {type=MovieSceneScriptingFloatChannel, keys [[80,0],[95,1]]}]}]
      }]
    }
   ]}
 }]
graphs [EventGraph {nodes [
   n1 = SongUpdated!() -> (OutputDelegate, then, NewSong:Pilgrim Song Metadata Object Reference) {is=title K2Node_CustomEvent}
   ResetForNewPlayerOrSong(
    execute=[@n1.then, @n2.then]
    self:WPB Pilgrim Quickplay Song Timeline Bar Object Reference=Get PilgrimQuickplay_SongTimelineBar() -> (PilgrimQuickplay_SongTimelineBar:WPB Pilgrim Quickplay Song Timeline Bar Object Reference)
   )
   n2 = ResetForNewPlayer() {is=title K2Node_CustomEvent}
   n3 = ToggleTimeline!() -> (OutputDelegate, then, Visible:bool=~) {is=title K2Node_CustomEvent}
   > n4 = Branch(Condition=@n3.Visible)
   > SetVisibility(self=Get Overlay_Main() -> (Overlay_Main:Overlay Object Reference), InVisibility=SelfHitTestInvisible)
   > PlayAnimationForward(InAnimation=Get A_Show() -> (A_Show:Widget Animation Object Reference))
   PlayAnimationReverse(execute=@n4.else, InAnimation=Get A_Show())
   Event Construct()
   > SetVisibility(self=Get Overlay_Main(), InVisibility=Collapsed)
   > PlayAnimationForward(InAnimation=Get A_Show())
   > SubscribeForChangingVisibilityByTag!(
    execute
    self:Pilgrim Game Widget Base Object Reference
    GameplayTag:"Gameplay Tag Structure (by ref)"=MakeLiteralGameplayTag!(Value:Gameplay Tag Structure=(TagName="Pilgrim.Widget.SongTimeline")) -> (ReturnValue:Gameplay Tag Structure)
   ) -> (then)
   n5 = Event PreConstruct()
   > SetBarWidth!(
    execute
    self:WPB Pilgrim Quickplay Song Timeline Bar Object Reference=Get PilgrimQuickplay_SongTimelineBar()
    Width:double=Get BarWidth() -> (BarWidth:float)
   ) -> (then)
   > Branch(Condition=@n5.IsDesignTime)
   > "SetText (Text)"(
    self=Get ArtistName() -> (ArtistName:Common Text Object Reference)
    InText=SelectText(
     A="LOCTEXT(\"Blur\")"
     B="LOCTEXT(\"John Williams && The London Symphony Orchestra\")"
     bPickA=Get Preview ShortOrLong ArtistName() -> (Preview ShortOrLong ArtistName:bool)
    )
   )
   > "SetText (Text)"(
    self=Get SongName() -> (SongName:Common Text Object Reference)
    InText=SelectText(
     A="LOCTEXT(\"Song 2\")"
     B="LOCTEXT(\"Star Wars Main Title Theme/March of the Resistance (Fortnite Rearrangement)\")"
     bPickA=Get Preview ShortOrLong SongTitle() -> (Preview ShortOrLong SongTitle:bool)
    )
   )
   "Animation Finished (A_Show)"!() -> (OutputDelegate:Delegate, then:Exec) {is=title K2Node_WidgetAnimationEvent}
   > n6 = Branch(Condition=Get SongPlaying!(self:Pilgrim Currently Playing Song View Model Object Reference=Get PilgrimCurrentlyPlayingSongVM() -> (PilgrimCurrentlyPlayingSongVM:Pilgrim Currently Playing Song View Model Object Reference)) -> (SongPlaying:bool))
   SetVisibility(execute=@n6.else, self=Get Overlay_Main(), InVisibility=Collapsed)
  ]}]
variables [
 BarWidth {type=number, default=370}
 Preview ShortOrLong ArtistName {type=boolean, default=true}
 Preview ShortOrLong SongTitle {type=boolean, default=true}
 ThumbIconDynamicMat {type=MaterialInstanceDynamic}
]
widgets {
 info {inheritedWidgetCount=0, parentClass=/Script/PilgrimCoreUI.PilgrimGameWidgetBase, rootWidgetClass=&Overlay, widgetCount=13}
 widgets [
  Overlay_Main {widgetClassPath=&Overlay, bIsVariable=true, widget=Overlay_Main, properties {visibility=Collapsed}}
  ScaleBox_0 {widgetClassPath=&ScaleBox, parent=Overlay_Main, widget=ScaleBox_0, slot {class=OverlaySlot}}
  VerticalBox {widgetClassPath=&VerticalBox, parent=ScaleBox_0, widget=VerticalBox, slot {class=ScaleBoxSlot}}
  SizeBox_MaxInfoWidth {
   widgetClassPath=&SizeBox
   parent=VerticalBox
   widget=SizeBox_MaxInfoWidth
   properties {bOverride_MaxDesiredWidth=true, maxDesiredWidth=374}
   slot {class=VerticalBoxSlot, properties {padding=<ltrb 0 0 0 10>, size {sizeRule=Automatic, value=4}}}
  }
  Border_SongInfo {
   widgetClassPath=&CommonBorder
   parent=SizeBox_MaxInfoWidth
   widget=Border_SongInfo
   properties {
    background {
     bIsDynamicallyLoaded=false
     drawAs=Image
     imageType=FullColor
     mirroring=NoMirror
     resourceName=None
     resourceObject=/PilgrimQuickplay/Widgets/Materials/MI_UI_TimelineBar
     tiling=NoTile
     imageSize=<32 32>
     margin=<ltrb 0 0 0 0>
     outlineSettings {
      bUseBrushTransparency=false
      roundingType=HalfHeightRadius
      width=0
      color {colorUseRule=UseColor_Specified, specifiedColor=<rgba 0 0 0 0>}
      cornerRadii=<0 0 0 0>
     }
     tintColor {colorUseRule=UseColor_Specified, specifiedColor=<rgba 1 1 1 0.45>}
     uVRegion {bIsValid=false, max=<0 0>, min=<0 0>}
    }
    padding=<ltrb 15 12 15 6>
   }
   slot {class=SizeBoxSlot}
  }
  VerticalBox_158 {
   widgetClassPath=&VerticalBox
   parent=Border_SongInfo
   widget=VerticalBox_158
   slot {class=BorderSlot, properties {padding=<ltrb 15 12 15 6>}}
  }
  ArtistName {
   widgetClassPath=&CommonTextBlock
   bIsVariable=true
   parent=VerticalBox_158
   widget=ArtistName
   properties {
    clipping=OnDemand
    scrollStyle=&TextScrollStyle-Base_C
    text=Artist Name
    textOverflowPolicy=Ellipsis
    font {
     bForceMonospaced=false
     bMaterialIsStencil=false
     fontMaterial=null
     fontObject=&HeadingNow
     letterSpacing=0
     monospacedWidth=1
     size=12
     skewAmount=0
     typefaceFontName=75 Medium
     outlineSettings {
      bApplyOutlineToDropShadows=false
      bMiteredCorners=false
      bSeparateFillAlpha=false
      outlineMaterial=null
      outlineSize=0
      outlineColor=<rgba 0 0 0 1>
     }
    }
    shadowOffset=<0 0>
   }
   slot {class=VerticalBoxSlot, properties {size {sizeRule=Fill, value=1}}}
  }
  SongName {
   widgetClassPath=&CommonTextBlock
   bIsVariable=true
   parent=VerticalBox_158
   widget=SongName
   properties {
    clipping=OnDemand
    scrollStyle=&TextScrollStyle-Base_C
    text=Song Name
    textOverflowPolicy=Ellipsis
    font {
     bForceMonospaced=false
     bMaterialIsStencil=false
     fontMaterial=null
     fontObject=&HeadingNow
     letterSpacing=0
     monospacedWidth=1
     size=15
     skewAmount=0
     typefaceFontName=75 Medium
     outlineSettings {
      bApplyOutlineToDropShadows=false
      bMiteredCorners=false
      bSeparateFillAlpha=false
      outlineMaterial=null
      outlineSize=0
      outlineColor=<rgba 0 0 0 1>
     }
    }
    margin=<ltrb 0 0 0 12>
    shadowOffset=<0 0>
   }
   slot {class=VerticalBoxSlot, properties {padding=<ltrb 0 -1 0 0>, size {sizeRule=Fill, value=1}}}
  }
  Overlay_81 {
   widgetClassPath=&Overlay
   parent=VerticalBox
   widget=Overlay_81
   slot {class=VerticalBoxSlot, properties {verticalAlignment=VAlign_Center}}
  }
  PilgrimQuickplay_SongTimelineBar {
   widgetClassPath=&/WPB_PilgrimQuickplay_SongTimelineBar
   bIsVariable=true
   parent=Overlay_81
   widget=PilgrimQuickplay_SongTimelineBar
   properties {breakBarHeight=0.4}
   slot {class=OverlaySlot, properties {horizontalAlignment=HAlign_Fill, verticalAlignment=VAlign_Center, padding=<ltrb 2 0 2 0>}}
  }
  PlayHead {
   widgetClassPath=&Image
   bIsVariable=true
   parent=Overlay_81
   widget=PlayHead
   properties {
    pixelSnapping=Disabled
    visibility=SelfHitTestInvisible
    brush {
     bIsDynamicallyLoaded=false
     drawAs=Image
     imageType=FullColor
     mirroring=NoMirror
     resourceName=None
     resourceObject=/PilgrimQuickplay/Widgets/Materials/MI_UI_ProgressLine
     tiling=NoTile
     imageSize=<370 18>
     margin=<ltrb 0 0 0 0>
     outlineSettings {
      bUseBrushTransparency=false
      roundingType=HalfHeightRadius
      width=0
      color {colorUseRule=UseColor_Specified, specifiedColor=<rgba 0 0 0 0>}
      cornerRadii=<0 0 0 0>
     }
     tintColor {colorUseRule=UseColor_Specified, specifiedColor=<rgba 1 1 1 1>}
     uVRegion {bIsValid=false, max=<0 0>, min=<0 0>}
    }
   }
   slot {class=OverlaySlot, properties {horizontalAlignment=HAlign_Fill, verticalAlignment=VAlign_Center}}
  }
  Border_149 {
   widgetClassPath=&Border
   parent=VerticalBox
   widget=Border_149
   properties {
    horizontalAlignment=HAlign_Right
    verticalAlignment=VAlign_Center
    background {
     bIsDynamicallyLoaded=false
     drawAs=Image
     imageType=FullColor
     mirroring=NoMirror
     resourceName=None
     resourceObject=&MI_UI_TimelineBar
     tiling=NoTile
     imageSize=<32 32>
     margin=<ltrb 0 0 0 0>
     outlineSettings {
      bUseBrushTransparency=false
      roundingType=HalfHeightRadius
      width=0
      color {colorUseRule=UseColor_Specified, specifiedColor=<rgba 0 0 0 0>}
      cornerRadii=<0 0 0 0>
     }
     tintColor {colorUseRule=UseColor_Specified, specifiedColor=<rgba 1 1 1 1>}
     uVRegion {bIsValid=false, max=<0 0>, min=<0 0>}
    }
    brushColor=<rgba 1 1 1 0>
    padding=<ltrb 2 4 2 4>
    renderTransformPivot=<0 0.5>
   }
   slot {class=VerticalBoxSlot, properties {verticalAlignment=VAlign_Bottom, size {sizeRule=Fill, value=1}}}
  }
  SongDuration {
   widgetClassPath=&CommonTextBlock
   bIsVariable=true
   parent=Border_149
   widget=SongDuration
   properties {
    clipping=OnDemand
    text=-03:16
    textTransformPolicy=ToUpper
    font {
     bForceMonospaced=false
     bMaterialIsStencil=false
     fontMaterial=null
     fontObject=&HeadingNow
     letterSpacing=0
     monospacedWidth=1
     size=12
     skewAmount=0
     typefaceFontName=75 Medium Italic
     outlineSettings {
      bApplyOutlineToDropShadows=false
      bMiteredCorners=false
      bSeparateFillAlpha=false
      outlineMaterial=null
      outlineSize=0
      outlineColor=<rgba 0 0 0 1>
     }
    }
    renderTransformPivot=<0 0.5>
    shadowOffset=<0 0>
   }
   slot {class=BorderSlot, properties {horizontalAlignment=HAlign_Right, verticalAlignment=VAlign_Center, padding=<ltrb 2 4 2 4>}}
  }
 ]
}
