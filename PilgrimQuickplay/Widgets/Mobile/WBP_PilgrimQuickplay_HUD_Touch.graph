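# Touch HUD widget for Pilgrim Quickplay, laying out scores, song timeline and ident, pinned quests, hamburger menus, variety multiplier and band overdrive debug switcher.
# A platform visibility wrapper keeps the emote button off console, Mac, gamepad and mouse/keyboard.
# On Construct, it awaits the playspace's event router, then hides the hamburger menus while the quest screen is visible.
# A fade-out animation for the multiplier exists, but this graph never plays it.
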WidgetBlueprint /PilgrimQuickplay/Widgets/Mobile/WBP_PilgrimQuickplay_HUD_Touch
animations [InstrumentVarietyMultiplier_FadeOut {
  binding_count=1
  length_seconds=5
  object=&self:InstrumentVarietyMultiplier_FadeOut
  sequence {bindings [WBP_PilgrimQuickplay_InstrumentVarietyMultiplier {
     id=#1
     possessed_class=WBP_PilgrimQuickplay_InstrumentVarietyMultiplier_C
     tracks [{
       type=MovieSceneFloatTrack
       display_name=Render Opacity
       sections [{
         type=MovieSceneFloatSection
         end_frame=100
         start_frame=0
         channels [None {type=MovieSceneScriptingFloatChannel, keys [[0,1],[80,1],[100,0]]}]
        }]
      }]
    }]}
 }]
graphs [EventGraph {nodes [
   Event Construct()
   > n1 = PilgrimWaitForEventRouter!(
    execute:Exec
    ContextActor:Actor Object Reference=GetSparksMusicPlayspaceOfClass!(
     PlayspaceType:Sparks Music Playspace Class Reference=/Script/PilgrimQuickplayRuntime.PilgrimQuickplayPlayspace
     Actor:Actor Object Reference=GetOwningPlayer()
     bMatchChildTypes:bool=false
    ) -> (ReturnValue:Pilgrim Quickplay Playspace Object Reference)
   ) -> (then:Exec, OnEventRouterAvailable:Exec, EventRouter:Gameplay Event Router Component Object Reference) {is=title K2Node_AsyncAction}
   n2 = StartListeningToEvent!(
    execute:Exec=@n1.OnEventRouterAvailable
    Target:Gameplay Event Router Component Object Reference=@n1
    EventType:Script Struct Object Reference=/Script/SparksCoreUI.SparksEvent_QuestScreenVisibilityChanged
    Context:Object Reference
    EventBubblingRule:EEventBubblingRule Enum
   ) -> (
    then:Exec
    AsyncTaskProxy:Async Action Start Listening To Event Object Reference
    OnEventReceived:Exec
    Payload:Sparks Event Quest Screen Visibility Changed Structure
   )
   SetVisibility(
    execute=@n2.OnEventReceived
    self=Get WBP_PilgrimTouch_HamburgerMenus() -> (WBP_PilgrimTouch_HamburgerMenus:WBP Pilgrim Touch Hamburger Menus Object Reference)
    InVisibility=Select(
     Option 0=SelfHitTestInvisible
     Option 1=Hidden
     Index=Break Sparks Event Quest Screen Visibility Changed!(SparksEvent_QuestScreenVisibilityChanged:"Sparks Event Quest Screen Visibility Changed Structure (by ref)"=@n2.Payload) -> (bIsVisible:bool=~) {is=title K2Node_BreakStruct}
    )
   )
  ]}]
variables [
 EmotePromtTagAddedEventHandle {type=AbilityAsync_WaitGameplayTagAdded, default=null}
 EmotePromtTagRemovedEventHandle_0 {type=AbilityAsync_WaitGameplayTagRemoved, default=null}
 EventRouter {type=GameplayEventRouterComponent, default=null}
 HamburgerButtonTagAddedEventHandle {type=AbilityAsync_WaitGameplayTagAdded, default=null}
 HamburgerButtonTagRemovedEventHandle {type=AbilityAsync_WaitGameplayTagRemoved, default=null}
 LabelGemMap {default {}}
 Local Music Track {type=BP_PilgrimTrack_C, default=null}
]
widgets {
 info {inheritedWidgetCount=0, parentClass=&UserWidget, rootWidgetClass=&Overlay, widgetCount=13}
 widgets [
  Overlay_44 {widgetClassPath=&Overlay, widget=Overlay_44}
  WBP_PilgrimQuickplay_Debug_BandOverdriveUI_Switcher {
   widgetClassPath=/PilgrimQuickplay/Widgets/WBP_PilgrimQuickplay_Debug_BandOverdriveUI_Switcher
   bIsVariable=true
   parent=Overlay_44
   widget=WBP_PilgrimQuickplay_Debug_BandOverdriveUI_Switcher
   properties {renderTransform {angle=0, scale=<1.5 1.5>, shear=<0 0>, translation=<0 0>}}
   slot {class=OverlaySlot, properties {horizontalAlignment=HAlign_Center, verticalAlignment=VAlign_Center, padding=<ltrb 0 0 0 300>}}
  }
  WBP_PilgrimQuickplay_InstrumentVarietyMultiplier {
   widgetClassPath=/PilgrimQuickplay/Widgets/WBP_PilgrimQuickplay_InstrumentVarietyMultiplier
   bIsVariable=true
   parent=Overlay_44
   widget=WBP_PilgrimQuickplay_InstrumentVarietyMultiplier
   slot {class=OverlaySlot, properties {horizontalAlignment=HAlign_Center, verticalAlignment=VAlign_Center, padding=<ltrb 0 0 0 300>}}
  }
  CommonVisibilityWidgetLegacy_0 {
   widgetClassPath=&CommonVisibilityWidgetLegacy
   parent=Overlay_44
   widget=CommonVisibilityWidgetLegacy_0
   properties {
    bShowForErebus=false
    bShowForGamepad=false
    bShowForMac=false
    bShowForMouseAndKeyboard=false
    bShowForPS4=false
    bShowForPS5=false
    bShowForSwitch2=false
    bShowForXBox=false
    bShowForXSX=false
    padding=<ltrb 4 2 4 124>
   }
   slot {class=OverlaySlot, properties {horizontalAlignment=HAlign_Center, verticalAlignment=VAlign_Bottom, padding=<ltrb 4 2 4 100>}}
  }
  PilgrimTouchButton_emote {
   widgetClassPath=&/WBP_Pilgrim_TouchButton_Emote
   bIsVariable=true
   parent=CommonVisibilityWidgetLegacy_0
   widget=PilgrimTouchButton_emote
   slot {class=BorderSlot, properties {padding=<ltrb 4 2 4 124>}}
  }
  VBox_RightSideInfo {
   widgetClassPath=&VerticalBox
   bIsVariable=true
   parent=Overlay_44
   widget=VBox_RightSideInfo
   slot {class=OverlaySlot, properties {padding=<ltrb 48 120 0 0>}}
  }
  WPB_PilgrimQuickplay_SongTimeline {
   widgetClassPath=/PilgrimQuickplay/Widgets/WPB_PilgrimQuickplay_SongTimeline
   bIsVariable=true
   parent=VBox_RightSideInfo
   widget=WPB_PilgrimQuickplay_SongTimeline
   properties {barWidth=470}
   slot {class=VerticalBoxSlot, properties {horizontalAlignment=HAlign_Left, verticalAlignment=VAlign_Bottom}}
  }
  WBP_PilgrimQuickplay_LocalScore_Mobile {
   widgetClassPath=&/WBP_PilgrimQuickplay_LocalScore_Mobile
   bIsVariable=true
   parent=VBox_RightSideInfo
   widget=WBP_PilgrimQuickplay_LocalScore_Mobile
   slot {class=VerticalBoxSlot}
  }
  WBP_SongIdent {
   widgetClassPath=/PilgrimQuickplay/Widgets/WBP_SongIdent
   bIsVariable=true
   parent=VBox_RightSideInfo
   widget=WBP_SongIdent
   slot {class=VerticalBoxSlot, properties {horizontalAlignment=HAlign_Left, verticalAlignment=VAlign_Top}}
  }
  WBP_PilgrimQuickplay_BandScore_Mobile {
   widgetClassPath=&/WBP_PilgrimQuickplay_BandScore_Mobile
   bIsVariable=true
   parent=Overlay_44
   widget=WBP_PilgrimQuickplay_BandScore_Mobile
   slot {class=OverlaySlot, properties {horizontalAlignment=HAlign_Right, padding=<ltrb 0 120 48 0>}}
  }
  WBP_PilgrimTouch_HamburgerMenus {
   widgetClassPath=&/WBP_PilgrimTouch_HamburgerMenus
   bIsVariable=true
   parent=Overlay_44
   widget=WBP_PilgrimTouch_HamburgerMenus
   slot {class=OverlaySlot, properties {horizontalAlignment=HAlign_Right, padding=<ltrb 0 0 15 0>}}
  }
  SizeBox_0 {
   widgetClassPath=&SizeBox
   parent=Overlay_44
   widget=SizeBox_0
   properties {bOverride_WidthOverride=true, widthOverride=472.33}
   slot {class=OverlaySlot, properties {padding=<ltrb 48 44 0 0>}}
  }
  PinnedQuestList {widgetClassPath=/Game/Athena/HUD/PinnedContentList, bIsVariable=true, parent=SizeBox_0, widget=PinnedQuestList, slot {class=SizeBoxSlot}}
 ]
}
