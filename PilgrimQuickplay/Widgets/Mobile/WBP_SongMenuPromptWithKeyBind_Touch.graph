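# Touch song-menu prompt for Pilgrim Quickplay, showing a song-menu icon and a keybind hint that is hidden on touch and several platforms.
# A transparent button over the prompt broadcasts a QuestScreenToggleEvent through the player's event router, cached asynchronously on Construct.
# Async tag listeners on PlayerState collapse the widget when Pilgrim.Widget.SongMenuPrompt is added and show it again when removed.
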WidgetBlueprint /PilgrimQuickplay/Widgets/Mobile/WBP_SongMenuPromptWithKeyBind_Touch
graphs [EventGraph {nodes [
   "On Pressed (SongMenuButton)"() {is=title K2Node_ComponentBoundEvent}
   > Broadcast Event!(
    execute
    self:Gameplay Event Router Component Object Reference=Get Event Router() -> (Event Router:Gameplay Event Router Component Object Reference)
    EventData:Quest Screen Toggle Event Structure=Make Quest Screen Toggle Event!() -> (QuestScreenToggleEvent:Quest Screen Toggle Event Structure) {is=title K2Node_MakeStruct}
    EventContext:Object Reference=GetOwningPlayer()
   ) -> (then, ReturnValue:bool)
   Event Construct()
   > n1 = PilgrimWaitForEventRouter!(execute:Exec, ContextActor:Actor Object Reference=GetOwningPlayer()) -> (then:Exec, OnEventRouterAvailable:Exec, EventRouter:Gameplay Event Router Component Object Reference) {is=title K2Node_AsyncAction}
   > n2 = Sequence()
   n3 = Get PlayerState!(self:Controller Object Reference=GetOwningPlayer()) -> (PlayerState:Player State Object Reference)
   > n4 = WaitGameplayTagAddToActor!(
    execute:Exec
    TargetActor:Actor Object Reference=@n3
    Tag:Gameplay Tag Structure=(TagName="Pilgrim.Widget.SongMenuPrompt")
    OnlyTriggerOnce:bool=false
   ) -> (then:Exec, AsyncTaskProxy:Ability Async Wait Gameplay Tag Added Object Reference, Added:Exec) {is=title K2Node_AsyncAction}
   > Set SongMenuPromptTagAddedEventHandle(SongMenuPromptTagAddedEventHandle:Ability Async Wait Gameplay Tag Added Object Reference=@n4) -> (Output_Get:Ability Async Wait Gameplay Tag Added Object Reference)
   SetVisibility(execute=@n4.Added, InVisibility=Collapsed)
   n5 = WaitGameplayTagRemoveFromActor!(
    execute:Exec=@n2.then_1
    TargetActor:Actor Object Reference=@n3
    Tag:Gameplay Tag Structure=(TagName="Pilgrim.Widget.SongMenuPrompt")
    OnlyTriggerOnce:bool=false
   ) -> (then:Exec, AsyncTaskProxy:Ability Async Wait Gameplay Tag Removed Object Reference, Removed:Exec) {is=title K2Node_AsyncAction}
   > Set SongMenuPromptTagRemovedEventHandle(SongMenuPromptTagRemovedEventHandle:Ability Async Wait Gameplay Tag Removed Object Reference=@n5) -> (Output_Get:Ability Async Wait Gameplay Tag Removed Object Reference)
   SetVisibility(execute=@n5.Removed, InVisibility=SelfHitTestInvisible)
   Set Event Router(execute=@n1.OnEventRouterAvailable, Event Router:Gameplay Event Router Component Object Reference=@n1) -> (Output_Get:Gameplay Event Router Component Object Reference)
  ]}]
variables [
 Event Router {type=GameplayEventRouterComponent, default=null}
 SongMenuPromptTagAddedEventHandle {type=AbilityAsync_WaitGameplayTagAdded, default=null}
 SongMenuPromptTagRemovedEventHandle {type=AbilityAsync_WaitGameplayTagRemoved, default=null}
]
widgets {
 info {inheritedWidgetCount=0, parentClass=&CommonUserWidget, rootWidgetClass=&ScaleBox, widgetCount=7}
 widgets [
  ScaleBox_0 {widgetClassPath=&ScaleBox, widget=ScaleBox_0, properties {stretch=UserSpecified}}
  SongMenuPrompt {
   widgetClassPath=&Overlay
   bIsVariable=true
   parent=ScaleBox_0
   widget=SongMenuPrompt
   slot {class=ScaleBoxSlot, properties {horizontalAlignment=HAlign_Fill, verticalAlignment=VAlign_Fill}}
  }
  HorizontalBox_0 {
   widgetClassPath=&HorizontalBox
   parent=SongMenuPrompt
   widget=HorizontalBox_0
   slot {class=OverlaySlot, properties {horizontalAlignment=HAlign_Fill, verticalAlignment=VAlign_Fill, padding=<ltrb 20 20 20 20>}}
  }
  CommonVisibilityWidgetLegacy_0 {
   widgetClassPath=&CommonVisibilityWidgetLegacy
   parent=HorizontalBox_0
   widget=CommonVisibilityWidgetLegacy_0
   properties {
    bShowForIOS=false
    bShowForMac=false
    bShowForPS4=false
    bShowForPS5=false
    bShowForTouch=false
    bShowForXBox=false
    bShowForXSX=false
    horizontalAlignment=HAlign_Center
    verticalAlignment=VAlign_Center
    padding=<ltrb 0 0 0 0>
   }
   slot {class=HorizontalBoxSlot}
  }
  KeybindWidget {
   widgetClassPath=/FortUILibrary/Input/KeybindWidget
   bIsVariable=true
   parent=CommonVisibilityWidgetLegacy_0
   widget=KeybindWidget
   properties {boundInputAction=/PilgrimCore/Input/InputActions/IA_PilgrimOpenSongMenu, boundKeyFallback=Gamepad_Special_Left}
   slot {class=BorderSlot, properties {horizontalAlignment=HAlign_Center, verticalAlignment=VAlign_Center, padding=<ltrb 0 0 0 0>}}
  }
  Image_Icon {
   widgetClassPath=&Image
   bIsVariable=true
   parent=HorizontalBox_0
   widget=Image_Icon
   properties {
    visibility=SelfHitTestInvisible
    brush {
     bIsDynamicallyLoaded=false
     drawAs=Image
     imageType=NoImage
     mirroring=NoMirror
     resourceName=None
     resourceObject=/FMCore/Textures/T_Icon_SongMenu
     tiling=NoTile
     imageSize=<84 84>
     margin=<ltrb 0 0 0 0>
     outlineSettings {
      bUseBrushTransparency=false
      roundingType=HalfHeightRadius
      width=0
      color {colorUseRule=UseColor_Specified, specifiedColor=<rgba 0 0 0 0>}
      cornerRadii=<0 0 0 0>
     }
     tintColor {colorUseRule=UseColor_Specified, specifiedColor=<rgba 1 1 1 1>}
     uVRegion {bIsValid=false, max=<0 0>, min=<0 0>}
    }
   }
   slot {class=HorizontalBoxSlot}
  }
  SongMenuButton {
   widgetClassPath=&Button
   parent=SongMenuPrompt
   widget=SongMenuButton
   properties {widgetStyle {
     clickedSlateSound {resourceObject=null}
     disabled {
      bIsDynamicallyLoaded=false
      drawAs=NoDrawType
      imageType=NoImage
      mirroring=NoMirror
      resourceName=None
      resourceObject=null
      tiling=NoTile
      imageSize=<0 0>
      margin=<ltrb 0 0 0 0>
      outlineSettings {
       bUseBrushTransparency=false
       roundingType=HalfHeightRadius
       width=0
       color {colorUseRule=UseColor_Specified, specifiedColor=<rgba 0 0 0 0>}
       cornerRadii=<0 0 0 0>
      }
      tintColor {colorUseRule=UseColor_Specified, specifiedColor=<rgba 1 1 1 1>}
      uVRegion {bIsValid=false, max=<0 0>, min=<0 0>}
     }
     disabledForeground {colorUseRule=UseColor_Specified, specifiedColor=<rgba 0.5271 0.5271 0.5271 1>}
     hovered {
      bIsDynamicallyLoaded=false
      drawAs=NoDrawType
      imageType=NoImage
      mirroring=NoMirror
      resourceName=None
      resourceObject=null
      tiling=NoTile
      imageSize=<32 32>
      margin=<ltrb 0 0 0 0>
      outlineSettings {
       bUseBrushTransparency=true
       roundingType=FixedRadius
       width=1
       color {colorUseRule=UseColor_Specified, specifiedColor=<rgba 0.7243 0.7243 0.7243 1>}
       cornerRadii=<4 4 4 4>
      }
      tintColor {colorUseRule=UseColor_Specified, specifiedColor=<rgba 0.7243 0.7243 0.7243 1>}
      uVRegion {bIsValid=false, max=<0 0>, min=<0 0>}
     }
     hoveredForeground {colorUseRule=UseColor_Specified, specifiedColor=<rgba 0 1 1 1>}
     hoveredSlateSound {resourceObject=null}
     normal {
      bIsDynamicallyLoaded=false
      drawAs=NoDrawType
      imageType=NoImage
      mirroring=NoMirror
      resourceName=None
      resourceObject=null
      tiling=NoTile
      imageSize=<32 32>
      margin=<ltrb 0 0 0 0>
      outlineSettings {
       bUseBrushTransparency=true
       roundingType=FixedRadius
       width=1
       color {colorUseRule=UseColor_Specified, specifiedColor=<rgba 0.6951 0.6951 0.6951 1>}
       cornerRadii=<4 4 4 4>
      }
      tintColor {colorUseRule=UseColor_Specified, specifiedColor=<rgba 0.4955 0.4955 0.4955 1>}
      uVRegion {bIsValid=false, max=<0 0>, min=<0 0>}
     }
     normalForeground {colorUseRule=UseColor_Specified, specifiedColor=<rgba 0 1 1 1>}
     normalPadding=<ltrb 12 1.5 12 1.5>
     pressed {
      bIsDynamicallyLoaded=false
      drawAs=NoDrawType
      imageType=NoImage
      mirroring=NoMirror
      resourceName=None
      resourceObject=null
      tiling=NoTile
      imageSize=<32 32>
      margin=<ltrb 0 0 0 0>
      outlineSettings {
       bUseBrushTransparency=true
       roundingType=FixedRadius
       width=1
       color {colorUseRule=UseColor_Specified, specifiedColor=<rgba 0.7243 0.7243 0.7243 1>}
       cornerRadii=<4 4 4 4>
      }
      tintColor {colorUseRule=UseColor_Specified, specifiedColor=<rgba 0.3843 0.3843 0.3843 1>}
      uVRegion {bIsValid=false, max=<0 0>, min=<0 0>}
     }
     pressedForeground {colorUseRule=UseColor_Specified, specifiedColor=<rgba 0 1 1 1>}
     pressedPadding=<ltrb 12 1.5 12 1.5>
     pressedSlateSound {resourceObject=null}
    }}
   slot {class=OverlaySlot, properties {horizontalAlignment=HAlign_Fill, verticalAlignment=VAlign_Fill}}
  }
 ]
}
