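# Green room button widget for Pilgrim Quickplay that prompts the player to choose a song or change the setlist.
# Gameplay event router listeners show it during pregame and results states, collapse its content on one timer state, and refresh text on setlist changes.
# The label depends on whether the local player has added songs.
# An async gameplay tag watcher toggles the keybind, whose visibility changes adjust text padding.
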WidgetBlueprint /PilgrimQuickplay/Widgets/WBP_Pilgrim_GreenRoom_ChooseASongUI
graphs [
 ChangeButtonBorder {nodes [
   n1 = ChangeButtonBorder!() -> (then, CornerRadius:float=~) {is=title K2Node_FunctionEntry}
   > n2 = GetDynamicMaterial(self=Get ReadyLayer() -> (ReadyLayer:Image Object Reference))
   > n3 = GetDynamicMaterial(self=Get ButtonImage() -> (ButtonImage:Image Object Reference))
   > SetScalarParameterValue(self=[@n2, @n3], ParameterName=Corner Radius, Value=@n1)
  ]}
 EventGraph {nodes [
   n1 = "On Visibility Changed (KeybindWidget_ChangeBindings_1)"!() -> (OutputDelegate, then, InVisibility:ESlateVisibility Enum=Visible) {is=title K2Node_ComponentBoundEvent}
   > SetPadding!(
    execute
    self:Overlay Slot Object Reference=SlotAsOverlaySlot(Widget=Get VerticalBox_Text() -> (VerticalBox_Text:Vertical Box Object Reference))
    InPadding_Left:float=Select!(
     Visible:float=90
     Collapsed:float=64
     Hidden:float=64
     HitTestInvisible:float=90
     SelfHitTestInvisible:float=90
     Index:ESlateVisibility Enum=@n1.InVisibility
    ) -> (ReturnValue:float)
    InPadding_Top:float=8
    InPadding_Right:float=64
    InPadding_Bottom:float=8
   ) -> (then)
   Event Construct()
   > Set LocalPlayerState(LocalPlayerState:Player State Object Reference=Get PlayerState!(self:Controller Object Reference=GetOwningPlayer()) -> (PlayerState:Player State Object Reference)) -> (Output_Get:Player State Object Reference)
   > n2 = Sequence()
   > n3 = WaitGameplayTagCountChangedOnActor!(
    execute:Exec
    TargetActor:Actor Object Reference=Get PlayerState(self=GetOwningPlayer())
    Tag:Gameplay Tag Structure=(TagName="Pilgrim.GreenRoom.Input.ToggleReadyUp")
   ) -> (then:Exec, AsyncTaskProxy:Ability Async Wait Gameplay Tag Count Changed Object Reference, TagCountChanged:Exec, TagCount:int) {is=title K2Node_AsyncAction}
   > Add(
    TargetArray:Array of Cancellable Async Action Object References=Get AsyncEventHandles() -> (AsyncEventHandles:Array of Cancellable Async Action Object References)
    NewItem:"Cancellable Async Action Object Reference (by ref)"=@n3
   ) {is=title K2Node_CallArrayFunction}
   n4 = GetOwningPlayer()
   n5 = HasMatchingGameplayTag!(
    execute:Exec=@n3.TagCountChanged
    self:Object Reference=GetAbilitySystemComponent!(Actor:Actor Object Reference=Get PlayerState(self=@n4)) -> (ReturnValue:Ability System Component Object Reference)
    TagToCheck:Gameplay Tag Structure=(TagName="Pilgrim.GreenRoom.Input.ToggleReadyUp")
   ) -> (then:Exec, ReturnValue:bool) {is=title K2Node_Message}
   n6 = Get KeybindWidget_ChangeBindings_1() -> (KeybindWidget_ChangeBindings_1:Keybind Widget Object Reference)
   > SetIsEnabled(self=@n6, bInIsEnabled=@n5)
   > SetVisibility(
    self=@n6
    InVisibility=Select!(Option 0:ESlateVisibility Enum=Collapsed, Option 1:ESlateVisibility Enum=SelfHitTestInvisible, Index:bool=@n5) -> (ReturnValue:ESlateVisibility Enum)
   )
   n7 = PlaylistDataReadyAsync!(execute:Exec=@n2.then_1) -> (
    then:Exec
    OnReady:Exec
    GameState:Fort Game State Athena Object Reference
    Playlist:Fort Playlist Object Reference
    PlaylistContextTags:"Gameplay Tag Container Structure (by ref)"
   ) {is=title K2Node_AsyncAction}
   n8 = PilgrimWaitForEventRouter!(execute:Exec=@n7.OnReady, ContextActor:Actor Object Reference=GetOwningPlayer()) -> (then:Exec, OnEventRouterAvailable:Exec, EventRouter:Gameplay Event Router Component Object Reference) {is=title K2Node_AsyncAction}
   Set EventRouter(execute=@n8.OnEventRouterAvailable, EventRouter:Gameplay Event Router Component Object Reference=@n8) -> (Output_Get:Gameplay Event Router Component Object Reference)
   > Update Button Text()
   > n9 = Sequence!(execute) -> (then_0, then_1, then_2:Exec)
   n10 = Get EventRouter() -> (EventRouter:Gameplay Event Router Component Object Reference)
   > n11 = StartListeningToEvent!(
    execute:Exec
    Target:Gameplay Event Router Component Object Reference=@n10
    EventType:Script Struct Object Reference=/PilgrimQuickplay/Items/PilgrimEvent_GreenRoomTimerStateChanged
    Context:Object Reference
    EventBubblingRule:EEventBubblingRule Enum
    None:%=@n10
   ) -> (
    then:Exec
    AsyncTaskProxy:Async Action Start Listening To Event Object Reference
    OnEventReceived:Exec
    Payload_GreenRoomTimerState_2_#1:GreenRoomTimerState Enum
    Payload_FinalTimerLength_5_#2:double=~
   )
   > Add(
    TargetArray:Array of Async Action Start Listening To Event Object References=Get GameplayEventHandles() -> (GameplayEventHandles:Array of Async Action Start Listening To Event Object References)
    NewItem:"Async Action Start Listening To Event Object Reference (by ref)"=@n11
   )
   n12 = Switch on GreenRoomTimerState!(execute=@n11.OnEventReceived, Selection:GreenRoomTimerState Enum=@n11.Payload_GreenRoomTimerState_2_#1) -> (NewEnumerator0:Exec, NewEnumerator1:Exec, NewEnumerator2:Exec, NewEnumerator3:Exec, NewEnumerator4:Exec) {is=title K2Node_SwitchEnum}
   SetVisibility(
    execute=@n12.NewEnumerator2
    self=Get HorizontalBoxReadyUnreadyWithInput() -> (HorizontalBoxReadyUnreadyWithInput:Overlay Object Reference)
    InVisibility=Collapsed
   )
   n13 = Get EventRouter()
   n14 = StartListeningToEvent!(
    execute=@n9.then_1
    Target=@n13
    EventType=/Script/PilgrimQPSharedRuntime.PilgrimEvent_QuickplayState_Began
    Context
    EventBubblingRule
    None:%=@n13
   ) -> (then, AsyncTaskProxy, OnEventReceived, Payload_StateId:EPilgrimQuickplayState Enum)
   > Add(TargetArray=Get GameplayEventHandles(), NewItem=@n14)
   n15 = Branch(
    execute=@n14.OnEventReceived
    Condition=OR Boolean!(
     A="Equal (Enum)"(A:EPilgrimQuickplayState Enum=@n14.Payload_StateId, B:EPilgrimQuickplayState Enum=Pregame)
     B="Equal (Enum)"(A=@n14.Payload_StateId, B=SongResults)
     C:bool="Equal (Enum)"(A=@n14.Payload_StateId, B=SetResults)
    ) -> (ReturnValue)
   )
   > SetVisibility(InVisibility=Visible)
   > SetVisibility(self=Get HorizontalBoxReadyUnreadyWithInput(), InVisibility=SelfHitTestInvisible)
   > Update Button Text()
   SetVisibility(execute=@n15.else, InVisibility=Hidden)
   n16 = StartListeningToEvent!(
    execute=@n9.then_2
    Target=Get EventRouter()
    EventType=/Script/PilgrimCoreRuntime.PilgrimEvent_BandState_CustomSetlistChanged
    Context
    EventBubblingRule
   ) -> (then, AsyncTaskProxy, OnEventReceived, Payload_ChangedSetlist:Pilgrim Custom Setlist Object Reference)
   > Add(TargetArray=Get GameplayEventHandles(), NewItem=@n16)
   Update Button Text(execute=@n16.OnEventReceived)
   Event Destruct()
   > Parent: Destruct()
  ]}
 Update Button Text {nodes [
   Update Button Text() {is=title K2Node_FunctionEntry}
   > n1 = Get!(execute:Exec) -> (then:Exec, else:Exec, LocalPlayerState:Player State Object Reference)
   n2 = GetSparksMusicPlayspaceOfClass!(
    PlayspaceType:Sparks Music Playspace Class Reference=/PilgrimQuickplay/Playspaces/BP_PilgrimQuickplayPlayspace.BP_PilgrimQuickplayPlayspace_C
    Actor:Actor Object Reference=@n1
    bMatchChildTypes:bool=false
   ) -> (ReturnValue:BP Pilgrim Quickplay Playspace Object Reference)
   n3 = Length(TargetArray:Array of Names=GetSongsAddedByPlayer!(
     self:Pilgrim Custom Setlist Object Reference=Get PilgrimCustomSetlist!(self:BP Pilgrim Quickplay Playspace Object Reference=@n2) -> (PilgrimCustomSetlist:Pilgrim Custom Setlist Object Reference)
     Player:Player State Object Reference=Get LocalPlayerState() -> (LocalPlayerState:Player State Object Reference)
    ) -> (ReturnValue:Array of Names))
   > "SetText (Text)"(
    self=[Get Block_Text_1() -> (Block_Text_1:Common Text Object Reference), Get Block_Text() -> (Block_Text:Common Text Object Reference)]
    InText=SelectText(A="LOCTEXT(\"Change Setlist\")", B="LOCTEXT(\"Choose a Song\")", bPickA="integer > integer"(A=@n3))
   )
  ]}
]
variables [
 _DisabledByExternalSource {type=boolean}
 AsyncEventHandles {type="array<CancellableAsyncAction>", default []}
 BackgroundCornerRadius {type=number}
 CountDown {type=WidgetAnimation, default=null}
 EventRouter {type=GameplayEventRouterComponent, default=null}
 InitialCountdownState {type=WidgetAnimation, default=null}
 IsReadyUpDisabled {type=boolean}
 LocalPlayerState {type=PlayerState, default=null}
 ReadyText {type=string, default=Ready Up}
 ReadyText_0 {type=string, default=Ready Up}
]
widgets {
 info {inheritedWidgetCount=0, parentClass=/Script/PilgrimCoreUI.PilgrimGameWidgetBase, rootWidgetClass=&Overlay, widgetCount=17}
 widgets [
  PromptParent {widgetClassPath=&Overlay, bIsVariable=true, widget=PromptParent}
  HorizontalBoxReadyUnreadyWithInput {
   widgetClassPath=&Overlay
   bIsVariable=true
   parent=PromptParent
   widget=HorizontalBoxReadyUnreadyWithInput
   slot {class=OverlaySlot, properties {horizontalAlignment=HAlign_Fill, verticalAlignment=VAlign_Fill}}
  }
  ReadyUpOverlay {
   widgetClassPath=&Overlay
   bIsVariable=true
   parent=HorizontalBoxReadyUnreadyWithInput
   widget=ReadyUpOverlay
   slot {class=OverlaySlot, properties {horizontalAlignment=HAlign_Fill, verticalAlignment=VAlign_Fill}}
  }
  Spacer_130 {
   widgetClassPath=&Spacer
   parent=ReadyUpOverlay
   widget=Spacer_130
   properties {size=<1 64>}
   slot {class=OverlaySlot, properties {verticalAlignment=VAlign_Fill}}
  }
  ReadyLayer {
   widgetClassPath=&Image
   bIsVariable=true
   parent=ReadyUpOverlay
   widget=ReadyLayer
   properties {
    visibility=Collapsed
    brush {
     bIsDynamicallyLoaded=false
     drawAs=Image
     imageType=NoImage
     mirroring=NoMirror
     resourceName=None
     resourceObject=/PilgrimCore/Widgets/Materials/MI_UI_Pilgrim_ReadyButtonLayer
     tiling=NoTile
     imageSize=<32 32>
     margin=<ltrb 0 0 0 0>
     outlineSettings {
      bUseBrushTransparency=false
      roundingType=HalfHeightRadius
      width=0
      color {colorUseRule=UseColor_Specified, specifiedColor=<rgba 0 0 0 0>}
      cornerRadii=<0 0 0 0>
     }
     tintColor {colorUseRule=UseColor_Specified, specifiedColor=<rgba 1 1 1 1>}
     uVRegion {bIsValid=false, max=<0 0>, min=<0 0>}
    }
   }
   slot {class=OverlaySlot, properties {horizontalAlignment=HAlign_Fill, verticalAlignment=VAlign_Fill}}
  }
  ButtonImage {
   widgetClassPath=&Image
   bIsVariable=true
   parent=ReadyUpOverlay
   widget=ButtonImage
   properties {brush {
     bIsDynamicallyLoaded=false
     drawAs=Image
     imageType=FullColor
     mirroring=NoMirror
     resourceName=None
     resourceObject=/SparksUICommon/UILibrary/MI_FMCore_Button_Regular_ReadyUp
     tiling=NoTile
     imageSize=<32 32>
     margin=<ltrb 0 0 0 0>
     outlineSettings {
      bUseBrushTransparency=false
      roundingType=HalfHeightRadius
      width=0
      color {colorUseRule=UseColor_Specified, specifiedColor=<rgba 0 0 0 0>}
      cornerRadii=<0 0 0 0>
     }
     tintColor {colorUseRule=UseColor_Specified, specifiedColor=<rgba 1 1 1 1>}
     uVRegion {bIsValid=false, max=<0 0>, min=<0 0>}
    }}
   slot {class=OverlaySlot, properties {horizontalAlignment=HAlign_Fill, verticalAlignment=VAlign_Fill}}
  }
  VerticalBox_Text {
   widgetClassPath=&VerticalBox
   bIsVariable=true
   parent=ReadyUpOverlay
   widget=VerticalBox_Text
   slot {class=OverlaySlot, properties {horizontalAlignment=HAlign_Center, verticalAlignment=VAlign_Center, padding=<ltrb 64 8 64 8>}}
  }
  CommonVisibilityWidgetLegacy_120 {
   widgetClassPath=&CommonVisibilityWidgetLegacy
   parent=VerticalBox_Text
   widget=CommonVisibilityWidgetLegacy_120
   properties {bShowForTouch=false, padding=<ltrb 0 0 0 0>}
   slot {class=VerticalBoxSlot, properties {horizontalAlignment=HAlign_Center}}
  }
  Block_Text {
   widgetClassPath=&CommonTextBlock
   bIsVariable=true
   parent=CommonVisibilityWidgetLegacy_120
   widget=Block_Text
   properties {
    justification=Center
    scrollStyle=&TextScrollStyle-Base_C
    text=Choose a Song
    textTransformPolicy=ToUpper
    colorAndOpacity {colorUseRule=UseColor_Specified, specifiedColor=<rgba 0.0048 0.004 0.0116 1>}
    font {
     bForceMonospaced=false
     bMaterialIsStencil=false
     fontMaterial=null
     fontObject=&HeadingNow
     letterSpacing=0
     monospacedWidth=1
     size=21
     skewAmount=0
     typefaceFontName=86 Bold Italic
     outlineSettings {
      bApplyOutlineToDropShadows=false
      bMiteredCorners=false
      bSeparateFillAlpha=false
      outlineMaterial=null
      outlineSize=0
      outlineColor=<rgba 0 0 0 1>
     }
    }
    shadowOffset=<0 0>
   }
   slot {class=BorderSlot, properties {padding=<ltrb 0 0 0 0>}}
  }
  CommonVisibilityWidgetLegacy {
   widgetClassPath=&CommonVisibilityWidgetLegacy
   parent=VerticalBox_Text
   widget=CommonVisibilityWidgetLegacy
   properties {bShowForGamepad=false, bShowForMouseAndKeyboard=false, padding=<ltrb 0 0 0 0>}
   slot {class=VerticalBoxSlot, properties {horizontalAlignment=HAlign_Center}}
  }
  Block_Text_1 {
   widgetClassPath=&CommonTextBlock
   bIsVariable=true
   parent=CommonVisibilityWidgetLegacy
   widget=Block_Text_1
   properties {
    justification=Center
    scrollStyle=&TextScrollStyle-Base_C
    text=Choose a Song
    textTransformPolicy=ToUpper
    colorAndOpacity {colorUseRule=UseColor_Specified, specifiedColor=<rgba 0.8714 1 0.9823 1>}
    font {
     bForceMonospaced=false
     bMaterialIsStencil=false
     fontMaterial=null
     fontObject=&HeadingNow
     letterSpacing=0
     monospacedWidth=1
     size=21
     skewAmount=0
     typefaceFontName=86 Bold Italic
     outlineSettings {
      bApplyOutlineToDropShadows=false
      bMiteredCorners=false
      bSeparateFillAlpha=false
      outlineMaterial=null
      outlineSize=0
      outlineColor=<rgba 0 0 0 1>
     }
    }
    shadowOffset=<0 0>
   }
   slot {class=BorderSlot, properties {padding=<ltrb 0 0 0 0>}}
  }
  Block_SecondaryText {
   widgetClassPath=&CommonTextBlock
   bIsVariable=true
   parent=VerticalBox_Text
   widget=Block_SecondaryText
   properties {
    text=(Hold)
    visibility=Collapsed
    colorAndOpacity {colorUseRule=UseColor_Specified, specifiedColor=<rgba 0.0048 0.004 0.0116 0.5>}
    font {
     bForceMonospaced=false
     bMaterialIsStencil=false
     fontMaterial=null
     fontObject=&HeadingNow
     letterSpacing=0
     monospacedWidth=1
     size=15
     skewAmount=0
     typefaceFontName=75 Medium
     outlineSettings {
      bApplyOutlineToDropShadows=false
      bMiteredCorners=false
      bSeparateFillAlpha=false
      outlineMaterial=null
      outlineSize=0
      outlineColor=<rgba 0 0 0 1>
     }
    }
    renderTransform {angle=0, scale=<1 1>, shear=<0 0>, translation=<0 -2>}
    shadowOffset=<0 0>
   }
   slot {class=VerticalBoxSlot, properties {horizontalAlignment=HAlign_Center}}
  }
  FortResponsivePanel_0 {
   widgetClassPath=/Script/FortniteUI.FortResponsivePanel
   parent=HorizontalBoxReadyUnreadyWithInput
   widget=FortResponsivePanel_0
   slot {class=OverlaySlot, properties {verticalAlignment=VAlign_Center, padding=<ltrb 23 0 0 0>}}
  }
  CommonVisibilityWidgetLegacy_0 {
   widgetClassPath=&CommonVisibilityWidgetLegacy
   parent=FortResponsivePanel_0
   widget=CommonVisibilityWidgetLegacy_0
   properties {bShowForTouch=false, padding=<ltrb 0 0 0 0>}
   slot {class=FortResponsivePanelSlot}
  }
  Overlay_0 {
   widgetClassPath=&Overlay
   parent=CommonVisibilityWidgetLegacy_0
   widget=Overlay_0
   slot {class=BorderSlot, properties {padding=<ltrb 0 0 0 0>}}
  }
  Temp_HoldToFill_Underlayer {
   widgetClassPath=&Image
   bIsVariable=true
   parent=Overlay_0
   widget=Temp_HoldToFill_Underlayer
   properties {brush {
     bIsDynamicallyLoaded=false
     drawAs=Image
     imageType=FullColor
     mirroring=NoMirror
     resourceName=None
     resourceObject=/PilgrimCore/Widgets/Materials/MI_UI_Pilgrim_Panel_Underlayer_ColorFill_PromtUnderlayer
     tiling=NoTile
     imageSize=<64 64>
     margin=<ltrb 0 0 0 0>
     outlineSettings {
      bUseBrushTransparency=false
      roundingType=HalfHeightRadius
      width=0
      color {colorUseRule=UseColor_Specified, specifiedColor=<rgba 0 0 0 0>}
      cornerRadii=<0 0 0 0>
     }
     tintColor {colorUseRule=UseColor_Specified, specifiedColor=<rgba 1 1 1 0.45>}
     uVRegion {bIsValid=false, max=<0 0>, min=<0 0>}
    }}
   slot {class=OverlaySlot, properties {horizontalAlignment=HAlign_Center, verticalAlignment=VAlign_Center}}
  }
  KeybindWidget_ChangeBindings_1 {
   widgetClassPath=/FortUILibrary/Input/KeybindWidget
   bIsVariable=true
   parent=Overlay_0
   widget=KeybindWidget_ChangeBindings_1
   properties {bUseKeyboardTextureForKeys=true, boundAction=Use, forcedHoldKeybindStatus=ForcedHold, visibility=Visible, boundActionFallbacks [UseOrReload]}
   slot {class=OverlaySlot, properties {horizontalAlignment=HAlign_Center, verticalAlignment=VAlign_Center}}
  }
 ]
}
