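# Touch green-room screen for Pilgrim Quickplay, with ready-up, difficulty and music library buttons beside player and next-song info.
# Buttons call the Pilgrim player controller component or broadcast a playlist-open event through the playspace's event router.
# With the unified part picker enabled, it registers as widget host; injected content replaces instrument select and difficulty controls.
# Construct also starts FTUE-gated coach marks and async Jamming-tag listeners that toggle the SZ_TopRight panel.
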WidgetBlueprint /PilgrimQuickplay/Widgets/Mobile/WBP_Pilgrim_GreenRoom_Touch_ReadyUp
graphs [
 ActivateProPartsCoachmark {nodes [
   ActivateProPartsCoachmark() {is=title K2Node_FunctionEntry}
   n1 = Get WBP_PilgrimCoachMark_ProParts() -> (WBP_PilgrimCoachMark_ProParts:WBP Pilgrim Coach Mark Pro Parts Object Reference)
   > Branch(Condition=ShouldShowOnboarding!(self:Fort Sidebar Onboard Tooltip Widget Object Reference=@n1) -> (ReturnValue:bool))
   > ActivateWidget(self=@n1)
   > Bind Event to On Pro Parts Tooltip Dismissed(self:WBP Pilgrim Coach Mark Pro Parts Object Reference=@n1, Delegate=Create Event()) {is=title K2Node_AddDelegate}
  ]}
 ActivateStageTechCoachmark {nodes [
   ActivateStageTechCoachmark() {is=title K2Node_FunctionEntry}
   n1 = Get WBP_PilgrimCoachMark_StageTech() -> (WBP_PilgrimCoachMark_StageTech:WBP Pilgrim Coach Mark Stage Tech Object Reference)
   > Branch(Condition=ShouldShowOnboarding(self=@n1))
   > ActivateWidget(self=@n1)
   > Bind Event to On Stage Tech Tooltip Dismissed(self:WBP Pilgrim Coach Mark Stage Tech Object Reference=@n1, Delegate=Create Event()) {is=title K2Node_AddDelegate}
  ]}
 EventGraph {nodes [
   "On Pressed (Button_Ready)"() {is=title K2Node_ComponentBoundEvent}
   > FortShippingLog!(
    execute
    WorldContextObject:Object Reference=GetOwningPlayer()
    InString:String=WBP_Pilgrim_GreenRoom_Touch_ReadyUp::Button_Ready_New: Pressed
    bLogAsWarning:bool=false
   ) -> (then)
   > Ready Up()
   "On Pressed (Button_Difficulty)"() {is=title K2Node_ComponentBoundEvent}
   > Swap Difficulty()
   "On Pressed (Button_OpenMusicLibrary)"() {is=title K2Node_ComponentBoundEvent}
   n1 = GetOwningPlayer()
   > FortShippingLog(WorldContextObject=@n1, InString=WBP_Pilgrim_GreenRoom_Touch_ReadyUp::Button_OpenLibrary: Pressed, bLogAsWarning=false)
   n2 = GetSparksMusicPlayspaceOfClass!(
    PlayspaceType:Sparks Music Playspace Class Reference=/PilgrimQuickplay/Playspaces/BP_PilgrimQuickplayPlayspace.BP_PilgrimQuickplayPlayspace_C
    Actor:Actor Object Reference=@n1
    bMatchChildTypes:bool=false
   ) -> (ReturnValue:BP Pilgrim Quickplay Playspace Object Reference)
   > Is Valid(InputObject=@n2)
   n3 = GetGameplayEventRouter!(ContextActor:Actor Object Reference=@n2) -> (ReturnValue:Gameplay Event Router Component Object Reference)
   > Is Valid(InputObject=@n3)
   n4 = GetOwningPlayer()
   > Broadcast Event!(
    execute
    self:Gameplay Event Router Component Object Reference=@n3
    EventData:Pilgrim Event Playlist Display Request Open Structure=Make Pilgrim Event Playlist Display Request Open!(PlayerController_29_#1:Fort Player Controller Object Reference=Cast To FortPlayerController!(Object=@n4) -> (AsFort Player Controller, bSuccess:bool=~)) -> (PilgrimEvent_Playlist_Display_RequestOpen:Pilgrim Event Playlist Display Request Open Structure) {is=title K2Node_MakeStruct}
    EventContext:Object Reference
   ) -> (then, ReturnValue:bool)
   OnFirstCoachMarkClosed() {is=title K2Node_CustomEvent}
   > ActivateProPartsCoachmark() {is=title K2Node_CallFunction}
   ClearJamTagListeners() {is=title K2Node_CustomEvent}
   > Cancel(self:Cancellable Async Action Object Reference=Get Wait for Add Jamming Tag Async Action() -> (Wait for Add Jamming Tag Async Action:Ability Async Wait Gameplay Tag Added Object Reference))
   > Cancel(self=Get Wait for Remove Jamming Tag Async Action() -> (Wait for Remove Jamming Tag Async Action:Ability Async Wait Gameplay Tag Removed Object Reference))
   OnSecondCoachMarkClosed() {is=title K2Node_CustomEvent}
   n5 = GetComponentByClass!(
    self:Actor Object Reference=GetOwningPlayer()
    ComponentClass:Actor Component Class Reference=/Script/PilgrimCoreUI.PilgrimFTUEControllerComponent
   ) -> (ReturnValue:Pilgrim FTUEController Component Object Reference)
   > Is Valid(InputObject=@n5)
   > SetHasSeenRoadieCoachMark!(execute, self:Pilgrim FTUEController Component Object Reference=@n5, bHasSeenRoadieCoachMark:bool=true) -> (then)
   SetupJamTagListeners() {is=title K2Node_CustomEvent}
   n6 = GetOwningPlayer()
   > Is Valid(InputObject=@n6)
   > n7 = Sequence()
   > n8 = WaitGameplayTagAddToActor!(
    execute:Exec
    TargetActor:Actor Object Reference=@n6
    Tag:Gameplay Tag Structure=(TagName="Gameplay.Action.Player.Jam.Jamming")
    OnlyTriggerOnce:bool=false
   ) -> (then:Exec, AsyncTaskProxy:Ability Async Wait Gameplay Tag Added Object Reference, Added:Exec) {is=title K2Node_AsyncAction}
   > Set Wait for Add Jamming Tag Async Action(Wait for Add Jamming Tag Async Action:Ability Async Wait Gameplay Tag Added Object Reference=@n8) -> (Output_Get:Ability Async Wait Gameplay Tag Added Object Reference)
   Branch(execute=@n8.Added, Condition=IsJamShowingNowPlayingUI!() -> (IsShowing:bool))
   > SetVisibility(self=Get SZ_TopRight() -> (SZ_TopRight:Safe Zone Object Reference), InVisibility=Hidden)
   n9 = WaitGameplayTagRemoveFromActor!(
    execute:Exec=@n7.then_1
    TargetActor:Actor Object Reference=@n6
    Tag:Gameplay Tag Structure=(TagName="Gameplay.Action.Player.Jam.Jamming")
    OnlyTriggerOnce:bool=false
   ) -> (then:Exec, AsyncTaskProxy:Ability Async Wait Gameplay Tag Removed Object Reference, Removed:Exec) {is=title K2Node_AsyncAction}
   > Set Wait for Remove Jamming Tag Async Action(Wait for Remove Jamming Tag Async Action:Ability Async Wait Gameplay Tag Removed Object Reference=@n9) -> (Output_Get:Ability Async Wait Gameplay Tag Removed Object Reference)
   Branch(execute=@n9.Removed, Condition=IsJamShowingNowPlayingUI())
   > SetVisibility(self=Get SZ_TopRight(), InVisibility=SelfHitTestInvisible)
   Event Destruct()
   > Unbind all Events from On Stage Tech Tooltip Dismissed!(execute:Exec, self:WBP Pilgrim Coach Mark Stage Tech Object Reference=Get WBP_PilgrimCoachMark_StageTech()) -> (then:Exec) {is=title K2Node_ClearDelegate}
   > Unbind all Events from On Pro Parts Tooltip Dismissed!(execute:Exec, self:WBP Pilgrim Coach Mark Pro Parts Object Reference=Get WBP_PilgrimCoachMark_ProParts()) -> (then:Exec) {is=title K2Node_ClearDelegate}
   > UnregisterWidgetHost!(
    execute
    WidgetHost:Widget Object Reference
    WidgetHostTag:"Gameplay Tag Structure (by ref)"=Get UniversalPartPickerSlotTag() -> (UniversalPartPickerSlotTag:Gameplay Tag Structure)
   ) -> (then)
   Event Construct()
   > n10 = Sequence!(execute) -> (then_0, then_1, then_2:Exec)
   n11 = GetComponentByClass(self=GetOwningPlayer(), ComponentClass=&PilgrimFTUEControllerComponent)
   > Is Valid(InputObject=@n11)
   n12 = Get WBP_PilgrimCoachMark_StageTech()
   n13 = NOT Boolean(A=GetHasSeenRoadieCoachMark!(self:Pilgrim FTUEController Component Object Reference=@n11) -> (bOutHasValidSettings:bool, ReturnValue:bool).ReturnValue)
   > n14 = Branch(Condition=AND Boolean(A=ShouldShowOnboarding(self=@n12), B=@n13))
   > ActivateStageTechCoachmark() {is=title K2Node_CallFunction}
   Branch(execute=@n14.else, Condition=@n13)
   > ActivateProPartsCoachmark()
   Branch(execute=@n10.then_1, Condition=IsUnifiedPartPickerEnabled!() -> (ReturnValue:bool))
   > RegisterWidgetHost!(
    execute
    WidgetHost:Widget Object Reference
    WidgetHostTag:"Gameplay Tag Structure (by ref)"=Get UniversalPartPickerSlotTag()
    IsPlayerSpecific:bool=true
   ) -> (then)
   SetupJamTagListeners(execute=@n10.then_2) {is=title K2Node_CallFunction}
   Event OnInitialized()
   > UnregisterWidgetHost(WidgetHostTag=Get UniversalPartPickerSlotTag())
   n15 = Event Add Content To Host Widget!() -> (
    OutputDelegate
    then
    WidgetHostTag:"Gameplay Tag Structure (by ref)"
    ContentWidget:Widget Object Reference
    Padding:"Margin Structure (by ref)"
    bFill:bool
   ) {is=title K2Node_Event}
   > Branch(Condition="Equal (GameplayTag)"(A=@n15.WidgetHostTag, B=Get UniversalPartPickerSlotTag()))
   > RemoveFromParent(self=Get WBP_Pilgrim_GreenRoom_InstrumentSelectUI() -> (WBP_Pilgrim_GreenRoom_InstrumentSelectUI:WBP Pilgrim Green Room Instrument Select UI Object Reference))
   > RemoveFromParent(self=Get Button_Difficulty() -> (Button_Difficulty:Button Object Reference))
   > n16 = AddChildToVerticalBox(self=Get VerticalBox() -> (VerticalBox:Vertical Box Object Reference), Content=@n15.ContentWidget)
   > SetHorizontalAlignment(self:Vertical Box Slot Object Reference=@n16, InHorizontalAlignment=HAlign_Fill)
  ]}
 Ready Up {nodes [
   Ready Up() {is=title K2Node_FunctionEntry}
   > n1 = Branch(Condition=Get IsReadyUpDisabled!(self:WBP Pilgrim Green Room Toggle Ready UI Object Reference=Get WBP_Pilgrim_GreenRoom_ToggleReadyUI() -> (WBP_Pilgrim_GreenRoom_ToggleReadyUI:WBP Pilgrim Green Room Toggle Ready UI Object Reference)) -> (IsReadyUpDisabled:bool))
   n2 = Cast To FortPlayerControllerGameplay!(execute=@n1.else, Object=GetOwningPlayer()) -> (then, CastFailed, AsFort Player Controller Gameplay:Fort Player Controller Gameplay Object Reference) {is=title K2Node_DynamicCast}
   n3 = GetComponentByClass(self=@n2, ComponentClass=/PilgrimCore/Player/Components/BP_PilgrimPlayerControllerComponent.BP_PilgrimPlayerControllerComponent_C) -> (ReturnValue:BP Pilgrim Player Controller Component Object Reference)
   > Is Valid(InputObject=@n3)
   > Request Server Toggle Player ReadyUp!(execute, self:BP Pilgrim Player Controller Component Object Reference=@n3, Controller:Fort Player Controller Object Reference=@n2) -> (then)
  ]}
 RemoveContentFromWidgetHost {nodes [
   Remove Content From Host Widget!() -> (
    then
    WidgetHostTag:"Gameplay Tag Structure (by ref)"
    ContentWidget:Widget Object Reference
    RemovalType:EWidgetSlottingRemovalType Enum=Unparent
   ) {is=title K2Node_FunctionEntry}
   > Return Node!(execute, ReturnValue:bool=false)
  ]}
 Swap Difficulty {nodes [
   Swap Difficulty() {is=title K2Node_FunctionEntry}
   n1 = GetComponentByClass(
    self=GetOwningPlayer()
    ComponentClass=&BP_PilgrimPlayerControllerComponent_C
   )
   > Is Valid(InputObject=@n1)
   > Swap Difficulty Input Pressed(self:BP Pilgrim Player Controller Component Object Reference=@n1)
  ]}
]
variables [
 UniversalPartPickerSlotTag {type=GameplayTag, default {tagName=Pilgrim.Hud.PartPicker.Slot}}
 Wait for Add Jamming Tag Async Action {type=AbilityAsync_WaitGameplayTagAdded, default=null}
 Wait for Remove Jamming Tag Async Action {type=AbilityAsync_WaitGameplayTagRemoved, default=null}
]
widgets {
 info {inheritedWidgetCount=0, parentClass=&UserWidget, rootWidgetClass=&CanvasPanel, widgetCount=32}
 widgets [
  CanvasPanel_0 {widgetClassPath=&CanvasPanel, widget=CanvasPanel_0}
  SZ_TopLeft {
   widgetClassPath=&SafeZone
   parent=CanvasPanel_0
   widget=SZ_TopLeft
   properties {padBottom=false, padLeft=false, padRight=false, padTop=false}
   slot {
    class=CanvasPanelSlot
    properties {bAutoSize=true, layoutData {alignment=<1 0>, anchors {maximum=<1 0>, minimum=<1 0>}, offsets=<ltrb -30 220 124 124>}}
   }
  }
  Overlay_0 {widgetClassPath=&Overlay, parent=SZ_TopLeft, widget=Overlay_0, slot {class=SafeZoneSlot}}
  Backplate_TextChat {
   widgetClassPath=&Image
   parent=Overlay_0
   widget=Backplate_TextChat
   properties {brush {
     bIsDynamicallyLoaded=false
     drawAs=Image
     imageType=FullColor
     mirroring=NoMirror
     resourceName=None
     resourceObject=/PilgrimQuickplay/UI/MI_PilgrimQuickplay_UIKit_Button_QuietRoundedCorners
     tiling=NoTile
     imageSize=<88 88>
     margin=<ltrb 0 0 0 0>
     outlineSettings {
      bUseBrushTransparency=false
      roundingType=HalfHeightRadius
      width=0
      color {colorUseRule=UseColor_Specified, specifiedColor=<rgba 0 0 0 0>}
      cornerRadii=<0 0 0 0>
     }
     tintColor {colorUseRule=UseColor_Specified, specifiedColor=<rgba 1 1 1 1>}
     uVRegion {bIsValid=false, max=<0 0>, min=<0 0>}
    }}
   slot {class=OverlaySlot, properties {horizontalAlignment=HAlign_Center, verticalAlignment=VAlign_Center}}
  }
  WBP_PilgrimTextChatMenuPrompt {
   widgetClassPath=&/WBP_PilgrimTextChatMenuPrompt
   parent=Overlay_0
   widget=WBP_PilgrimTextChatMenuPrompt
   properties {removePadding=true, renderTransform {angle=0, scale=<0.7 0.7>, shear=<0 0>, translation=<0 0>}}
   slot {class=OverlaySlot, properties {horizontalAlignment=HAlign_Fill, verticalAlignment=VAlign_Fill}}
  }
  CommonVisualAttachment_0 {
   widgetClassPath=&CommonVisualAttachment
   parent=CanvasPanel_0
   widget=CommonVisualAttachment_0
   properties {contentAnchor=<0.8 0>}
   slot {class=CanvasPanelSlot, properties {layoutData {alignment=<0 0>, anchors {maximum=<1 0>, minimum=<1 0>}, offsets=<ltrb -140 226 0 0>}}}
  }
  WBP_PilgrimCoachMark_StageTech {
   widgetClassPath=/PilgrimCore/Widgets/FTUE/CoachMarks/WBP_PilgrimCoachMark_StageTech
   bIsVariable=true
   parent=CommonVisualAttachment_0
   widget=WBP_PilgrimCoachMark_StageTech
   properties {arrow Position=Top_Right, desiredFocusWidget {widgetName=None}}
   slot {class=SizeBoxSlot}
  }
  SZ_TopRight {
   widgetClassPath=&SafeZone
   bIsVariable=true
   parent=CanvasPanel_0
   widget=SZ_TopRight
   properties {padBottom=false, padLeft=false, padRight=false, padTop=false}
   slot {class=CanvasPanelSlot, properties {layoutData {alignment=<0 0>, anchors {maximum=<0 0>, minimum=<0 0>}, offsets=<ltrb 30 100 467 517>}}}
  }
  VerticalBox_1 {widgetClassPath=&VerticalBox, parent=SZ_TopRight, widget=VerticalBox_1, slot {class=SafeZoneSlot}}
  Overlay_1 {
   widgetClassPath=&Overlay
   parent=VerticalBox_1
   widget=Overlay_1
   slot {class=VerticalBoxSlot, properties {horizontalAlignment=HAlign_Left, verticalAlignment=VAlign_Top, padding=<ltrb 0 -18 0 0>}}
  }
  Backplate_SongMenu {
   widgetClassPath=&Image
   parent=Overlay_1
   widget=Backplate_SongMenu
   properties {brush {
     bIsDynamicallyLoaded=false
     drawAs=Image
     imageType=FullColor
     mirroring=NoMirror
     resourceName=None
     resourceObject=&MI_PilgrimQuickplay_UIKit_Button_QuietRoundedCorners
     tiling=NoTile
     imageSize=<130 88>
     margin=<ltrb 0 0 0 0>
     outlineSettings {
      bUseBrushTransparency=false
      roundingType=HalfHeightRadius
      width=0
      color {colorUseRule=UseColor_Specified, specifiedColor=<rgba 0 0 0 0>}
      cornerRadii=<0 0 0 0>
     }
     tintColor {colorUseRule=UseColor_Specified, specifiedColor=<rgba 1 1 1 1>}
     uVRegion {bIsValid=false, max=<0 0>, min=<0 0>}
    }}
   slot {class=OverlaySlot, properties {horizontalAlignment=HAlign_Fill, verticalAlignment=VAlign_Center}}
  }
  WBP_SongMenuPrompt_Touch {
   widgetClassPath=&/WBP_SongMenuPromptWithKeyBind_Touch
   parent=Overlay_1
   widget=WBP_SongMenuPrompt_Touch
   slot {class=OverlaySlot, properties {horizontalAlignment=HAlign_Center, verticalAlignment=VAlign_Center}}
  }
  WBP_Pilgrim_GreenRoom_BandwidthPlayerList {
   widgetClassPath=/PilgrimQuickplay/Widgets/GreenRoom/WBP_Pilgrim_GreenRoom_PlayerInfoList
   parent=VerticalBox_1
   widget=WBP_Pilgrim_GreenRoom_BandwidthPlayerList
   slot {class=VerticalBoxSlot}
  }
  WBP_PilgrimQuickplay_NextSongDetailsPanel {
   widgetClassPath=/PilgrimQuickplay/Widgets/Results/SongResults/WBP_PilgrimQuickplay_NextSongDetailsPanel
   parent=VerticalBox_1
   widget=WBP_PilgrimQuickplay_NextSongDetailsPanel
   properties {isInGreenRoom=true}
   slot {class=VerticalBoxSlot, properties {padding=<ltrb 51 32 0 0>}}
  }
  CanvasPanel {
   widgetClassPath=&CanvasPanel
   parent=CanvasPanel_0
   widget=CanvasPanel
   slot {
    class=CanvasPanelSlot
    properties {layoutData {alignment=<0 0>, anchors {maximum=<1 1>, minimum=<0 1>}, offsets=<ltrb 0.0001 -1094 0.0002 0>}}
   }
  }
  SizeBox_7 {
   widgetClassPath=&SizeBox
   parent=CanvasPanel
   widget=SizeBox_7
   properties {bOverride_MinDesiredWidth=true, maxDesiredWidth=400, minDesiredHeight=400, minDesiredWidth=400, widthOverride=400}
   slot {
    class=CanvasPanelSlot
    properties {bAutoSize=true, layoutData {alignment=<0.5 1>, anchors {maximum=<0.7 1>, minimum=<0.7 1>}, offsets=<ltrb -72.4362 -40 418.8579 627.873>}}
   }
  }
  VerticalBox_2 {
   widgetClassPath=&VerticalBox
   parent=SizeBox_7
   widget=VerticalBox_2
   properties {bOverride_Cursor=true}
   slot {class=SizeBoxSlot}
  }
  ScaleBox {
   widgetClassPath=&ScaleBox
   parent=VerticalBox_2
   widget=ScaleBox
   slot {class=VerticalBoxSlot, properties {padding=<ltrb 0 0 0 10>}}
  }
  WBP_Pilgrim_GreenRoom_InstrumentNotice_1 {
   widgetClassPath=/PilgrimQuickplay/Widgets/WBP_Pilgrim_GreenRoom_InstrumentNotice
   parent=ScaleBox
   widget=WBP_Pilgrim_GreenRoom_InstrumentNotice_1
   slot {class=ScaleBoxSlot, properties {horizontalAlignment=HAlign_Fill, verticalAlignment=VAlign_Fill}}
  }
  VerticalBox {
   widgetClassPath=&VerticalBox
   bIsVariable=true
   parent=VerticalBox_2
   widget=VerticalBox
   slot {class=VerticalBoxSlot, properties {padding=<ltrb 0 0 0 10>}}
  }
  CommonVisualAttachment {
   widgetClassPath=&CommonVisualAttachment
   parent=VerticalBox
   widget=CommonVisualAttachment
   properties {contentAnchor=<-0.2 1.125>}
   slot {class=VerticalBoxSlot}
  }
  WBP_PilgrimCoachMark_ProParts {
   widgetClassPath=/PilgrimCore/Widgets/FTUE/CoachMarks/WBP_PilgrimCoachMark_ProParts
   bIsVariable=true
   parent=CommonVisualAttachment
   widget=WBP_PilgrimCoachMark_ProParts
   properties {desiredFocusWidget {widgetName=None}}
   slot {class=SizeBoxSlot}
  }
  WBP_Pilgrim_GreenRoom_InstrumentSelectUI {
   widgetClassPath=/PilgrimQuickplay/Widgets/WBP_Pilgrim_GreenRoom_InstrumentSelectUI
   bIsVariable=true
   parent=VerticalBox
   widget=WBP_Pilgrim_GreenRoom_InstrumentSelectUI
   slot {class=VerticalBoxSlot, properties {horizontalAlignment=HAlign_Center}}
  }
  Button_Difficulty {
   widgetClassPath=&Button
   parent=VerticalBox_2
   widget=Button_Difficulty
   properties {widgetStyle {
     clickedSlateSound {resourceObject=null}
     disabled {
      bIsDynamicallyLoaded=false
      drawAs=NoDrawType
      imageType=NoImage
      mirroring=NoMirror
      resourceName=None
      resourceObject=null
      tiling=NoTile
      imageSize=<0 0>
      margin=<ltrb 0 0 0 0>
      outlineSettings {
       bUseBrushTransparency=false
       roundingType=HalfHeightRadius
       width=0
       color {colorUseRule=UseColor_Specified, specifiedColor=<rgba 0 0 0 0>}
       cornerRadii=<0 0 0 0>
      }
      tintColor {colorUseRule=UseColor_Specified, specifiedColor=<rgba 1 1 1 1>}
      uVRegion {bIsValid=false, max=<0 0>, min=<0 0>}
     }
     disabledForeground {colorUseRule=UseColor_Specified, specifiedColor=<rgba 0.5271 0.5271 0.5271 1>}
     hovered {
      bIsDynamicallyLoaded=false
      drawAs=NoDrawType
      imageType=NoImage
      mirroring=NoMirror
      resourceName=None
      resourceObject=null
      tiling=NoTile
      imageSize=<32 32>
      margin=<ltrb 0 0 0 0>
      outlineSettings {
       bUseBrushTransparency=true
       roundingType=FixedRadius
       width=1
       color {colorUseRule=UseColor_Specified, specifiedColor=<rgba 0.7243 0.7243 0.7243 1>}
       cornerRadii=<4 4 4 4>
      }
      tintColor {colorUseRule=UseColor_Specified, specifiedColor=<rgba 0.7243 0.7243 0.7243 1>}
      uVRegion {bIsValid=false, max=<0 0>, min=<0 0>}
     }
     hoveredForeground {colorUseRule=UseColor_Specified, specifiedColor=<rgba 0 1 1 1>}
     hoveredSlateSound {resourceObject=null}
     normal {
      bIsDynamicallyLoaded=false
      drawAs=NoDrawType
      imageType=NoImage
      mirroring=NoMirror
      resourceName=None
      resourceObject=null
      tiling=NoTile
      imageSize=<32 32>
      margin=<ltrb 0 0 0 0>
      outlineSettings {
       bUseBrushTransparency=true
       roundingType=FixedRadius
       width=1
       color {colorUseRule=UseColor_Specified, specifiedColor=<rgba 0.6951 0.6951 0.6951 1>}
       cornerRadii=<4 4 4 4>
      }
      tintColor {colorUseRule=UseColor_Specified, specifiedColor=<rgba 0.4955 0.4955 0.4955 1>}
      uVRegion {bIsValid=false, max=<0 0>, min=<0 0>}
     }
     normalForeground {colorUseRule=UseColor_Specified, specifiedColor=<rgba 0 1 1 1>}
     normalPadding=<ltrb 0 0 0 0>
     pressed {
      bIsDynamicallyLoaded=false
      drawAs=NoDrawType
      imageType=NoImage
      mirroring=NoMirror
      resourceName=None
      resourceObject=null
      tiling=NoTile
      imageSize=<32 32>
      margin=<ltrb 0 0 0 0>
      outlineSettings {
       bUseBrushTransparency=true
       roundingType=FixedRadius
       width=1
       color {colorUseRule=UseColor_Specified, specifiedColor=<rgba 0.7243 0.7243 0.7243 1>}
       cornerRadii=<4 4 4 4>
      }
      tintColor {colorUseRule=UseColor_Specified, specifiedColor=<rgba 0.3843 0.3843 0.3843 1>}
      uVRegion {bIsValid=false, max=<0 0>, min=<0 0>}
     }
     pressedForeground {colorUseRule=UseColor_Specified, specifiedColor=<rgba 0 1 1 1>}
     pressedPadding=<ltrb 0 0 0 0>
     pressedSlateSound {resourceObject=null}
    }}
   slot {class=VerticalBoxSlot, properties {padding=<ltrb 0 0 0 16>}}
  }
  SizeBox_6 {
   widgetClassPath=&SizeBox
   parent=Button_Difficulty
   widget=SizeBox_6
   properties {bOverride_HeightOverride=true, heightOverride=95, minDesiredHeight=112}
   slot {class=ButtonSlot, properties {horizontalAlignment=HAlign_Fill, verticalAlignment=VAlign_Fill, padding=<ltrb 0 0 0 0>}}
  }
  WBP_Pilgrim_GreenRoom_DifficultySelectUI {
   widgetClassPath=/PilgrimQuickplay/Widgets/WBP_Pilgrim_GreenRoom_DifficultySelectUI
   parent=SizeBox_6
   widget=WBP_Pilgrim_GreenRoom_DifficultySelectUI
   properties {showDifficultyIcon=true}
   slot {class=SizeBoxSlot}
  }
  Button_OpenMusicLibrary {
   widgetClassPath=&Button
   parent=VerticalBox_2
   widget=Button_OpenMusicLibrary
   properties {widgetStyle {
     clickedSlateSound {resourceObject=null}
     disabled {
      bIsDynamicallyLoaded=false
      drawAs=NoDrawType
      imageType=NoImage
      mirroring=NoMirror
      resourceName=None
      resourceObject=null
      tiling=NoTile
      imageSize=<0 0>
      margin=<ltrb 0 0 0 0>
      outlineSettings {
       bUseBrushTransparency=false
       roundingType=HalfHeightRadius
       width=0
       color {colorUseRule=UseColor_Specified, specifiedColor=<rgba 0 0 0 0>}
       cornerRadii=<0 0 0 0>
      }
      tintColor {colorUseRule=UseColor_Specified, specifiedColor=<rgba 1 1 1 1>}
      uVRegion {bIsValid=false, max=<0 0>, min=<0 0>}
     }
     disabledForeground {colorUseRule=UseColor_Specified, specifiedColor=<rgba 0.5271 0.5271 0.5271 1>}
     hovered {
      bIsDynamicallyLoaded=false
      drawAs=NoDrawType
      imageType=NoImage
      mirroring=NoMirror
      resourceName=None
      resourceObject=null
      tiling=NoTile
      imageSize=<32 32>
      margin=<ltrb 0 0 0 0>
      outlineSettings {
       bUseBrushTransparency=true
       roundingType=FixedRadius
       width=1
       color {colorUseRule=UseColor_Specified, specifiedColor=<rgba 0.7243 0.7243 0.7243 1>}
       cornerRadii=<4 4 4 4>
      }
      tintColor {colorUseRule=UseColor_Specified, specifiedColor=<rgba 0.7243 0.7243 0.7243 1>}
      uVRegion {bIsValid=false, max=<0 0>, min=<0 0>}
     }
     hoveredForeground {colorUseRule=UseColor_Specified, specifiedColor=<rgba 0 1 1 1>}
     hoveredSlateSound {resourceObject=null}
     normal {
      bIsDynamicallyLoaded=false
      drawAs=NoDrawType
      imageType=NoImage
      mirroring=NoMirror
      resourceName=None
      resourceObject=null
      tiling=NoTile
      imageSize=<32 32>
      margin=<ltrb 0 0 0 0>
      outlineSettings {
       bUseBrushTransparency=true
       roundingType=FixedRadius
       width=1
       color {colorUseRule=UseColor_Specified, specifiedColor=<rgba 0.6951 0.6951 0.6951 1>}
       cornerRadii=<4 4 4 4>
      }
      tintColor {colorUseRule=UseColor_Specified, specifiedColor=<rgba 0.4955 0.4955 0.4955 1>}
      uVRegion {bIsValid=false, max=<0 0>, min=<0 0>}
     }
     normalForeground {colorUseRule=UseColor_Specified, specifiedColor=<rgba 0 1 1 1>}
     normalPadding=<ltrb 0 0 0 0>
     pressed {
      bIsDynamicallyLoaded=false
      drawAs=NoDrawType
      imageType=NoImage
      mirroring=NoMirror
      resourceName=None
      resourceObject=null
      tiling=NoTile
      imageSize=<32 32>
      margin=<ltrb 0 0 0 0>
      outlineSettings {
       bUseBrushTransparency=true
       roundingType=FixedRadius
       width=1
       color {colorUseRule=UseColor_Specified, specifiedColor=<rgba 0.7243 0.7243 0.7243 1>}
       cornerRadii=<4 4 4 4>
      }
      tintColor {colorUseRule=UseColor_Specified, specifiedColor=<rgba 0.3843 0.3843 0.3843 1>}
      uVRegion {bIsValid=false, max=<0 0>, min=<0 0>}
     }
     pressedForeground {colorUseRule=UseColor_Specified, specifiedColor=<rgba 0 1 1 1>}
     pressedPadding=<ltrb 0 0 0 0>
     pressedSlateSound {resourceObject=null}
    }}
   slot {class=VerticalBoxSlot, properties {padding=<ltrb 0 0 0 16>}}
  }
  SizeBox_2 {
   widgetClassPath=&SizeBox
   parent=Button_OpenMusicLibrary
   widget=SizeBox_2
   properties {bOverride_HeightOverride=true, heightOverride=95, widthOverride=720}
   slot {class=ButtonSlot, properties {horizontalAlignment=HAlign_Fill, verticalAlignment=VAlign_Fill, padding=<ltrb 0 0 0 0>}}
  }
  WBP_Pilgrim_GreenRoom_ChooseASongUI {
   widgetClassPath=/PilgrimQuickplay/Widgets/WBP_Pilgrim_GreenRoom_ChooseASongUI
   parent=SizeBox_2
   widget=WBP_Pilgrim_GreenRoom_ChooseASongUI
   slot {class=SizeBoxSlot}
  }
  Button_Ready {
   widgetClassPath=&Button
   parent=VerticalBox_2
   widget=Button_Ready
   properties {widgetStyle {
     clickedSlateSound {resourceObject=null}
     disabled {
      bIsDynamicallyLoaded=false
      drawAs=NoDrawType
      imageType=NoImage
      mirroring=NoMirror
      resourceName=None
      resourceObject=null
      tiling=NoTile
      imageSize=<0 0>
      margin=<ltrb 0 0 0 0>
      outlineSettings {
       bUseBrushTransparency=false
       roundingType=HalfHeightRadius
       width=0
       color {colorUseRule=UseColor_Specified, specifiedColor=<rgba 0 0 0 0>}
       cornerRadii=<0 0 0 0>
      }
      tintColor {colorUseRule=UseColor_Specified, specifiedColor=<rgba 1 1 1 1>}
      uVRegion {bIsValid=false, max=<0 0>, min=<0 0>}
     }
     disabledForeground {colorUseRule=UseColor_Specified, specifiedColor=<rgba 0.5271 0.5271 0.5271 1>}
     hovered {
      bIsDynamicallyLoaded=false
      drawAs=NoDrawType
      imageType=NoImage
      mirroring=NoMirror
      resourceName=None
      resourceObject=null
      tiling=NoTile
      imageSize=<32 32>
      margin=<ltrb 0 0 0 0>
      outlineSettings {
       bUseBrushTransparency=true
       roundingType=FixedRadius
       width=1
       color {colorUseRule=UseColor_Specified, specifiedColor=<rgba 0.7243 0.7243 0.7243 1>}
       cornerRadii=<4 4 4 4>
      }
      tintColor {colorUseRule=UseColor_Specified, specifiedColor=<rgba 0.7243 0.7243 0.7243 1>}
      uVRegion {bIsValid=false, max=<0 0>, min=<0 0>}
     }
     hoveredForeground {colorUseRule=UseColor_Specified, specifiedColor=<rgba 0 1 1 1>}
     hoveredSlateSound {resourceObject=null}
     normal {
      bIsDynamicallyLoaded=false
      drawAs=NoDrawType
      imageType=NoImage
      mirroring=NoMirror
      resourceName=None
      resourceObject=null
      tiling=NoTile
      imageSize=<32 32>
      margin=<ltrb 0 0 0 0>
      outlineSettings {
       bUseBrushTransparency=true
       roundingType=FixedRadius
       width=1
       color {colorUseRule=UseColor_Specified, specifiedColor=<rgba 0.6951 0.6951 0.6951 1>}
       cornerRadii=<4 4 4 4>
      }
      tintColor {colorUseRule=UseColor_Specified, specifiedColor=<rgba 0.4955 0.4955 0.4955 1>}
      uVRegion {bIsValid=false, max=<0 0>, min=<0 0>}
     }
     normalForeground {colorUseRule=UseColor_Specified, specifiedColor=<rgba 0 1 1 1>}
     normalPadding=<ltrb 0 0 0 0>
     pressed {
      bIsDynamicallyLoaded=false
      drawAs=NoDrawType
      imageType=NoImage
      mirroring=NoMirror
      resourceName=None
      resourceObject=null
      tiling=NoTile
      imageSize=<32 32>
      margin=<ltrb 0 0 0 0>
      outlineSettings {
       bUseBrushTransparency=true
       roundingType=FixedRadius
       width=1
       color {colorUseRule=UseColor_Specified, specifiedColor=<rgba 0.7243 0.7243 0.7243 1>}
       cornerRadii=<4 4 4 4>
      }
      tintColor {colorUseRule=UseColor_Specified, specifiedColor=<rgba 0.3843 0.3843 0.3843 1>}
      uVRegion {bIsValid=false, max=<0 0>, min=<0 0>}
     }
     pressedForeground {colorUseRule=UseColor_Specified, specifiedColor=<rgba 0 1 1 1>}
     pressedPadding=<ltrb 0 0 0 0>
     pressedSlateSound {resourceObject=null}
    }}
   slot {class=VerticalBoxSlot}
  }
  SizeBox {
   widgetClassPath=&SizeBox
   parent=Button_Ready
   widget=SizeBox
   properties {bOverride_HeightOverride=true, heightOverride=95, widthOverride=720}
   slot {class=ButtonSlot, properties {horizontalAlignment=HAlign_Fill, verticalAlignment=VAlign_Fill, padding=<ltrb 0 0 0 0>}}
  }
  WBP_Pilgrim_GreenRoom_ToggleReadyUI {
   widgetClassPath=/PilgrimQuickplay/Widgets/WBP_Pilgrim_GreenRoom_ToggleReadyUI
   bIsVariable=true
   parent=SizeBox
   widget=WBP_Pilgrim_GreenRoom_ToggleReadyUI
   slot {class=SizeBoxSlot}
  }
 ]
}
